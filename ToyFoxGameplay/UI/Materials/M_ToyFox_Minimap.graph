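# M_ToyFox_Minimap is an alpha-composite UI material that draws decoration for the ToyFox HUD minimap.
# Box SDFs build outer and inner outlines and an overlay, anti-aliased through shared UI functions.
# A pattern texture, masked by an inverted radial gradient and scaled by Hex Intensity, forms the hex layer.
# A panned sine over radial coordinates creates a pulse, and weighted sums combine masks with grey and red tints.
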Material /ToyFoxGameplay/UI/Materials/M_ToyFox_Minimap
graph {
 properties {bAutomaticallySetUsageInEditor=false, blendMode=BLEND_AlphaComposite, materialDomain=MD_UI, usedShadingModels=Default Lit}
 nodes [
  n1 = Multiply(A=NamedRerouteUsage()) {constB=0.1}
  n2 = Multiply(A=@n1, B=Constant3Vector() {constant=<rgba 0.526 0.526 0.526 0>})
  n3 = NamedRerouteUsage()
  n4 = Constant3Vector() {constant=<rgba 0.7292 0.0291 0.0281 1>}
  n5 = Multiply(A=Add(A=@n1, B=Multiply(A=@n3) {constB=0.2}), B=@n4)
  n6 = Multiply(A=NamedRerouteUsage()) {constB=0.05}
  n7 = NamedRerouteUsage()
  Add(
   A=Add(A=Add(A=@n2, B=@n5), B=Multiply(A=@n4, B=@n6))
   B=Multiply(A=Multiply(A=@n7) {constB=0.12}, B=Constant3Vector() {constant=<rgba 1 0 0 0>})
  )
  Add(A=Multiply(A=NamedRerouteUsage()) {constB=0.2}, B=Multiply(A=NamedRerouteUsage()) {constB=0.3})
  n8 = MaterialFunctionCall(UVs=TextureCoordinate(), Radius=Constant() {r=0.6}, Density=Constant() {r=5}) {materialFunction=&RadialGradientExponential}
  n9 = NamedRerouteUsage()
  n10 = Constant() {r=0.6}
  n11 = TextureSample(UVs=Add(
    A=MaterialFunctionCall(UVs=@n9, Texture Scale=@n10) {materialFunction=&ScaleUVsByCenter}.UVs
    B=NamedRerouteUsage()
   )) {samplerType=SAMPLERTYPE_Alpha, texture=/ToyFoxGameplay/UI/Widgets/HUD/Assets/HUDFrame/T_UI_ToyFox_Pattern_02}
  n12 = Constant() {r=0.1}
  n13 = Constant() {r=0.08}
  n14 = Multiply(
   A=Multiply(
    A=OneMinus(None=@n8.RadialGradientExponential)
    B=MaterialFunctionCall(DistanceField=@n11.R, Threshold=@n12, Smoothing=@n13) {materialFunction=&MF_UI_AntiAliasing}.Outer
   )
   B=LinearInterpolate(Alpha=ScalarParameter() {parameterName=Hex Intensity}) {constA=2, constAlpha=1, constB=10}
  ) {constB=2}
  n15 = Multiply(A=@n14) {constB=0.2}
  n16 = Multiply(A=Multiply(A=Multiply(A=@n14) {constB=5}, B=NamedRerouteUsage())) {constB=0.2}
  NamedRerouteDeclaration(None=Multiply(A=Max(A=@n15, B=@n16), B=NamedRerouteUsage())) {name=Hex, nodeColor=<rgba 0 0 0 1>}
  NamedRerouteDeclaration(None=MaterialFunctionCall() {materialFunction=&MF_UI_UVScaling}.Scaled to Fit) {name=Scale to Fit, nodeColor=<rgba 0 0 0 1>}
  n17 = MaterialFunctionCall!(UVs=NamedRerouteUsage(), Size=Constant() {r=0.98}, CornerRoundness, KeepAspectRatio, PixelSize) {materialFunction=/FortUILibrary/MaterialFunctions/Shapes/SDFs/MF_SDF_Box}
  n18 = Constant() {r=0.015}
  n19 = Constant()
  NamedRerouteDeclaration(None=MaterialFunctionCall(DistanceField=MaterialFunctionCall!(DistanceField=@n17.SDF, Threshold, Thickness=@n18, Alignment=@n19, Smoothing) {materialFunction=/FortUILibrary/MaterialFunctions/Shapes/Modifiers/MF_UI_ShapeOutline}.SDFOutline) {materialFunction=&MF_UI_AntiAliasing}.Inner) {name=Outline, nodeColor=<rgba 0 0 0 1>}
  n20 = MaterialFunctionCall(UVs=NamedRerouteUsage(), Texture Scale=Constant() {r=1.4}) {materialFunction=&ScaleUVsByCenter}
  n21 = Constant() {r=0.57}
  n22 = Constant() {r=-0.5}
  NamedRerouteDeclaration(None=MaterialFunctionCall(
    DistanceField=MaterialFunctionCall(UVs=@n20.UVs, Size=@n21, CornerRoundness=@n22) {materialFunction=&MF_SDF_Box}.SDF
    Threshold=Constant() {r=0.1}
    Smoothing=Constant()
   ) {materialFunction=&MF_UI_AntiAliasing}.Inner) {name=Box Overlay, nodeColor=<rgba 0 0 0 1>}
  n23 = MaterialFunctionCall(UVs=NamedRerouteUsage(), Size=Constant() {r=0.92}) {materialFunction=&MF_SDF_Box}
  n24 = Constant() {r=0.015}
  n25 = Constant()
  NamedRerouteDeclaration(None=MaterialFunctionCall(DistanceField=MaterialFunctionCall(DistanceField=@n23.SDF, Thickness=@n24, Alignment=@n25) {materialFunction=&MF_UI_ShapeOutline}.SDFOutline) {materialFunction=&MF_UI_AntiAliasing}.Inner) {name=Outline_Inner, nodeColor=<rgba 0 0 0 1>}
  n26 = NamedRerouteUsage()
  n27 = Multiply(
   A=MaterialFunctionCall(UVs=MaterialFunctionCall!(Vector or UVs=@n26) {materialFunction=/FortUILibrary/MaterialFunctions/Basics/MF_UI_CheaperVectorToRadial}.Radial Coordinates) {materialFunction=&ScaleUVsByCenter}.UVs
   B=Constant2Vector() {g=0.5}
  )
  n28 = Constant2Vector() {g=-0.25}
  n29 = ComponentMask(None=Sine(None=Panner(Coordinate=@n27, Speed=@n28))) {g=true}
  NamedRerouteDeclaration(None=Power(Base=Saturate(None=@n29)) {constExponent=3}) {name=RadialPulse, nodeColor=<rgba 0 0 0 1>}
 ]
}
