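# ToyFox HUD element widget that displays a minimap overlay image, built on FortHUDElementWidget.
# An Overlay root holds a single Image that fills it, draws the M_ToyFox_Minimap material and is scaled up slightly.
# The widget has no graph logic or variables of its own, and it also carries an inherited TutorialHighlight widget.
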WidgetBlueprint /ToyFoxGameplay/UI/Widgets/HUD/WBP_ToyFox_AthenaMinimapContainer
graphs [EventGraph {links [], nodes []}]
variables []
widgets {
 info {inheritedWidgetCount=1, parentClass=&FortHUDElementWidget, rootWidgetClass=&Overlay, widgetCount=3}
 widgets [
  Overlay_0 {widgetClassPath=&Overlay, widget=Overlay_0}
  Image_MinimapOverlay {
   widgetClassPath=&Image
   bIsVariable=true
   parent=Overlay_0
   widget=Image_MinimapOverlay
   properties {
    brush {
     bIsDynamicallyLoaded=false
     drawAs=Image
     imageType=FullColor
     mirroring=NoMirror
     resourceName=None
     resourceObject=/ToyFoxGameplay/UI/Materials/M_ToyFox_Minimap
     tiling=NoTile
     imageSize=<126 126>
     margin=<ltrb 0 0 0 0>
     outlineSettings {
      bUseBrushTransparency=false
      roundingType=HalfHeightRadius
      width=0
      color {colorUseRule=UseColor_Specified, specifiedColor=<rgba 0 0 0 0>}
      cornerRadii=<0 0 0 0>
     }
     tintColor {colorUseRule=UseColor_Specified, specifiedColor=<rgba 1 1 1 1>}
     uVRegion {bIsValid=false, max=<0 0>, min=<0 0>}
    }
    renderTransform {angle=0, scale=<1.05 1.05>, shear=<0 0>, translation=<0 0>}
   }
   slot {class=OverlaySlot, properties {horizontalAlignment=HAlign_Fill, verticalAlignment=VAlign_Fill}}
  }
  TutorialHighlight {widgetClassPath=&TutorialHighlight}
 ]
}
