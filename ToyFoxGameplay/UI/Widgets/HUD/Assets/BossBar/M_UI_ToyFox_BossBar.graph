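# UI material for the ToyFox HUD boss health-style bar, drawn with alpha-composite blending.
# Mirrored SDF textures, thresholded through anti-aliasing functions, form the frame, outline, background and fill masks.
# A Progress parameter moves the fill cutoff and its edge highlight, while rotated, panning boxes modulated by noise create a fill texture.
# Color and opacity are layered separately, with parameters for hit, pulsing and frame flashes, a shine sweep, blend amounts and master opacity.
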Material /ToyFoxGameplay/UI/Widgets/HUD/Assets/BossBar/M_UI_ToyFox_BossBar
graph {
 properties {bAutomaticallySetUsageInEditor=false, blendMode=BLEND_AlphaComposite, materialDomain=MD_UI, usedShadingModels=Default Lit}
 nodes [
  Add(A=NamedRerouteUsage()) {constB=0.1}
  Constant3Vector() {constant=<rgba 3 0.2876 0.279 1>}
  MaterialFunctionCall(Speed=ScalarParameter() {defaultValue=0.05, desc=Test, parameterName=DebugTimeSine}) {materialFunction=&DebugTimeSine}
  n1 = NamedRerouteUsage()
  Multiply(A=NamedRerouteUsage(), B=@n1)
  n2 = Multiply(A=NamedRerouteUsage()) {constB=0.1}
  n3 = LinearInterpolate(A=Constant3Vector() {constant=<rgba 0.82 0 0.0772 1>}, Alpha=ScalarParameter() {parameterName=Final Hit Flash}) {constB=0.3}
  n4 = Constant3Vector() {constant=<rgba 0.612 0 0 0>}
  n5 = Multiply(A=@n4) {constB=0.1}
  n6 = ScalarParameter() {parameterName=Final Hit Flash}
  n7 = NamedRerouteUsage()
  n8 = ScalarParameter() {defaultValue=0.8, parameterName=Progress Opacity}
  n9 = LinearInterpolate(
   A=@n4
   B=Constant3Vector() {constant=<rgba 0.0391 0.0022 0 1>}
   Alpha=MaterialFunctionCall(Speed=Constant() {r=2}) {materialFunction=&DebugTimeSine}.Result
  )
  n10 = ScalarParameter() {parameterName=Flash On}
  n11 = Multiply(A=NamedRerouteUsage())
  n12 = LinearInterpolate(
   A=Multiply(A=LinearInterpolate(A=@n5, Alpha=@n6) {constB=0}, B=Multiply(A=@n7, B=@n8))
   B=Multiply(A=LinearInterpolate(A=@n4, B=@n9, Alpha=@n10), B=@n11)
   Alpha=@n11
  )
  n13 = Constant3Vector() {constant=<rgba 3 0.0729 0.0342 1>}
  n14 = Multiply(A=NamedRerouteUsage()) {constB=0.2}
  n15 = Constant3Vector() {constant=<rgba 0.1667 0.1667 0.1667 1>}
  n16 = ScalarParameter() {parameterName=Frame Flash}
  n17 = Add(
   A=Add(A=Multiply(A=@n2, B=@n3), B=LinearInterpolate(A=@n12, B=@n13, Alpha=@n14))
   B=Multiply(A=LinearInterpolate(A=@n15, Alpha=@n16), B=NamedRerouteUsage())
  )
  n18 = Multiply(A=NamedRerouteUsage(), B=NamedRerouteUsage()) {constA=1}
  Multiply(A=Saturate(None=Add(A=@n17, B=@n18)), B=ScalarParameter() {defaultValue=1, parameterName=Master Opacity})
  n19 = Multiply(A=NamedRerouteUsage(), B=ScalarParameter() {defaultValue=0.5, parameterName=Progress Blend})
  n20 = Multiply(A=NamedRerouteUsage(), B=ScalarParameter() {defaultValue=0.5, parameterName=Frame Blend})
  n21 = Add(A=Multiply(A=NamedRerouteUsage()) {constB=0.4}, B=Add(A=Multiply(A=@n19) {constB=0.8}, B=Multiply(A=@n20) {constB=0}))
  n22 = Multiply(A=NamedRerouteUsage(), B=ScalarParameter() {defaultValue=0.1, parameterName=Progress Edge Blend})
  Multiply(A=Saturate(None=Add(A=@n21, B=@n22)), B=ScalarParameter() {defaultValue=1, parameterName=Master Opacity})
  n23 = MaterialFunctionCall() {desc=Uvs, materialFunction=&GetUserInterfaceUV}
  n24 = TextureSample(UVs=MaterialFunctionCall!(UV=@n23.Normalized UV, Mirror Y, Swizzle UV) {materialFunction=/FortUILibrary/MaterialFunctions/Basics/MF_UI_Symmetry_2x1_New}.Symmetry) {samplerType=SAMPLERTYPE_LinearGrayscale, texture=&/T_UI_ToyFox_BossFrame}
  n25 = Constant() {r=0.4}
  NamedRerouteDeclaration(None=Multiply(A=MaterialFunctionCall(DistanceField=@n24.R, Threshold=@n25) {materialFunction=&MF_UI_AntiAliasing}.Outer) {constB=0.4}) {name=Frame, nodeColor=<rgba 0 0 0 1>}
  n26 = TextureSample(UVs=MaterialFunctionCall(UV=@n23.Normalized UV) {materialFunction=&MF_UI_Symmetry_2x1_New}.Symmetry) {samplerType=SAMPLERTYPE_LinearGrayscale, texture=&/T_UI_ToyFox_SDF_Fill}
  n27 = Constant() {r=0.65}
  n28 = MaterialFunctionCall(DistanceField=@n26.R, Threshold=Constant() {r=0.52}) {materialFunction=&MF_UI_AntiAliasing}
  n29 = NamedRerouteDeclaration(None=Multiply(
    A=MaterialFunctionCall(DistanceField=@n26.R, Threshold=@n27) {materialFunction=&MF_UI_AntiAliasing}.Inner
    B=@n28.Outer
   )) {name=Progress Outline, nodeColor=<rgba 0 0 0 1>}
  NamedRerouteDeclaration(None=Add(A=Multiply(A=@n29) {constB=0.1, desc=Outline Additive Blend}, B=@n28.Outer)) {name=Progress Bg, nodeColor=<rgba 0 0 0 1>}
  n30 = TextureCoordinate()
  n31 = Subtract(A=ComponentMask(None=@n30) {r=true}) {constB=0.5}
  n32 = Time()
  n33 = LinearInterpolate(
   A=ScalarParameter() {parameterName=Janky Shine}
   B=Frac(None=Multiply(A=@n32) {constB=0.5})
   Alpha=ScalarParameter() {parameterName=Auto Shine}
  ) {constAlpha=0}
  n34 = OneMinus(None=Multiply(A=Abs(None=@n31), B=LinearInterpolate(Alpha=@n33) {constA=16, constB=0}) {constB=5})
  n35 = AppendVector(A=Add(A=Multiply(A=@n34) {constB=5}) {constB=-3.4073}, B=Add(A=ComponentMask(None=@n30) {g=true}) {constB=0.218})
  n36 = Multiply(A=Saturate(None=MaterialFunctionCall(UV Channel=@n35) {materialFunction=&LinearGradient}.UGradient))
  n37 = Abs(None=Subtract(A=Frac(None=@n36)) {constB=0.5})
  n38 = Power(Base=OneMinus(None=Multiply(A=@n37) {constB=2}))
  n39 = NamedRerouteUsage()
  NamedRerouteDeclaration(None=Multiply(A=Multiply(A=@n38, B=@n39), B=ScalarParameter() {defaultValue=1, parameterName=Shine Opacity})) {name=Janky Shine, nodeColor=<rgba 0 0 0 1>}
  n40 = Time()
  n41 = MaterialFunctionCall!(
   Gradient
   a
   b
   c
   d=MaterialFunctionCall(R=Multiply(A=@n40) {constB=5}, G=Constant() {r=0.7}, B=Constant() {r=0.1}) {materialFunction=&AppendMany}.RGB
  ) {materialFunction=/FortUILibrary/MaterialFunctions/Color/MF_UI_CosineGradient}
  NamedRerouteDeclaration(None=Add(A=Multiply(A=@n41.Result)) {constB=0.2}) {name=Party Colors, nodeColor=<rgba 0 0 0 1>}
  n42 = MaterialFunctionCall() {materialFunction=&MF_UI_UVScaling}
  NamedRerouteDeclaration(None=@n42.Scaled to Fit) {name=Scale To Fit, nodeColor=<rgba 0 0 0 1>}
  n43 = NamedRerouteUsage()
  n44 = Constant2Vector() {g=0.1, r=0.1}
  n45 = Constant() {r=0.86}
  n46 = Multiply(A=MaterialFunctionCall(UVs=Panner(Coordinate=@n43, Speed=@n44), Rotation Center=@n45, "Rotation Angle (0-1)"=@n45) {materialFunction=&CustomRotator}.Rotated Values) {constB=4}
  n47 = Panner(Coordinate=TextureCoordinate() {uTiling=4, vTiling=0.2}, Speed=AppendVector(A=Constant() {r=0.2}, B=Constant() {r=0.4097})) {speedY=0.5}
  n48 = Multiply(
   A=MaterialFunctionCall(UVs=Frac(None=@n46)) {materialFunction=&MF_UI_Box}.SDF
   B=Add(A=TextureSample(UVs=@n47) {samplerType=SAMPLERTYPE_Grayscale, texture=/Game/UI/Frontend/ItemShop/Art/Rumble/LowResBlurredNoise_Rumble}.R) {constB=0.6}
  ) {constA=0.6}
  n49 = Constant() {r=0.45}
  n50 = Constant() {r=0.2}
  NamedRerouteDeclaration(None=Multiply(A=MaterialFunctionCall(DistanceField=@n48, Threshold=@n49, Smoothing=@n50) {materialFunction=&MF_UI_AntiAliasing}.Outer) {constB=0.1}) {name=Fill Texture, nodeColor=<rgba 0 0 0 1>}
  NamedRerouteDeclaration(None=@n42.Scaled To Fill) {name=Scale To Fill, nodeColor=<rgba 0 0 0 1>}
  n51 = Saturate(None=TextureCoordinate())
  n52 = Constant() {desc=Progress Start, r=0.33}
  n53 = Constant() {desc=Progress End, r=1.28}
  n54 = ScalarParameter() {desc=Default 0, group=Progress, parameterName=Progress, sliderMax=1, sortPriority=3}
  n55 = ComponentMask(None=Add(A=Multiply(A=@n51) {constB=-1}, B=LinearInterpolate(A=@n52, B=@n53, Alpha=@n54)) {constB=0.5}) {r=true}
  n56 = MaterialFunctionCall(DistanceField=@n55, Threshold=Constant() {r=0.3}) {materialFunction=&MF_UI_AntiAliasing}
  n57 = Multiply(A=@n56.Outer, B=NamedRerouteUsage())
  n58 = Multiply(A=@n57) {constB=0.7}
  n59 = Multiply(A=@n57, B=@n1)
  NamedRerouteDeclaration(None=Saturate(None=Add(A=@n58, B=@n59))) {name=Progress Fill, nodeColor=<rgba 0 0 0 1>}
  n60 = Multiply(A=Add(A=OneMinus(None=@n55)) {constB=-0.58}) {constB=10}
  n61 = Multiply(A=@n56.Outer, B=@n60)
  NamedRerouteDeclaration(None=Saturate(None=Multiply(A=@n61, B=@n57))) {name=Progress Edge Highlight, nodeColor=<rgba 0 0 0 1>}
  NamedRerouteDeclaration(None=MaterialFunctionCall() {materialFunction=&LinearGradient}.UGradient) {name=UGradient, nodeColor=<rgba 0 0 0 1>}
  NamedRerouteDeclaration(None=MaterialFunctionCall(DistanceField=@n26.R, Threshold=Constant() {r=0.45}, Smoothing=Constant() {r=0.2}) {materialFunction=&MF_UI_AntiAliasing}.Outer) {name=Progress Bar Edge Highlight Mask, nodeColor=<rgba 0 0 0 1>}
  NamedRerouteUsage()
  Power(Base=@n60) {constExponent=10}
 ]
}
