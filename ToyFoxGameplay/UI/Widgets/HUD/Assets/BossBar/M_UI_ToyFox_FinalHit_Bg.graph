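# UI material for the background of the ToyFox boss bar "Final Hit" HUD element, drawn with alpha-composite blending.
# A mirrored, skewed SDF box forms the text backing and outline, colored by a time-animated cosine gradient that can swap to near-black.
# A mirrored edge texture forms brackets with parameter-driven pulse and slide.
# Edge Switcher blends backing and brackets in the color and opacity chains, with Edge Flash In tinting and Master Opacity scaling both.
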Material /ToyFoxGameplay/UI/Widgets/HUD/Assets/BossBar/M_UI_ToyFox_FinalHit_Bg
graph {
 properties {bAutomaticallySetUsageInEditor=false, blendMode=BLEND_AlphaComposite, materialDomain=MD_UI, usedShadingModels=Default Lit}
 nodes [
  n1 = Multiply(A=NamedRerouteUsage())
  Add(A=@n1, B=Multiply(A=NamedRerouteUsage(), B=ScalarParameter() {parameterName=Bg Outline Opacity}))
  n2 = NamedRerouteUsage()
  n3 = MaterialFunctionCall(UVs=@n2, "Rotation Angle (0-1)"=Constant() {r=0.5}) {materialFunction=&CustomRotator}
  n4 = Constant2Vector() {g=0.4, r=0.9}
  Multiply(
   A=MaterialFunctionCall(UVs=@n2, Size=Constant2Vector() {g=0.07, r=1}) {materialFunction=&MF_UI_Box}.Fill
   B=Multiply(A=MaterialFunctionCall!(UVs=@n3.Rotated Values, Size=@n4, Sides, InnerAngle, OutlineThickness, OutlineAlignment, GlowMin, GlowMax) {materialFunction=/FortUILibrary/MaterialFunctions/Shapes/MF_UI_RegularPolygon}.Fill)
  )
  n5 = Time()
  n6 = MaterialFunctionCall!(
   Gradient
   a
   b
   c
   d=MaterialFunctionCall(R=Multiply(A=@n5) {constB=5}, G=Constant() {r=0.3}, B=Constant()) {materialFunction=&AppendMany}.RGB
  ) {materialFunction=/FortUILibrary/MaterialFunctions/Color/MF_UI_CosineGradient}
  n7 = Multiply(
   A=LinearInterpolate(
    A=NamedRerouteDeclaration(None=@n6.Result) {name=Party O Clock, nodeColor=<rgba 0 0 0 1>}
    B=Constant3Vector() {constant=<rgba 0.0078 0 0 1>}
    Alpha=ScalarParameter() {defaultValue=1, parameterName=Text Bg Color Swap}
   )
   B=NamedRerouteUsage()
  )
  n8 = Multiply(
   A=NamedRerouteUsage()
   B=Multiply(A=Constant3Vector() {constant=<rgba 0.6302 0 0.0095 1>}, B=ScalarParameter() {parameterName=Bg Outline Opacity})
  )
  n9 = Multiply(A=NamedRerouteUsage(), B=ScalarParameter() {defaultValue=0.4, parameterName=Edge Opacity})
  n10 = LinearInterpolate(A=NamedRerouteUsage(), B=Constant3Vector() {constant=<rgba 0.4479 0.0064 0 1>}, Alpha=ScalarParameter() {parameterName=Edge Flash In})
  n11 = ScalarParameter() {defaultValue=1, parameterName=Master Opacity}
  Multiply(
   A=LinearInterpolate(
    A=Add(A=@n7, B=@n8)
    B=Multiply(A=@n9, B=@n10) {constB=0.6}
    Alpha=ScalarParameter() {defaultValue=0.5, desc=default: 0, parameterName=Edge Switcher}
   )
   B=@n11
  )
  n12 = NamedRerouteUsage()
  n13 = ScalarParameter() {defaultValue=0.4, parameterName=Edge Opacity}
  Multiply(
   A=LinearInterpolate(A=@n1, B=Multiply(A=@n12, B=@n13) {constB=0.4}, Alpha=ScalarParameter() {defaultValue=0.5, desc=default: 0, parameterName=Edge Switcher})
   B=@n11
  )
  n14 = MaterialFunctionCall(Float2=MaterialFunctionCall!(UV=MaterialFunctionCall() {materialFunction=&MF_UI_UVScaling}.ScaleX, Mirror Y, Swizzle UV) {materialFunction=/FortUILibrary/MaterialFunctions/Basics/MF_UI_Symmetry_2x1_New}.Symmetry) {materialFunction=&BreakOutFloat2Components}
  n15 = MaterialFunctionCall!(UV=AppendVector(A=Subtract(A=@n14.R) {constB=0.2}, B=@n14.G), Skew Angle=Constant2Vector() {r=45}) {materialFunction=/FortUILibrary/MaterialFunctions/Basics/MF_UI_Skew}
  n16 = Constant() {r=0.98}
  n17 = Constant() {r=0.08}
  n18 = MaterialFunctionCall(DistanceField=MaterialFunctionCall!(
    UVs=@n15.Skewed UVs
    Size=AppendVector(A=@n16, B=@n17)
    CornerRoundness=Constant() {r=0.01}
    KeepAspectRatio=StaticBool() {value=true}
    PixelSize
   ) {materialFunction=/FortUILibrary/MaterialFunctions/Shapes/SDFs/MF_SDF_Box}.SDF) {materialFunction=&MF_UI_AntiAliasing}
  n19 = MaterialFunctionCall(UVs=@n15.Skewed UVs, Size=AppendVector(A=Add(A=@n16) {constB=0.02}, B=@n17)) {materialFunction=&MF_SDF_Box}
  n20 = Constant() {r=0.02}
  NamedRerouteDeclaration(None=Multiply(
    A=NamedRerouteDeclaration(None=@n18.Inner) {name=Text Bg, nodeColor=<rgba 0 0 0 1>}
    B=MaterialFunctionCall!(DistanceField=@n19.SDF, Threshold, Thickness=@n20, Alignment, Smoothing) {materialFunction=/FortUILibrary/MaterialFunctions/Shapes/Modifiers/MF_UI_ShapeOutline}.Outline
   )) {name=Text Bg_Outline, nodeColor=<rgba 0 0 0 1>}
  n21 = MaterialFunctionCall() {materialFunction=&MF_UI_UVScaling}
  NamedRerouteDeclaration(None=@n21.ScaleX) {name=Scale X, nodeColor=<rgba 0 0 0 1>}
  n22 = Constant2Vector() {g=0.16, r=0.3}
  n23 = Constant2Vector() {g=0.16, r=0.31}
  n24 = LinearInterpolate(
   B=MaterialFunctionCall(Speed=Constant() {r=2}) {materialFunction=&DebugTimeSine}.Result
   Alpha=ScalarParameter() {parameterName=Bracket Pulse}
  )
  n25 = MaterialFunctionCall(UV=MaterialFunctionCall(UVs=NamedRerouteUsage(), Texture Scale=LinearInterpolate(A=@n22, B=@n23, Alpha=@n24)) {materialFunction=&ScaleUVsByCenter}.UVs) {materialFunction=&MF_UI_Symmetry_2x1_New}
  n26 = ScalarParameter() {parameterName=Bracket Movement}
  n27 = AppendVector(
   A=Add(A=ComponentMask(None=@n25.Symmetry) {r=true}, B=LinearInterpolate(Alpha=@n26) {constA=2.1, constB=0}) {desc=Offset}
   B=ComponentMask(None=@n25.Symmetry) {g=true}
  )
  NamedRerouteDeclaration(None=MaterialFunctionCall(
    DistanceField=TextureSample(UVs=@n27) {samplerType=SAMPLERTYPE_LinearGrayscale, texture=&/T_UI_ToyFox_FinalHitEdge}.R
    Threshold=Constant() {r=0.3}
    Smoothing=Constant() {r=0.1}
   ) {materialFunction=&MF_UI_AntiAliasing}.Outer) {name=Edges, nodeColor=<rgba 0 0 0 1>}
  NamedRerouteDeclaration(None=@n21.Scaled to Fit) {name=Scale To Fit, nodeColor=<rgba 0 0 0 1>}
  NamedRerouteUsage()
 ]
}
