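# UI material for the ToyFox boss bar final-hit exclamation icon, drawn from a distance-field texture with alpha-composite blending.
# A center copy and two side copies, spread by R Move, get separate opacities and a time-animated cosine-gradient color mixed with white.
# Parameters drive horizontal shake and pulsing scale, while scrolling glitch noise builds periodically distorted icon UVs.
# Rotated box SDFs define a diamond and outline, and some layers are multiplied by zero.
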Material /ToyFoxGameplay/UI/Widgets/HUD/Assets/BossBar/M_UI_ToyFox_FinalHit_Exclamation
graph {
 properties {bAutomaticallySetUsageInEditor=false, blendMode=BLEND_AlphaComposite, materialDomain=MD_UI, usedShadingModels=Default Lit}
 nodes [
  Add(
   A=Multiply(A=NamedRerouteUsage(), B=ScalarParameter() {defaultValue=0.5, parameterName=Icon Opacity})
   B=Multiply(A=NamedRerouteUsage()) {constB=0}
  )
  n1 = Multiply(A=Time()) {constB=5}
  n2 = Constant() {r=0.7}
  n3 = Constant() {r=0.1}
  n4 = Multiply(A=MaterialFunctionCall!(
    Gradient
    a
    b
    c
    d=MaterialFunctionCall(R=@n1, G=@n2, B=@n3) {materialFunction=&AppendMany}.RGB
   ) {materialFunction=/FortUILibrary/MaterialFunctions/Color/MF_UI_CosineGradient}.Result)
  n5 = Multiply(A=NamedRerouteUsage(), B=ScalarParameter() {defaultValue=0.5, parameterName=Icon Opacity})
  n6 = Multiply(A=LinearInterpolate(A=Constant3Vector() {constant=<rgba 1 1 1 1>}, B=@n4, Alpha=NamedRerouteUsage()), B=@n5)
  n7 = Multiply(A=NamedRerouteUsage(), B=ScalarParameter() {defaultValue=0.5, parameterName=Icon RL Opacity})
  n8 = NamedRerouteUsage()
  Add(A=Add(A=@n6, B=Multiply(A=@n4, B=@n7)), B=Multiply(A=Constant3Vector() {constant=<rgba 0.0234 0 0.0003 1>}, B=Multiply(A=@n8) {constB=0}))
  Constant3Vector() {constant=<rgba 1 0 0.6542 1>}
  Constant3Vector() {constant=<rgba 0.487 0.487 0.487 1>}
  Constant() {r=-0.2}
  Frac(None=Multiply(A=Time()) {constB=0})
  n9 = Constant() {r=0.5}
  LinearInterpolate(Alpha=@n9) {constB=0.7}
  n10 = Constant3Vector() {constant=<rgba 0.0208 0.0208 0.0208 1>}
  n11 = Constant3Vector() {constant=<rgba 0.1536 0.1536 0.1536 1>}
  n12 = Multiply(A=NamedRerouteUsage()) {constB=0.2}
  LinearInterpolate(A=Multiply(A=LinearInterpolate(A=@n10, B=@n11, Alpha=@n12), B=@n12), B=@n6, Alpha=@n5)
  LinearInterpolate()
  n13 = NamedRerouteUsage()
  n14 = Add(A=@n13, B=Constant2Vector())
  n15 = MaterialFunctionCall(
   UVs=MaterialFunctionCall(UVs=@n14, "Rotation Angle (0-1)"=Constant() {r=0.12}) {materialFunction=&CustomRotator}.Rotated Values
   Size=LinearInterpolate(Alpha=ScalarParameter() {defaultValue=0.5, parameterName=Diamond Scale}) {constB=0.7}
  ) {materialFunction=&MF_UI_Box}
  MaterialFunctionCall(DistanceField=@n15.SDFOutline, Threshold=Constant() {r=0.1}) {materialFunction=&MF_UI_AntiAliasing}
  MaterialFunctionCall() {materialFunction=&MF_UI_EaseCurves}
  MaterialFunctionCall() {materialFunction=&DebugTimeSine}
  n16 = LinearInterpolate(Alpha=MaterialFunctionCall(Speed=Constant() {r=10}) {materialFunction=&DebugTimeSine}.Result) {constA=0.02, constB=-0.02}
  n17 = ScalarParameter() {parameterName=Shakey Shake}
  n18 = Add(A=@n13, B=AppendVector(A=LinearInterpolate(B=@n16, Alpha=@n17), B=Constant()))
  n19 = Constant2Vector() {r=-0.02}
  n20 = Frac(None=Multiply(A=Time(), B=Constant() {r=1}) {constB=0.5})
  n21 = LinearInterpolate(
   A=LinearInterpolate(Alpha=ScalarParameter() {defaultValue=1, parameterName=Icon Scale}) {constB=0.5}
   Alpha=LinearInterpolate(B=SmoothStep(Value=@n20) {constMin=0.6}, Alpha=ScalarParameter() {defaultValue=1, parameterName=Scale Pulse})
  ) {constA=0.5, constB=0.6}
  n22 = TextureSample(UVs=MaterialFunctionCall(UVs=Add(A=@n18, B=@n19), Texture Scale=@n21) {materialFunction=&ScaleUVsByCenter}.UVs) {samplerType=SAMPLERTYPE_LinearGrayscale, texture=&/T_UI_Icon_ToyFox_Exclamation}
  n23 = MaterialFunctionCall(DistanceField=@n22.R, Threshold=Constant() {r=0.3}) {materialFunction=&MF_UI_AntiAliasing}
  n24 = MaterialFunctionCall!(DistanceField=@n22.R, Threshold, Thickness, Alignment=Constant(), Smoothing) {materialFunction=/FortUILibrary/MaterialFunctions/Shapes/Modifiers/MF_UI_ShapeOutline}
  n25 = Constant() {r=0.55}
  NamedRerouteDeclaration(None=Multiply(
    A=NamedRerouteDeclaration(None=@n23.Outer) {name=Icon, nodeColor=<rgba 0 0 0 1>}
    B=MaterialFunctionCall(DistanceField=@n24.SDFOutline, Threshold=@n25) {materialFunction=&MF_UI_AntiAliasing}.Inner
   )) {name=Icon Outline, nodeColor=<rgba 0 0 0 1>}
  n26 = MaterialFunctionCall(UVs=@n14, "Rotation Angle (0-1)"=Constant() {r=0.12}) {materialFunction=&CustomRotator}
  NamedRerouteDeclaration(None=MaterialFunctionCall(
    DistanceField=MaterialFunctionCall(UVs=@n26.Rotated Values, Size=@n9) {materialFunction=&MF_UI_Box}.Fill
    Threshold=Constant() {r=0.1}
   ) {materialFunction=&MF_UI_AntiAliasing}.Outer) {name=Diamond Outline, nodeColor=<rgba 0 0 0 1>}
  NamedRerouteDeclaration(None=Multiply(A=@n15.Fill, B=ScalarParameter() {defaultValue=1, parameterName=Diamond Opacity})) {name=Diamond, nodeColor=<rgba 0 0 0 1>}
  n27 = MaterialFunctionCall() {materialFunction=&MF_UI_UVScaling}
  NamedRerouteDeclaration(None=@n27.Scaled to Fit) {name=Scale To Fit, nodeColor=<rgba 0 0 0 1>}
  n28 = NamedRerouteUsage()
  n29 = Time()
  n30 = ScalarParameter() {defaultValue=1, parameterName=Glitch In}
  n31 = TextureSample(UVs=Panner(Coordinate=@n28, Time=Multiply(A=@n29, B=@n30), Speed=Constant2Vector() {g=20})) {samplerType=SAMPLERTYPE_LinearGrayscale, texture=/Game/UI/HUD/Textures/UIPatterns/T_UI_GlitchNoise}
  n32 = Multiply(A=Subtract(A=ComponentMask(None=@n31.RGB) {g=true, r=true}) {constB=0.5}, B=Constant() {desc=Distortion Amount, r=0.02}) {constB=2}
  n33 = Time()
  n34 = Constant() {r=0.2}
  n35 = SmoothStep(Value=Frac(None=Multiply(A=@n33, B=@n34) {constB=0.5})) {constMin=0.9}
  NamedRerouteDeclaration(None=Add(A=ComponentMask(None=@n28) {g=true, r=true}, B=Multiply(A=@n32, B=@n35))) {name=Icon UVs, nodeColor=<rgba 0 0 0 1>}
  n36 = AppendVector(A=LinearInterpolate(Alpha=ScalarParameter() {parameterName=R Move}) {constB=0.3}, B=Constant())
  n37 = TextureSample(UVs=MaterialFunctionCall(UVs=Add(A=@n18, B=@n36), Texture Scale=@n21) {materialFunction=&ScaleUVsByCenter}.UVs) {samplerType=SAMPLERTYPE_LinearGrayscale, texture=&T_UI_Icon_ToyFox_Exclamation}
  n38 = Constant() {r=0.3}
  n39 = AppendVector(A=LinearInterpolate(Alpha=ScalarParameter() {parameterName=R Move}) {constA=-0.04, constB=-0.3}, B=Constant())
  n40 = TextureSample(UVs=MaterialFunctionCall(UVs=Add(A=@n18, B=@n39), Texture Scale=@n21) {materialFunction=&ScaleUVsByCenter}.UVs) {samplerType=SAMPLERTYPE_LinearGrayscale, texture=&T_UI_Icon_ToyFox_Exclamation}
  n41 = Constant() {r=0.3}
  NamedRerouteDeclaration(None=Add(
    A=MaterialFunctionCall(DistanceField=@n37.R, Threshold=@n38) {materialFunction=&MF_UI_AntiAliasing}.Outer
    B=MaterialFunctionCall(DistanceField=@n40.R, Threshold=@n41) {materialFunction=&MF_UI_AntiAliasing}.Outer
   )) {name=Icon R, nodeColor=<rgba 0 0 0 1>}
  NamedRerouteDeclaration(None=@n27.Scaled To Fill) {name=Scale To Fill, nodeColor=<rgba 0 0 0 1>}
  NamedRerouteUsage()
  Subtract()
  TextureCoordinate() {uTiling=2, vTiling=2}
  TextureCoordinate()
 ]
}
