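# Translucent UI material for the ToyFox boss bar, named for the "FinalHit" font, that builds an animated text color.
# MF_UI_CosineGradient takes a time-driven vector whose red channel advances over time, producing a cycling color that is then brightened by a constant offset.
# The "Font Color Switcher" scalar blends from a near-black red to that animated gradient.
# A separate "Color" vector parameter has no visible connection to the blend.
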Material /ToyFoxGameplay/UI/Widgets/HUD/Assets/BossBar/M_UI_ToyFox_FinalHit_Font
graph {
 properties {bAutomaticallySetUsageInEditor=false, blendMode=BLEND_Translucent, materialDomain=MD_UI, usedShadingModels=Default Lit}
 nodes [
  n1 = MaterialFunctionCall(R=Multiply(A=Time()) {constB=5}, G=Constant() {r=0.3}, B=Constant()) {materialFunction=&AppendMany}
  LinearInterpolate(
   A=Constant3Vector() {constant=<rgba 0.0339 0 0.001 1>}
   B=Add(A=MaterialFunctionCall!(Gradient, a, b, c, d=@n1.RGB) {materialFunction=/FortUILibrary/MaterialFunctions/Color/MF_UI_CosineGradient}.Result) {constB=0.6}
   Alpha=ScalarParameter() {parameterName=Font Color Switcher}
  )
  VectorParameter() {parameterName=Color, defaultValue=<rgba 0.7578 0.7578 0.7578 1>}
 ]
}
