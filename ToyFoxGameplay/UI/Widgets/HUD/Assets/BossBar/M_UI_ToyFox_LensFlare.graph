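# Additive UI material for a lens flare in the ToyFox boss bar HUD.
# The red channel of a grayscale flare texture goes through a Power node, then is added to a named reroute value.
# A named reroute declaration holds a cosine gradient whose phase input packs scaled Time with two constants, so the first color channel cycles over time while the others stay fixed.
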Material /ToyFoxGameplay/UI/Widgets/HUD/Assets/BossBar/M_UI_ToyFox_LensFlare
graph {
 properties {bAutomaticallySetUsageInEditor=false, blendMode=BLEND_Additive, materialDomain=MD_UI, usedShadingModels=Default Lit}
 nodes [
  Add(
   A=Power(Base=TextureSample() {samplerType=SAMPLERTYPE_LinearGrayscale, texture=&/T_UI_ToyFox_LensFlare}.R)
   B=NamedRerouteUsage()
  )
  n1 = Multiply(A=Time()) {constB=5}
  n2 = Constant() {r=0.3}
  n3 = Constant()
  NamedRerouteDeclaration(None=MaterialFunctionCall!(
    Gradient
    a
    b
    c
    d=MaterialFunctionCall(R=@n1, G=@n2, B=@n3) {materialFunction=&AppendMany}.RGB
   ) {materialFunction=/FortUILibrary/MaterialFunctions/Color/MF_UI_CosineGradient}.Result) {nodeColor=<rgba 1 0.0938 0 1>}
 ]
}
