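# UI material that draws animated, color-cycling lightning with a soft glow for the boss bar HUD.
# Panning normal noise distorts lightning-texture UVs, and a quantized time-driven rotation makes the bolts jump.
# Scrolling squared radial masks and dissolve noise, each with a manual-override parameter, shape the texture channels.
# MF_UI_AntiAliasing thresholds turn these into bolt and glow layers, alongside a cosine-gradient palette and master opacity.
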Material /ToyFoxGameplay/UI/Widgets/HUD/Assets/BossBar/M_UI_ToyFox_Lightning
graph {
 properties {bAutomaticallySetUsageInEditor=false, blendMode=BLEND_AlphaComposite, materialDomain=MD_UI, usedShadingModels=Default Lit}
 nodes [
  Constant3Vector() {constant=<rgba 1 1 1 1>}
  Frac(None=Time())
  LinearInterpolate(
   A=NamedRerouteUsage()
   B=NamedRerouteUsage()
   Alpha=MaterialFunctionCall(Speed=Constant() {r=0.5}) {materialFunction=&DebugTimeSine}.Result
  )
  MaterialFunctionCall() {materialFunction=&DebugTimeSine}
  n1 = NamedRerouteUsage()
  n2 = Multiply(A=NamedRerouteUsage()) {constB=0.1}
  n3 = ComponentMask(None=TextureCoordinate()) {g=true, r=true}
  n4 = Panner(Coordinate=TextureCoordinate() {uTiling=0.3, vTiling=0.3}, Time=Time(), Speed=Constant2Vector())
  n5 = ScalarParameter() {parameterName=UV Distortion}
  n6 = Multiply(
   A=ComponentMask(None=TextureSample(UVs=@n4) {samplerType=SAMPLERTYPE_Normal, texture=/ToyFoxGameplay/UI/Materials/Textures/T_UI_ToyFox_NormalNoise}.R) {g=true, r=true}
   B=LinearInterpolate(Alpha=@n5)
  )
  n7 = Frac(None=Time())
  n8 = Constant()
  n9 = TextureSample(UVs=MaterialFunctionCall(
    UVs=Add(A=@n3, B=@n6)
    "Rotation Angle (0-1)"=MaterialFunctionCall!("Param (0-1)"=@n7, Tiles=@n8, Apply Offset) {materialFunction=/Game/Environments/Papaya/Materials/MaterialFunctions/MF_Papaya_Quantize_Scalar}.Result
   ) {materialFunction=&CustomRotator}.Rotated Values) {samplerType=SAMPLERTYPE_LinearColor, texture=/ToyFoxGameplay/UI/Materials/Textures/T_UI_ToyFox_Zord_Lightning}
  n10 = Multiply(A=NamedRerouteUsage(), B=ScalarParameter() {defaultValue=1, parameterName=Radial Mask Opacity})
  n11 = Multiply(A=Saturate(None=Subtract(A=@n9.R, B=@n10)), B=NamedRerouteUsage())
  n12 = MaterialFunctionCall(DistanceField=@n11, Threshold=Constant() {r=0.7}) {materialFunction=&MF_UI_AntiAliasing}
  n13 = ScalarParameter() {defaultValue=1, parameterName=Master Opacity}
  Multiply(A=Add(A=Multiply(A=@n1, B=@n2), B=Multiply(A=@n1, B=@n12.Outer)), B=@n13)
  n14 = Multiply(A=@n12.Outer) {constB=0.2, desc=Blend}
  Multiply(A=@n14, B=@n13)
  Multiply()
  Multiply(A=@n7, B=@n5) {constB=0.1}
  n15 = TextureCoordinate()
  n16 = ComponentMask(None=MaterialFunctionCall(Vector or UVs=Subtract(A=@n15) {constB=0.5}) {materialFunction=&VectorToRadialValue}.Radial Coordinates) {r=true}
  n17 = LinearInterpolate(
   A=Multiply(A=Time())
   B=LinearInterpolate(Alpha=ScalarParameter() {defaultValue=1, parameterName=Radial Mask Manual})
   Alpha=ScalarParameter() {defaultValue=1, desc=Default 1, parameterName=Manual Radial Mask Switch}
  )
  n18 = Subtract(A=Frac(None=Add(A=@n16, B=@n17))) {constB=0.5}
  NamedRerouteDeclaration(None=Multiply(A=Multiply(A=@n18, B=@n18)) {constB=2}) {name="Whee!", nodeColor=<rgba 0 0 0 1>}
  n19 = MaterialFunctionCall(R=Multiply(A=Time()) {constB=5}, G=Constant() {r=0.7}, B=Constant() {r=0.1}) {materialFunction=&AppendMany}
  NamedRerouteDeclaration(None=Multiply(A=MaterialFunctionCall!(Gradient, a, b, c, d=@n19.RGB) {materialFunction=/FortUILibrary/MaterialFunctions/Color/MF_UI_CosineGradient}.Result)) {name=Party Colors, nodeColor=<rgba 0 0 0 1>}
  n20 = TextureCoordinate() {uTiling=2, vTiling=2}
  n21 = MaterialFunctionCall(Speed=Constant() {r=0.5}) {materialFunction=&DebugTimeSine}
  n22 = Multiply(A=Subtract(A=@n21.Result) {constB=0.5}) {constB=4}
  n23 = ScalarParameter() {parameterName=Dissolve}
  n24 = Subtract(
   A=Saturate(None=TextureSample(UVs=@n20) {samplerType=SAMPLERTYPE_Grayscale, texture=&LowResBlurredNoise}.G)
   B=LinearInterpolate(
    A=Saturate(None=@n22)
    B=LinearInterpolate(Alpha=@n23)
    Alpha=ScalarParameter() {defaultValue=1, desc=Default 1, parameterName=Manual Dissolve Switch}
   )
  ) {constB=0}
  NamedRerouteDeclaration(None=Saturate(None=Add(A=@n24))) {name=Dissolve, nodeColor=<rgba 0 0 0 1>}
  n25 = NamedRerouteUsage()
  n26 = Multiply(A=Saturate(None=Subtract(A=@n9.B, B=@n25)), B=NamedRerouteUsage())
  n27 = Constant() {r=0.7}
  NamedRerouteDeclaration(None=Multiply(A=MaterialFunctionCall(DistanceField=@n26, Threshold=@n27) {materialFunction=&MF_UI_AntiAliasing}.Outer) {constB=0.2, desc=Blend}) {name=Lightning 2, nodeColor=<rgba 0 0.7422 1 1>}
  n28 = OneMinus(None=@n16)
  n29 = Time()
  n30 = Subtract(A=Frac(None=Add(A=@n28, B=@n29))) {constB=0.5}
  NamedRerouteDeclaration(None=Multiply(A=Multiply(A=@n30, B=@n30)) {constB=2}) {name="Reverse Whee!", nodeColor=<rgba 0 0 0 1>}
  n31 = Constant() {r=0.5}
  n32 = Constant() {r=0.3}
  NamedRerouteDeclaration(None=Multiply(A=MaterialFunctionCall(DistanceField=@n11, Threshold=@n31, Smoothing=@n32) {materialFunction=&MF_UI_AntiAliasing}.Outer) {constB=0.2, desc=Blend}) {name=GLow, nodeColor=<rgba 0.0547 0 1 1>}
  NamedRerouteDeclaration(None=@n14) {nodeColor=<rgba 1 0.9375 0 1>}
  Saturate(None=@n21.Result)
 ]
}
