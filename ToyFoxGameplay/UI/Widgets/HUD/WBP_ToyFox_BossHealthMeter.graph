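# ToyFox HUD boss health bar for "Rita's Monster", with an animated "FINAL POWER SWORD HIT!" callout while the enemy is stunned.
# On Construct it binds to B_ToyFoxGameStateComponent events and syncs to current enemy state.
# Health drives the bar material's Progress parameter, spawn-in plays the intro, and gated stun changes toggle callout intro/outro.
# The giant pawn's Finisher.Active tag reverses the intro before collapsing the widget; enemy destruction collapses it immediately.
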WidgetBlueprint /ToyFoxGameplay/UI/Widgets/HUD/WBP_ToyFox_BossHealthMeter
animations [
 Anim_ProgressIntro {
  binding_count=5
  length_seconds=0.25
  object=&self:Anim_ProgressIntro
  sequence {
   bindings [
    Text_Title {
     id=#1
     possessed_class=CommonTextBlock
     tracks [{
       type=MovieSceneFloatTrack
       display_name=Render Opacity
       sections [{type=MovieSceneFloatSection, end_frame=15, start_frame=0, channels [None {type=MovieSceneScriptingFloatChannel, keys [[0,0],[15,1]]}]}]
      }]
    }
    Image_ProgressBar {
     id=#2
     possessed_class=Image
     tracks [
      {
       type=MovieSceneFloatTrack
       display_name=Render Opacity
       sections [{type=MovieSceneFloatSection, end_frame=15, start_frame=0, channels [None {type=MovieSceneScriptingFloatChannel, keys [[0,0],[15,1]]}]}]
      }
      {
       type=MovieSceneWidgetMaterialTrack
       sections [{type=MovieSceneParameterSection, channels [Master Opacity {type=MovieSceneScriptingFloatChannel, keys [[0,0],[15,1]]}]}]
      }
     ]
    }
    Image_Icon {
     id=#3
     possessed_class=Image
     tracks [{
       type=MovieSceneByteTrack
       display_name=Visibility
       sections [{type=MovieSceneByteSection, channels [None {type=MovieSceneScriptingByteChannel, const [1,1]}]}]
      }]
    }
    Image_Glow_1 {
     id=#4
     possessed_class=Image
     tracks [{
       type=MovieSceneFloatTrack
       display_name=Render Opacity
       sections [{type=MovieSceneFloatSection, end_frame=15, start_frame=0, channels [None {type=MovieSceneScriptingFloatChannel, const [0,2]}]}]
      }]
    }
    Overlay_FinalHit {
     id=#5
     possessed_class=Overlay
     tracks [{
       type=MovieSceneByteTrack
       display_name=Visibility
       sections [{type=MovieSceneByteSection, channels [None {type=MovieSceneScriptingByteChannel, const [1,1]}]}]
      }]
    }
   ]
   display_rate {numerator=60}
  }
 }
 Anim_FinalHitIntro {
  binding_count=10
  length_seconds=2.0167
  object=&self:Anim_FinalHitIntro
  sequence {
   bindings [
    Image_TextBg {
     id=#6
     possessed_class=Image
     tracks [
      {
       type=MovieSceneFloatTrack
       display_name=Render Opacity
       sections [{type=MovieSceneFloatSection, channels [None {type=MovieSceneScriptingFloatChannel, keys [[9,0],[24,1]]}]}]
      }
      {
       type=MovieScene2DTransformTrack
       sections [{
         type=MovieScene2DTransformSection
         channels [
          Translation.X {type=MovieSceneScriptingFloatChannel, const [0,1]}
          Translation.Y {type=MovieSceneScriptingFloatChannel, const [0,1]}
          Angle {type=MovieSceneScriptingFloatChannel, const [0,1]}
          Scale.X {type=MovieSceneScriptingFloatChannel, const [1,1]}
          Scale.Y {type=MovieSceneScriptingFloatChannel, const [1,1]}
          Shear.X {type=MovieSceneScriptingFloatChannel, const [0,1]}
          Shear.Y {type=MovieSceneScriptingFloatChannel, const [0,1]}
         ]
        }]
      }
      {
       type=MovieSceneWidgetMaterialTrack
       sections [{
         type=MovieSceneParameterSection
         channels [
          Text Bg Color Swap {type=MovieSceneScriptingFloatChannel, keys [[15,0],[33,1]]}
          Edge Opacity {type=MovieSceneScriptingFloatChannel, keys [[0,0],[24,0],[67,1]]}
          Bracket Pulse {type=MovieSceneScriptingFloatChannel, const [0.3,1]}
          Master Opacity {type=MovieSceneScriptingFloatChannel, const [1,1]}
         ]
        }]
      }
     ]
    }
    UEFN_TextBlock_C_96 {
     id=#7
     possessed_class=UEFN_TextBlock_C
     tracks [
      {
       type=MovieSceneFloatTrack
       display_name=Render Opacity
       sections [{type=MovieSceneFloatSection, channels [None {type=MovieSceneScriptingFloatChannel, keys [[9,0],[25,1]]}]}]
      }
      {
       type=MovieScene2DTransformTrack
       sections [{
         type=MovieScene2DTransformSection
         channels [
          Translation.X {type=MovieSceneScriptingFloatChannel, const [0,1]}
          Translation.Y {type=MovieSceneScriptingFloatChannel, const [0,1]}
          Angle {type=MovieSceneScriptingFloatChannel, const [0,1]}
          Scale.X {type=MovieSceneScriptingFloatChannel, keys [[5.455,5],[12,1]]}
          Scale.Y {type=MovieSceneScriptingFloatChannel, keys [[5.455,3],[12,1]]}
          Shear.X {type=MovieSceneScriptingFloatChannel, const [0,1]}
          Shear.Y {type=MovieSceneScriptingFloatChannel, const [0,1]}
         ]
        }]
      }
      {
       type=MovieSceneWidgetMaterialTrack
       display_name=Font.FontMaterial
       sections [{type=MovieSceneParameterSection, channels [Font Color Switcher {type=MovieSceneScriptingFloatChannel, keys [[26,0],[35,1]]}]}]
      }
     ]
    }
    Image_Glow {
     id=#8
     possessed_class=Image
     tracks [
      {
       type=MovieSceneWidgetMaterialTrack
       sections [{type=MovieSceneParameterSection, channels [PulseSpeed {type=MovieSceneScriptingFloatChannel, const [0,1]}]}]
      }
      {
       type=MovieSceneFloatTrack
       display_name=Render Opacity
       sections [{type=MovieSceneFloatSection, channels [None {type=MovieSceneScriptingFloatChannel, keys [[0,0],[12.706,1],[48,0.1]]}]}]
      }
      {
       type=MovieSceneByteTrack
       display_name=Visibility
       sections [{type=MovieSceneByteSection, end_frame=0, start_frame=0, channels [None {type=MovieSceneScriptingByteChannel, const [4,1]}]}]
      }
     ]
    }
    Image_Icon {
     id=#9
     possessed_class=Image
     tracks [
      {
       type=MovieSceneWidgetMaterialTrack
       sections [{
         type=MovieSceneParameterSection
         channels [
          Diamond Scale {type=MovieSceneScriptingFloatChannel, keys [[17,0],[27,0.4]]}
          Icon Scale {type=MovieSceneScriptingFloatChannel, keys [[5,1.5],[13,0.7],[19,1]]}
          Icon Opacity {type=MovieSceneScriptingFloatChannel, keys [[0,0],[2,0.5],[13,0.5],[50,0.8]]}
          Scale Pulse {type=MovieSceneScriptingFloatChannel, keys [[50,0],[70,1]]}
          R Move {type=MovieSceneScriptingFloatChannel, keys [[0,1],[9,0]]}
          Shakey Shake {type=MovieSceneScriptingFloatChannel, keys [[7,0],[9,1],[27,0]]}
          Icon RL Opacity {type=MovieSceneScriptingFloatChannel, keys [[0,0],[7,1],[50,0.01]]}
         ]
        }]
      }
      {
       type=MovieSceneByteTrack
       display_name=Visibility
       sections [{type=MovieSceneByteSection, channels [None {type=MovieSceneScriptingByteChannel, const [4,1]}]}]
      }
      {
       type=MovieScene2DTransformTrack
       sections [{
         type=MovieScene2DTransformSection
         channels [
          Translation.X {type=MovieSceneScriptingFloatChannel, const [0,1]}
          Translation.Y {type=MovieSceneScriptingFloatChannel, const [-15,1]}
          Angle {type=MovieSceneScriptingFloatChannel, const [0,1]}
          Scale.X {type=MovieSceneScriptingFloatChannel, const [2,1]}
          Scale.Y {type=MovieSceneScriptingFloatChannel, const [2,1]}
          Shear.X {type=MovieSceneScriptingFloatChannel, const [0,1]}
          Shear.Y {type=MovieSceneScriptingFloatChannel, const [0,1]}
         ]
        }]
      }
      {
       type=MovieSceneFloatTrack
       display_name=Render Opacity
       sections [{type=MovieSceneFloatSection, channels [None {type=MovieSceneScriptingFloatChannel, const [1,2]}]}]
      }
     ]
    }
    Image_LensFlare {
     id=#10
     possessed_class=Image
     tracks [
      {
       type=MovieSceneFloatTrack
       display_name=Render Opacity
       sections [{type=MovieSceneFloatSection, channels [None {type=MovieSceneScriptingFloatChannel, keys [[12,0],[21,1],[44,0]]}]}]
      }
      {
       type=MovieScene2DTransformTrack
       sections [{
         type=MovieScene2DTransformSection
         channels [
          Translation.X {type=MovieSceneScriptingFloatChannel, keys [[0,-10],[27,115]]}
          Translation.Y {type=MovieSceneScriptingFloatChannel, const [0,1]}
          Angle {type=MovieSceneScriptingFloatChannel, const [0,1]}
          Scale.X {type=MovieSceneScriptingFloatChannel, const [1,1]}
          Scale.Y {type=MovieSceneScriptingFloatChannel, const [5,1]}
          Shear.X {type=MovieSceneScriptingFloatChannel, const [0,1]}
          Shear.Y {type=MovieSceneScriptingFloatChannel, const [0,1]}
         ]
        }]
      }
     ]
    }
    Image_Lightning {
     id=#11
     possessed_class=Image
     tracks [
      {
       type=MovieSceneFloatTrack
       display_name=Render Opacity
       sections [{type=MovieSceneFloatSection, channels [None {type=MovieSceneScriptingFloatChannel, const [1,1]}]}]
      }
      {
       type=MovieSceneWidgetMaterialTrack
       sections [{
         type=MovieSceneParameterSection
         end_frame=121
         start_frame=0
         channels [
          Radial Mask Manual {type=MovieSceneScriptingFloatChannel, keys [[0,0],[48,1]]}
          UV Distortion {type=MovieSceneScriptingFloatChannel, keys [[0,0],[26,0],[71,0.1]]}
          Radial Mask Opacity {type=MovieSceneScriptingFloatChannel, keys [[0,0],[20,0],[29,1]]}
          Master Opacity {type=MovieSceneScriptingFloatChannel, keys [[0,0],[15,0],[20,1],[53,1],[63,0]]}
          Dissolve {type=MovieSceneScriptingFloatChannel, keys [[39,0],[77,1]]}
         ]
        }]
      }
     ]
    }
    Image_Edges {
     id=#12
     possessed_class=Image
     tracks [
      {
       type=MovieSceneFloatTrack
       display_name=Render Opacity
       sections [{type=MovieSceneFloatSection, channels [None {type=MovieSceneScriptingFloatChannel, keys [[9,0],[18,1]]}]}]
      }
      {
       type=MovieScene2DTransformTrack
       sections [{
         type=MovieScene2DTransformSection
         channels [
          Translation.X {type=MovieSceneScriptingFloatChannel, const [0,1]}
          Translation.Y {type=MovieSceneScriptingFloatChannel, const [0,1]}
          Angle {type=MovieSceneScriptingFloatChannel, const [0,1]}
          Scale.X {type=MovieSceneScriptingFloatChannel, keys [[4,2],[18,1]]}
          Scale.Y {type=MovieSceneScriptingFloatChannel, const [1,2]}
          Shear.X {type=MovieSceneScriptingFloatChannel, const [0,1]}
          Shear.Y {type=MovieSceneScriptingFloatChannel, const [0,1]}
         ]
        }]
      }
      {
       type=MovieSceneWidgetMaterialTrack
       sections [{
         type=MovieSceneParameterSection
         channels [
          Bracket Movement {type=MovieSceneScriptingFloatChannel, const [0,2]}
          Edge Flash In {type=MovieSceneScriptingFloatChannel, keys [[0,0],[18,0],[27,1]]}
          Bracket Pulse {type=MovieSceneScriptingFloatChannel, keys [[0,1],[40,0.5],[71,1]]}
          Edge Opacity {type=MovieSceneScriptingFloatChannel, keys [[9,0],[32,0.4]]}
         ]
        }]
      }
     ]
    }
    Image_ProgressBar {
     id=#13
     possessed_class=Image
     tracks [
      {
       type=MovieSceneFloatTrack
       display_name=Render Opacity
       sections [{type=MovieSceneFloatSection, end_frame=12, start_frame=12, channels [None {type=MovieSceneScriptingFloatChannel, const [1,1]}]}]
      }
      {
       type=MovieSceneWidgetMaterialTrack
       sections [{
         type=MovieSceneParameterSection
         channels [
          Frame Flash {type=MovieSceneScriptingFloatChannel, keys [[0,0],[2,1],[38,0]]}
          Janky Shine {type=MovieSceneScriptingFloatChannel, keys [[7.714,0],[60,1]]}
          Flash On {type=MovieSceneScriptingFloatChannel, const [0,2]}
          Auto Shine {type=MovieSceneScriptingFloatChannel, keys [[67,0],[70,1]]}
          Shine Opacity {type=MovieSceneScriptingFloatChannel, keys [[64,1],[67,0],[84,0],[95,1]]}
          Final Hit Flash {type=MovieSceneScriptingFloatChannel, keys [[0,0],[9,0],[14,0.95]]}
         ]
        }]
      }
     ]
    }
    Overlay_FinalHit {
     id=#14
     possessed_class=Overlay
     tracks [
      {
       type=MovieSceneByteTrack
       display_name=Visibility
       sections [{type=MovieSceneByteSection, channels [None {type=MovieSceneScriptingByteChannel, const [4,1]}]}]
      }
      {
       type=MovieSceneFloatTrack
       display_name=Render Opacity
       sections [{type=MovieSceneFloatSection, channels [None {type=MovieSceneScriptingFloatChannel, const [1,1]}]}]
      }
     ]
    }
    Text_Title {
     id=#15
     possessed_class=CommonTextBlock
     tracks [{
       type=MovieSceneFloatTrack
       display_name=Render Opacity
       sections [{type=MovieSceneFloatSection, channels [None {type=MovieSceneScriptingFloatChannel, const [0,1]}]}]
      }]
    }
   ]
   display_rate {numerator=60}
   tracks [{
     type=MovieSceneAudioTrack
     display_name=SFX
     sections [{
       type=MovieSceneAudioSection
       end_frame=121
       start_frame=0
       channels [
        Volume {type=MovieSceneScriptingFloatChannel, default=1}
        Pitch {type=MovieSceneScriptingFloatChannel, default=1}
        AttachActor {type=MovieSceneScriptingActorReferenceChannel, default="<Struct 'MovieSceneObjectBindingID' (0x4697abf7140) {}>"}
       ]
      }]
    }]
  }
 }
 Anim_Outro {
  binding_count=10
  length_seconds=2.0167
  object=&self:Anim_Outro
  sequence {
   bindings [
    Image_TextBg {
     id=#6
     possessed_class=Image
     tracks [
      {
       type=MovieSceneFloatTrack
       display_name=Render Opacity
       sections [{type=MovieSceneFloatSection, channels [None {type=MovieSceneScriptingFloatChannel, const [1,1]}]}]
      }
      {
       type=MovieScene2DTransformTrack
       sections [{
         type=MovieScene2DTransformSection
         channels [
          Translation.X {type=MovieSceneScriptingFloatChannel, const [0,1]}
          Translation.Y {type=MovieSceneScriptingFloatChannel, const [0,1]}
          Angle {type=MovieSceneScriptingFloatChannel, const [0,1]}
          Scale.X {type=MovieSceneScriptingFloatChannel, const [1,1]}
          Scale.Y {type=MovieSceneScriptingFloatChannel, const [1,1]}
          Shear.X {type=MovieSceneScriptingFloatChannel, const [0,1]}
          Shear.Y {type=MovieSceneScriptingFloatChannel, const [0,1]}
         ]
        }]
      }
      {
       type=MovieSceneWidgetMaterialTrack
       sections [{
         type=MovieSceneParameterSection
         channels [
          Text Bg Color Swap {type=MovieSceneScriptingFloatChannel, const [1,1]}
          Edge Opacity {type=MovieSceneScriptingFloatChannel, const [0.4,1]}
          Bracket Pulse {type=MovieSceneScriptingFloatChannel, const [0.3,1]}
          Master Opacity {type=MovieSceneScriptingFloatChannel, keys [[0,1],[11,0]]}
         ]
        }]
      }
     ]
    }
    UEFN_TextBlock_C_96 {
     id=#7
     possessed_class=UEFN_TextBlock_C
     tracks [
      {
       type=MovieSceneFloatTrack
       display_name=Render Opacity
       sections [{type=MovieSceneFloatSection, channels [None {type=MovieSceneScriptingFloatChannel, keys [[0,1],[11,0]]}]}]
      }
      {
       type=MovieScene2DTransformTrack
       sections [{
         type=MovieScene2DTransformSection
         channels [
          Translation.X {type=MovieSceneScriptingFloatChannel, const [0,1]}
          Translation.Y {type=MovieSceneScriptingFloatChannel, const [0,1]}
          Angle {type=MovieSceneScriptingFloatChannel, const [0,1]}
          Scale.X {type=MovieSceneScriptingFloatChannel, const [1,1]}
          Scale.Y {type=MovieSceneScriptingFloatChannel, const [1,1]}
          Shear.X {type=MovieSceneScriptingFloatChannel, const [0,1]}
          Shear.Y {type=MovieSceneScriptingFloatChannel, const [0,1]}
         ]
        }]
      }
      {
       type=MovieSceneWidgetMaterialTrack
       display_name=Font.FontMaterial
       sections [{type=MovieSceneParameterSection, channels [Font Color Switcher {type=MovieSceneScriptingFloatChannel, const [1,1]}]}]
      }
     ]
    }
    Image_Glow {
     id=#8
     possessed_class=Image
     tracks [
      {
       type=MovieSceneWidgetMaterialTrack
       sections [{type=MovieSceneParameterSection, channels [PulseSpeed {type=MovieSceneScriptingFloatChannel, const [0,1]}]}]
      }
      {
       type=MovieSceneFloatTrack
       display_name=Render Opacity
       sections [{type=MovieSceneFloatSection, channels [None {type=MovieSceneScriptingFloatChannel, keys [[0,0.1],[18,0]]}]}]
      }
      {
       type=MovieSceneByteTrack
       display_name=Visibility
       sections [{type=MovieSceneByteSection, end_frame=18, start_frame=0, channels [None {type=MovieSceneScriptingByteChannel, keys [[0,4],[18,1]]}]}]
      }
     ]
    }
    Image_Icon {
     id=#9
     possessed_class=Image
     tracks [
      {
       type=MovieSceneWidgetMaterialTrack
       sections [{
         type=MovieSceneParameterSection
         channels [
          Icon Scale {type=MovieSceneScriptingFloatChannel, const [1,1]}
          Icon Opacity {type=MovieSceneScriptingFloatChannel, keys [[0,0.8],[11,0]]}
          Scale Pulse {type=MovieSceneScriptingFloatChannel, keys [[0,0.4],[11,0]]}
          R Move {type=MovieSceneScriptingFloatChannel, const [0,1]}
          Shakey Shake {type=MovieSceneScriptingFloatChannel, const [0,1]}
          Icon RL Opacity {type=MovieSceneScriptingFloatChannel, keys [[0,0.05],[11,0]]}
         ]
        }]
      }
      {
       type=MovieSceneByteTrack
       display_name=Visibility
       sections [{type=MovieSceneByteSection, channels [None {type=MovieSceneScriptingByteChannel, const [4,1]}]}]
      }
      {
       type=MovieScene2DTransformTrack
       sections [{
         type=MovieScene2DTransformSection
         channels [
          Translation.X {type=MovieSceneScriptingFloatChannel, const [0,1]}
          Translation.Y {type=MovieSceneScriptingFloatChannel, const [-15,1]}
          Angle {type=MovieSceneScriptingFloatChannel, const [0,1]}
          Scale.X {type=MovieSceneScriptingFloatChannel, const [2,1]}
          Scale.Y {type=MovieSceneScriptingFloatChannel, const [2,1]}
          Shear.X {type=MovieSceneScriptingFloatChannel, const [0,1]}
          Shear.Y {type=MovieSceneScriptingFloatChannel, const [0,1]}
         ]
        }]
      }
      {
       type=MovieSceneFloatTrack
       display_name=Render Opacity
       sections [{type=MovieSceneFloatSection, channels [None {type=MovieSceneScriptingFloatChannel, const [1,1]}]}]
      }
     ]
    }
    Image_LensFlare {
     id=#10
     possessed_class=Image
     tracks [
      {
       type=MovieSceneFloatTrack
       display_name=Render Opacity
       sections [{type=MovieSceneFloatSection, channels [None {type=MovieSceneScriptingFloatChannel, const [0,1]}]}]
      }
      {
       type=MovieScene2DTransformTrack
       sections [{
         type=MovieScene2DTransformSection
         channels [
          Translation.X {type=MovieSceneScriptingFloatChannel, const [115,1]}
          Translation.Y {type=MovieSceneScriptingFloatChannel, const [0,1]}
          Angle {type=MovieSceneScriptingFloatChannel, const [0,1]}
          Scale.X {type=MovieSceneScriptingFloatChannel, const [1,1]}
          Scale.Y {type=MovieSceneScriptingFloatChannel, const [5,1]}
          Shear.X {type=MovieSceneScriptingFloatChannel, const [0,1]}
          Shear.Y {type=MovieSceneScriptingFloatChannel, const [0,1]}
         ]
        }]
      }
     ]
    }
    Image_Lightning {
     id=#11
     possessed_class=Image
     tracks [
      {
       type=MovieSceneFloatTrack
       display_name=Render Opacity
       sections [{type=MovieSceneFloatSection, channels [None {type=MovieSceneScriptingFloatChannel, keys [[0,1],[18,0]]}]}]
      }
      {
       type=MovieSceneWidgetMaterialTrack
       sections [{type=MovieSceneParameterSection, end_frame=121, start_frame=0, channels [Master Opacity {type=MovieSceneScriptingFloatChannel, const [0,1]}]}]
      }
     ]
    }
    Image_Edges {
     id=#12
     possessed_class=Image
     tracks [
      {
       type=MovieSceneFloatTrack
       display_name=Render Opacity
       sections [{type=MovieSceneFloatSection, channels [None {type=MovieSceneScriptingFloatChannel, const [1,1]}]}]
      }
      {
       type=MovieScene2DTransformTrack
       sections [{
         type=MovieScene2DTransformSection
         channels [
          Translation.X {type=MovieSceneScriptingFloatChannel, const [0,1]}
          Translation.Y {type=MovieSceneScriptingFloatChannel, const [0,1]}
          Angle {type=MovieSceneScriptingFloatChannel, const [0,1]}
          Scale.X {type=MovieSceneScriptingFloatChannel, const [1,1]}
          Scale.Y {type=MovieSceneScriptingFloatChannel, const [1,1]}
          Shear.X {type=MovieSceneScriptingFloatChannel, const [0,1]}
          Shear.Y {type=MovieSceneScriptingFloatChannel, const [0,1]}
         ]
        }]
      }
      {
       type=MovieSceneWidgetMaterialTrack
       sections [{
         type=MovieSceneParameterSection
         channels [
          Bracket Movement {type=MovieSceneScriptingFloatChannel, const [0,1]}
          Edge Flash In {type=MovieSceneScriptingFloatChannel, const [1,1]}
          Bracket Pulse {type=MovieSceneScriptingFloatChannel, const [1,1]}
          Edge Opacity {type=MovieSceneScriptingFloatChannel, keys [[0,0.4],[11,0]]}
         ]
        }]
      }
     ]
    }
    Image_ProgressBar {
     id=#13
     possessed_class=Image
     tracks [
      {
       type=MovieSceneFloatTrack
       display_name=Render Opacity
       sections [{type=MovieSceneFloatSection, end_frame=12, start_frame=12, channels [None {type=MovieSceneScriptingFloatChannel, const [1,1]}]}]
      }
      {
       type=MovieSceneWidgetMaterialTrack
       sections [{
         type=MovieSceneParameterSection
         channels [
          Frame Flash {type=MovieSceneScriptingFloatChannel, const [0,2]}
          Janky Shine {type=MovieSceneScriptingFloatChannel, keys [[0,1],[13,0]]}
          Flash On {type=MovieSceneScriptingFloatChannel, const [0,1]}
          Auto Shine {type=MovieSceneScriptingFloatChannel, keys [[0,1],[13,0]]}
          Shine Opacity {type=MovieSceneScriptingFloatChannel, keys [[0,1],[13,0]]}
          Final Hit Flash {type=MovieSceneScriptingFloatChannel, keys [[0,1],[13,0]]}
         ]
        }]
      }
     ]
    }
    Overlay_FinalHit {
     id=#14
     possessed_class=Overlay
     tracks [
      {
       type=MovieSceneByteTrack
       display_name=Visibility
       sections [{type=MovieSceneByteSection, channels [None {type=MovieSceneScriptingByteChannel, keys [[0,4],[18,1]]}]}]
      }
      {
       type=MovieSceneFloatTrack
       display_name=Render Opacity
       sections [{type=MovieSceneFloatSection, channels [None {type=MovieSceneScriptingFloatChannel, keys [[0,1],[11,1],[18,0]]}]}]
      }
     ]
    }
    Text_Title {
     id=#15
     possessed_class=CommonTextBlock
     tracks [{
       type=MovieSceneFloatTrack
       display_name=Render Opacity
       sections [{type=MovieSceneFloatSection, channels [None {type=MovieSceneScriptingFloatChannel, keys [[0,0],[15,0],[34,1]]}]}]
      }]
    }
   ]
   display_rate {numerator=60}
  }
 }
]
graphs [EventGraph {nodes [
   n1 = HandleEnemyHealthPercentChanged!() -> (OutputDelegate, then, HealthPercent:double=~) {is=title K2Node_CustomEvent}
   > n2 = GetDynamicMaterial(self=Get Image_ProgressBar() -> (Image_ProgressBar:Image Object Reference))
   > SetScalarParameterValue(self=@n2, ParameterName=Progress, Value="Clamp (Float)"(Value=@n1.HealthPercent, Min=0, Max=1))
   HandleEnemyDestroyed() {is=title K2Node_CustomEvent}
   > SetVisibility(InVisibility=Collapsed)
   n3 = HandleEnemyStunnedActiveChanged!() -> (OutputDelegate, then, bStunActive:bool=~) {is=title K2Node_CustomEvent}
   > n4 = Delta Gate Wildcard!(execute:Exec, Variable:"Boolean (by ref)"=Get IsStunActive() -> (IsStunActive:bool), New Value:bool=@n3.bStunActive) -> (Success:Exec, Updated Variable:"Boolean (by ref)", Previous Variable:bool=~) {is=title K2Node_MacroInstance}
   > n5 = Branch(Condition=@n4)
   > PlayAnimationForward(InAnimation=Get Anim_FinalHitIntro() -> (Anim_FinalHitIntro:Widget Animation Object Reference))
   PlayAnimationForward(execute=@n5.else, InAnimation=Get Anim_Outro() -> (Anim_Outro:Widget Animation Object Reference))
   HandleFinisherExecuted() {is=title K2Node_CustomEvent}
   > n6 = Play Animation with Finished event!(
    execute:Exec
    Widget:User Widget Object Reference=Self-Reference()
    InAnimation:Widget Animation Object Reference=Get Anim_ProgressIntro() -> (Anim_ProgressIntro:Widget Animation Object Reference)
    StartAtTime:float=0
    NumLoopsToPlay:int=1
    PlayMode:EUMGSequencePlayMode Enum=Reverse
    PlaybackSpeed:float=1
   ) -> (then:Exec, Result:Widget Animation Handle Structure, Finished:Exec) {is=title K2Node_PlayAnimation2}
   SetVisibility(execute=@n6.Finished, InVisibility=Collapsed)
   HandleEnemySpawnInComplete() {is=title K2Node_CustomEvent}
   > SetVisibility(InVisibility=SelfHitTestInvisible)
   > PlayAnimation(InAnimation=Get Anim_ProgressIntro(), bRestoreState=true)
   Event Construct()
   > Parent: Construct()
   > SetVisibility(InVisibility=Collapsed)
   n7 = GetGameStateAthena!() -> (ReturnValue:Fort Game State Athena Object Reference)
   > Is Valid(InputObject=@n7)
   > n8 = GetComponentByClass!(
    execute
    self:Actor Object Reference=@n7
    ComponentClass:Actor Component Class Reference=/ToyFoxGameplay/Gameplay/B_ToyFoxGameStateComponent.B_ToyFoxGameStateComponent_C
   ) -> (then, ReturnValue:B Toy Fox Game State Component Object Reference)
   > Is Valid(InputObject=@n8)
   > Bind Event to On Enemy Health Percent Changed(self:B Toy Fox Game State Component Object Reference=@n8, Delegate=Create Event()) {is=title K2Node_AddDelegate}
   > HandleEnemyHealthPercentChanged!(
    execute
    self
    HealthPercent:double=GetEnemyHealthPercent!(self:B Toy Fox Game State Component Object Reference=@n8) -> (HealthPercent:double)
   ) -> (then) {is=title K2Node_CallFunction}
   > Bind Event to On Enemy Destroyed(self:B Toy Fox Game State Component Object Reference=@n8, Delegate=Create Event()) {is=title K2Node_AddDelegate}
   > Bind Event to On Enemy Stunned Active Changed(self:B Toy Fox Game State Component Object Reference=@n8, Delegate=Create Event()) {is=title K2Node_AddDelegate}
   > Bind Event to On Enemy Spawn in Complete(self:B Toy Fox Game State Component Object Reference=@n8, Delegate=Create Event()) {is=title K2Node_AddDelegate}
   > n9 = WaitGameplayTagAddToActor!(
    execute:Exec
    TargetActor:Actor Object Reference=GetGiantPawn!(self:B Toy Fox Game State Component Object Reference=@n8) -> (Pawn:Toy Fox Giant Player Pawn Object Reference)
    Tag:Gameplay Tag Structure=(TagName="ToyFox.Ability.Finisher.Active")
    OnlyTriggerOnce:bool=false
   ) -> (then:Exec, AsyncTaskProxy:Ability Async Wait Gameplay Tag Added Object Reference, Added:Exec) {is=title K2Node_AsyncAction}
   > Set AsyncAction_ListenForFinisher(AsyncAction_ListenForFinisher:Ability Async Wait Gameplay Tag Added Object Reference=@n9) -> (Output_Get:Ability Async Wait Gameplay Tag Added Object Reference)
   > Branch(Condition=IsEnemyActive!(self:B Toy Fox Game State Component Object Reference=@n8) -> (IsActive:bool))
   > SetVisibility(InVisibility=SelfHitTestInvisible)
   > n10 = Branch(Condition=IsEnemyStunned!(self:B Toy Fox Game State Component Object Reference=@n8) -> (IsStunned:bool))
   > HandleEnemyStunnedActiveChanged!(execute, self, bStunActive:bool=true) -> (then) {is=title K2Node_CallFunction}
   HandleEnemySpawnInComplete(execute=@n10.else) {is=title K2Node_CallFunction}
   HandleFinisherExecuted(execute=@n9.Added) {is=title K2Node_CallFunction}
   Event Destruct()
   > n11 = Sequence()
   > EndAction(self:Ability Async Object Reference=Get AsyncAction_ListenForFinisher() -> (AsyncAction_ListenForFinisher:Ability Async Wait Gameplay Tag Added Object Reference))
   n12 = GetGameStateAthena()
   > Is Valid(InputObject=@n12)
   > n13 = GetComponentByClass(self=@n12, ComponentClass=&B_ToyFoxGameStateComponent_C)
   > Is Valid(InputObject=@n13)
   > Unbind Event from On Enemy Health Percent Changed(self:B Toy Fox Game State Component Object Reference=@n13, Delegate=Create Event()) {is=title K2Node_RemoveDelegate}
   > Unbind Event from On Enemy Destroyed(self:B Toy Fox Game State Component Object Reference=@n13, Delegate=Create Event()) {is=title K2Node_RemoveDelegate}
   > Unbind Event from On Enemy Stunned Active Changed(self:B Toy Fox Game State Component Object Reference=@n13, Delegate=Create Event()) {is=title K2Node_RemoveDelegate}
   > Unbind Event from On Enemy Spawn in Complete(self:B Toy Fox Game State Component Object Reference=@n13, Delegate=Create Event()) {is=title K2Node_RemoveDelegate}
   Parent: Destruct(execute=@n11.then_1)
  ]}]
variables [
 AsyncAction_ListenForFinisher {type=AbilityAsync_WaitGameplayTagAdded, default=null}
 IsStunActive {type=boolean}
]
widgets {
 info {inheritedWidgetCount=1, parentClass=/Script/FortniteUI.AthenaCommonPrioritizedWidget, rootWidgetClass=&Overlay, widgetCount=16}
 widgets [
  Overlay_14 {widgetClassPath=&Overlay, widget=Overlay_14}
  Overlay_FinalHit {
   widgetClassPath=&Overlay
   parent=Overlay_14
   widget=Overlay_FinalHit
   properties {renderOpacity=0, visibility=Collapsed, renderTransform {angle=0, scale=<1 1>, shear=<0 0>, translation=<0 50>}}
   slot {class=OverlaySlot, properties {horizontalAlignment=HAlign_Center, verticalAlignment=VAlign_Center}}
  }
  Image_Edges {
   widgetClassPath=&Image
   bIsVariable=true
   parent=Overlay_FinalHit
   widget=Image_Edges
   properties {
    visibility=SelfHitTestInvisible
    brush {
     bIsDynamicallyLoaded=false
     drawAs=Image
     imageType=FullColor
     mirroring=NoMirror
     resourceName=None
     resourceObject=/ToyFoxGameplay/UI/Widgets/HUD/Assets/BossBar/MI_UI_ToyFox_FinalHit_Bg
     tiling=NoTile
     imageSize=<32 32>
     margin=<ltrb 0 0 0 0>
     outlineSettings {
      bUseBrushTransparency=false
      roundingType=HalfHeightRadius
      width=0
      color {colorUseRule=UseColor_Specified, specifiedColor=<rgba 0 0 0 0>}
      cornerRadii=<0 0 0 0>
     }
     tintColor {colorUseRule=UseColor_Specified, specifiedColor=<rgba 1 1 1 1>}
     uVRegion {bIsValid=false, max=<0 0>, min=<0 0>}
    }
   }
   slot {class=OverlaySlot, properties {horizontalAlignment=HAlign_Fill, verticalAlignment=VAlign_Fill}}
  }
  Image_TextBg {
   widgetClassPath=&Image
   bIsVariable=true
   parent=Overlay_FinalHit
   widget=Image_TextBg
   properties {
    visibility=SelfHitTestInvisible
    brush {
     bIsDynamicallyLoaded=false
     drawAs=Image
     imageType=FullColor
     mirroring=NoMirror
     resourceName=None
     resourceObject=/ToyFoxGameplay/UI/Widgets/HUD/Assets/BossBar/M_UI_ToyFox_FinalHit_Bg
     tiling=NoTile
     imageSize=<32 32>
     margin=<ltrb 0 0 0 0>
     outlineSettings {
      bUseBrushTransparency=false
      roundingType=HalfHeightRadius
      width=0
      color {colorUseRule=UseColor_Specified, specifiedColor=<rgba 0 0 0 0>}
      cornerRadii=<0 0 0 0>
     }
     tintColor {colorUseRule=UseColor_Specified, specifiedColor=<rgba 1 1 1 1>}
     uVRegion {bIsValid=false, max=<0 0>, min=<0 0>}
    }
   }
   slot {class=OverlaySlot, properties {horizontalAlignment=HAlign_Fill, verticalAlignment=VAlign_Fill}}
  }
  Image_LensFlare {
   widgetClassPath=&Image
   bIsVariable=true
   parent=Overlay_FinalHit
   widget=Image_LensFlare
   properties {
    renderOpacity=0
    visibility=SelfHitTestInvisible
    brush {
     bIsDynamicallyLoaded=false
     drawAs=Image
     imageType=FullColor
     mirroring=NoMirror
     resourceName=None
     resourceObject=/ToyFoxGameplay/UI/Widgets/HUD/Assets/BossBar/M_UI_ToyFox_LensFlare
     tiling=NoTile
     imageSize=<32 32>
     margin=<ltrb 0 0 0 0>
     outlineSettings {
      bUseBrushTransparency=false
      roundingType=HalfHeightRadius
      width=0
      color {colorUseRule=UseColor_Specified, specifiedColor=<rgba 0 0 0 0>}
      cornerRadii=<0 0 0 0>
     }
     tintColor {colorUseRule=UseColor_Specified, specifiedColor=<rgba 1 1 1 1>}
     uVRegion {bIsValid=false, max=<0 0>, min=<0 0>}
    }
    renderTransform {angle=0, scale=<4 4>, shear=<0 0>, translation=<0 0>}
   }
   slot {class=OverlaySlot, properties {horizontalAlignment=HAlign_Fill, verticalAlignment=VAlign_Fill}}
  }
  Spacer_47 {
   widgetClassPath=&Spacer
   parent=Overlay_FinalHit
   widget=Spacer_47
   properties {size=<450 60>}
   slot {class=OverlaySlot, properties {horizontalAlignment=HAlign_Fill, verticalAlignment=VAlign_Center}}
  }
  UEFN_TextBlock_C_96 {
   widgetClassPath=&UEFN_TextBlock
   parent=Overlay_FinalHit
   widget=UEFN_TextBlock_C_96
   properties {
    text="FINAL POWER SWORD HIT!"
    visibility=SelfHitTestInvisible
    font {
     bForceMonospaced=false
     bMaterialIsStencil=false
     fontMaterial=/ToyFoxGameplay/UI/Widgets/HUD/Assets/BossBar/M_UI_ToyFox_FinalHit_Font
     fontObject=&BurbankSmall
     letterSpacing=0
     monospacedWidth=1
     size=16.5
     skewAmount=0.2
     typefaceFontName=Black
     outlineSettings {
      bApplyOutlineToDropShadows=false
      bMiteredCorners=false
      bSeparateFillAlpha=false
      outlineMaterial=&M-FontShadow
      outlineSize=3
      outlineColor=<rgba 0 0 0 0.5>
     }
    }
   }
   slot {class=OverlaySlot, properties {horizontalAlignment=HAlign_Center, verticalAlignment=VAlign_Center, padding=<ltrb 30 10 30 10>}}
  }
  Overlay_ProgressBar {
   widgetClassPath=&Overlay
   parent=Overlay_14
   widget=Overlay_ProgressBar
   slot {class=OverlaySlot, properties {horizontalAlignment=HAlign_Fill, verticalAlignment=VAlign_Fill}}
  }
  SizeBox_ProgressBar {
   widgetClassPath=&SizeBox
   parent=Overlay_ProgressBar
   widget=SizeBox_ProgressBar
   properties {bOverride_HeightOverride=true, bOverride_WidthOverride=true, heightOverride=70, widthOverride=700}
   slot {class=OverlaySlot}
  }
  Image_ProgressBar {
   widgetClassPath=&Image
   bIsVariable=true
   parent=SizeBox_ProgressBar
   widget=Image_ProgressBar
   properties {
    visibility=SelfHitTestInvisible
    brush {
     bIsDynamicallyLoaded=false
     drawAs=Image
     imageType=FullColor
     mirroring=NoMirror
     resourceName=None
     resourceObject=/ToyFoxGameplay/UI/Widgets/HUD/Assets/BossBar/M_UI_ToyFox_BossBar
     tiling=NoTile
     imageSize=<32 32>
     margin=<ltrb 0 0 0 0>
     outlineSettings {
      bUseBrushTransparency=false
      roundingType=HalfHeightRadius
      width=0
      color {colorUseRule=UseColor_Specified, specifiedColor=<rgba 0 0 0 0>}
      cornerRadii=<0 0 0 0>
     }
     tintColor {colorUseRule=UseColor_Specified, specifiedColor=<rgba 1 1 1 1>}
     uVRegion {bIsValid=false, max=<0 0>, min=<0 0>}
    }
   }
   slot {class=SizeBoxSlot}
  }
  Image_Glow_1 {
   widgetClassPath=&Image
   bIsVariable=true
   parent=Overlay_ProgressBar
   widget=Image_Glow_1
   properties {
    renderOpacity=0
    visibility=Collapsed
    brush {
     bIsDynamicallyLoaded=false
     drawAs=Image
     imageType=FullColor
     mirroring=NoMirror
     resourceName=None
     resourceObject=/ToyFoxGameplay/UI/Widgets/HUD/Assets/BossBar/MI_UI_ToyFox_Glow
     tiling=NoTile
     imageSize=<64 64>
     margin=<ltrb 0 0 0 0>
     outlineSettings {
      bUseBrushTransparency=false
      roundingType=HalfHeightRadius
      width=0
      color {colorUseRule=UseColor_Specified, specifiedColor=<rgba 0 0 0 0>}
      cornerRadii=<0 0 0 0>
     }
     tintColor {colorUseRule=UseColor_Specified, specifiedColor=<rgba 1 1 1 1>}
     uVRegion {bIsValid=false, max=<0 0>, min=<0 0>}
    }
    renderTransform {angle=0, scale=<2 1.9>, shear=<0 0>, translation=<0 -10>}
   }
   slot {class=OverlaySlot, properties {horizontalAlignment=HAlign_Fill, verticalAlignment=VAlign_Fill}}
  }
  Text_Title {
   widgetClassPath=&CommonTextBlock
   bIsVariable=true
   parent=Overlay_14
   widget=Text_Title
   properties {
    mobileFontSizeMultiplier=1.2
    text=Rita's Monster
    textTransformPolicy=ToUpper
    wrapTextAt=680
    font {
     bForceMonospaced=false
     bMaterialIsStencil=false
     fontMaterial=null
     fontObject=&BurbankSmall
     letterSpacing=0
     monospacedWidth=1
     size=16.5
     skewAmount=0
     typefaceFontName=Black
     outlineSettings {
      bApplyOutlineToDropShadows=false
      bMiteredCorners=false
      bSeparateFillAlpha=false
      outlineMaterial=&M-FontShadow
      outlineSize=3
      outlineColor=<rgba 0 0 0 0.5>
     }
    }
    renderTransform {angle=0, scale=<1 1>, shear=<0 0>, translation=<0 -25>}
   }
   slot {class=OverlaySlot, properties {horizontalAlignment=HAlign_Center, verticalAlignment=VAlign_Fill}}
  }
  Image_Icon {
   widgetClassPath=&Image
   bIsVariable=true
   parent=Overlay_14
   widget=Image_Icon
   properties {
    renderOpacity=0
    visibility=Collapsed
    brush {
     bIsDynamicallyLoaded=false
     drawAs=Image
     imageType=FullColor
     mirroring=NoMirror
     resourceName=None
     resourceObject=/ToyFoxGameplay/UI/Widgets/HUD/Assets/BossBar/M_UI_ToyFox_FinalHit_Exclamation
     tiling=NoTile
     imageSize=<80 80>
     margin=<ltrb 0 0 0 0>
     outlineSettings {
      bUseBrushTransparency=false
      roundingType=HalfHeightRadius
      width=0
      color {colorUseRule=UseColor_Specified, specifiedColor=<rgba 0 0 0 0>}
      cornerRadii=<0 0 0 0>
     }
     tintColor {colorUseRule=UseColor_Specified, specifiedColor=<rgba 1 1 1 1>}
     uVRegion {bIsValid=false, max=<0 0>, min=<0 0>}
    }
    renderTransform {angle=0, scale=<2 2>, shear=<0 0>, translation=<0 -7>}
   }
   slot {class=OverlaySlot, properties {horizontalAlignment=HAlign_Center, verticalAlignment=VAlign_Center}}
  }
  Image_Glow {
   widgetClassPath=&Image
   bIsVariable=true
   parent=Overlay_14
   widget=Image_Glow
   properties {
    renderOpacity=0
    visibility=Collapsed
    brush {
     bIsDynamicallyLoaded=false
     drawAs=Image
     imageType=FullColor
     mirroring=NoMirror
     resourceName=None
     resourceObject=/ToyFoxGameplay/UI/Materials/MI_UI_ToyFox_Glow
     tiling=NoTile
     imageSize=<64 64>
     margin=<ltrb 0 0 0 0>
     outlineSettings {
      bUseBrushTransparency=false
      roundingType=HalfHeightRadius
      width=0
      color {colorUseRule=UseColor_Specified, specifiedColor=<rgba 0 0 0 0>}
      cornerRadii=<0 0 0 0>
     }
     tintColor {colorUseRule=UseColor_Specified, specifiedColor=<rgba 1 1 1 1>}
     uVRegion {bIsValid=false, max=<0 0>, min=<0 0>}
    }
    renderTransform {angle=0, scale=<3 3>, shear=<0 0>, translation=<0 -10>}
   }
   slot {class=OverlaySlot, properties {horizontalAlignment=HAlign_Center, verticalAlignment=VAlign_Center}}
  }
  Image_Lightning {
   widgetClassPath=&Image
   bIsVariable=true
   parent=Overlay_14
   widget=Image_Lightning
   properties {
    renderOpacity=0
    brush {
     bIsDynamicallyLoaded=false
     drawAs=Image
     imageType=FullColor
     mirroring=NoMirror
     resourceName=None
     resourceObject=/ToyFoxGameplay/UI/Widgets/HUD/Assets/BossBar/M_UI_ToyFox_Lightning
     tiling=NoTile
     imageSize=<80 80>
     margin=<ltrb 0 0 0 0>
     outlineSettings {
      bUseBrushTransparency=false
      roundingType=HalfHeightRadius
      width=0
      color {colorUseRule=UseColor_Specified, specifiedColor=<rgba 0 0 0 0>}
      cornerRadii=<0 0 0 0>
     }
     tintColor {colorUseRule=UseColor_Specified, specifiedColor=<rgba 1 1 1 1>}
     uVRegion {bIsValid=false, max=<0 0>, min=<0 0>}
    }
    renderTransform {angle=0, scale=<1.5 1.5>, shear=<0 0>, translation=<0 -20>}
   }
   slot {class=OverlaySlot, properties {horizontalAlignment=HAlign_Center, verticalAlignment=VAlign_Center}}
  }
  TutorialHighlight {widgetClassPath=&TutorialHighlight}
 ]
}
