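# UI material for the ToyFox health bottom bar, drawing a border shape with alpha-composite blending.
# Vertically tiled and offset UVs pass through fill scaling to sample an MSDF top-border texture, which converts to a signed distance field.
# The field's anti-aliased outer mask is dimmed and multiplied by a grey tint that the Flash parameter lerps toward white.
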Material /ToyFoxGameplay/UI/Materials/M_ToyFoxHealthBottomBar
graph {
 properties {bAutomaticallySetUsageInEditor=false, blendMode=BLEND_AlphaComposite, materialDomain=MD_UI, usedShadingModels=Default Lit}
 nodes [
  Constant3Vector() {constant=<rgba 1 0 0 0>}
  Constant()
  n1 = Add(A=TextureCoordinate() {vTiling=5}, B=Constant2Vector() {g=-3})
  n2 = MaterialFunctionCall!(MSDF RGB=TextureSample(UVs=MaterialFunctionCall(UV=@n1) {materialFunction=&MF_UI_UVScaling}.Scaled To Fill) {texture=/ToyFoxGameplay/UI/Materials/Textures/T_UI_ToyFox_TopBorder_1}.RGB) {materialFunction=/FortUILibrary/MaterialFunctions/Shapes/Modifiers/MF_UI_MSDFtoSDF}
  n3 = Constant() {r=0.1}
  Multiply(
   A=LinearInterpolate(A=Constant3Vector() {constant=<rgba 0.2865 0.273 0.2732 1>}, Alpha=ScalarParameter() {parameterName=Flash})
   B=Multiply(A=MaterialFunctionCall(DistanceField=@n2.SDF, Threshold=@n3) {materialFunction=&MF_UI_AntiAliasing}.Outer) {constB=0.3}
  )
 ]
}
