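# Alpha-composite UI material for an animated ToyFox glow effect with streaking lines.
# Time-driven cosine gradients supply shifting colors, which the "WowRed!" parameter blends toward red.
# The R, G and B channels of one lines texture pan at separate speeds and are anti-alias thresholded into three layers.
# Color and layer terms are summed and scaled by Master Opacity, and an inverted SDF box defines an edge mask.
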Material /ToyFoxGameplay/UI/Materials/M_ToyFox_FX_Glow
graph {
 properties {bAutomaticallySetUsageInEditor=false, blendMode=BLEND_AlphaComposite, materialDomain=MD_UI, usedShadingModels=Default Lit}
 nodes [
  n1 = MaterialFunctionCall(R=Multiply(A=Time()) {constB=5}, G=Constant() {r=0.7}, B=Constant() {r=0.1}) {materialFunction=&AppendMany}
  n2 = NamedRerouteDeclaration(None=Multiply(A=MaterialFunctionCall!(Gradient, a, b, c, d=@n1.RGB) {materialFunction=/FortUILibrary/MaterialFunctions/Color/MF_UI_CosineGradient}.Result)) {name=Color, nodeColor=<rgba 0 0 0 1>}
  Add(A=@n2) {constB=0.2}
  n3 = Time()
  n4 = MaterialFunctionCall(d=MaterialFunctionCall(R=Multiply(A=@n3) {constB=2}, G=Constant() {r=0.5}, B=Constant() {r=0.1}) {materialFunction=&AppendMany}.RGB) {materialFunction=&MF_UI_CosineGradient}
  Add(A=@n4.Result) {constB=3.2}
  n5 = Multiply(
   A=LinearInterpolate(A=@n2, B=Constant3Vector() {constant=<rgba 2 0 0 0>}, Alpha=ScalarParameter() {parameterName="WowRed!"})
   B=Multiply(A=NamedRerouteUsage())
  )
  n6 = NamedRerouteUsage()
  n7 = LinearInterpolate(
   A=Multiply(A=Multiply(A=@n4.Result), B=Multiply(A=@n6) {constB=0.1})
   B=Constant3Vector() {constant=<rgba 0.9141 0 0 0>}
   Alpha=ScalarParameter() {parameterName="WowRed!"}
  )
  n8 = Multiply(A=NamedRerouteUsage())
  n9 = Multiply(A=Add(A=Add(A=@n5, B=@n7), B=Multiply(A=@n7, B=@n8)), B=ScalarParameter() {defaultValue=1, parameterName=Master Opacity})
  n10 = NamedRerouteUsage()
  Multiply(A=@n9, B=@n10)
  Multiply(A=@n9, B=@n10)
  n11 = Time()
  n12 = TextureSample(UVs=Panner(Coordinate=TextureCoordinate(), Time=Multiply(A=@n11)) {speedX=-2}) {texture=/ToyFoxGameplay/UI/Materials/Textures/T_UI_ToyFox_Lines}
  NamedRerouteDeclaration(None=MaterialFunctionCall(DistanceField=Multiply(A=@n12.R), Threshold=Constant() {r=0.3}, Smoothing=Constant() {r=0.1}) {materialFunction=&MF_UI_AntiAliasing}.Outer) {name=Layer 01, nodeColor=<rgba 0 0 0 1>}
  n13 = Panner(Coordinate=TextureCoordinate(), Time=Multiply(A=Time())) {speedX=-0.4}
  NamedRerouteDeclaration(None=MaterialFunctionCall(
    DistanceField=TextureSample(UVs=@n13) {texture=&T_UI_ToyFox_Lines}.G
    Threshold=Constant() {r=0.3}
    Smoothing=Constant() {r=0.2}
   ) {materialFunction=&MF_UI_AntiAliasing}.Outer) {name=Layer 02, nodeColor=<rgba 0 0 0 1>}
  n14 = Panner(Coordinate=TextureCoordinate(), Time=Multiply(A=Time())) {speedX=-1}
  NamedRerouteDeclaration(None=MaterialFunctionCall(
    DistanceField=TextureSample(UVs=@n14) {texture=&T_UI_ToyFox_Lines}.B
    Threshold=Constant() {r=0.3}
    Smoothing=Constant() {r=0.2}
   ) {materialFunction=&MF_UI_AntiAliasing}.Outer) {name=Layer 03, nodeColor=<rgba 0 0 0 1>}
  n15 = MaterialFunctionCall() {materialFunction=&MF_UI_UVScaling}
  n16 = Constant2Vector() {g=-0.078, r=-0.042}
  NamedRerouteDeclaration(None=OneMinus(None=MaterialFunctionCall!(UVs=@n15.Scaled To Fill, Size=@n16, CornerRoundness, KeepAspectRatio, PixelSize) {materialFunction=/FortUILibrary/MaterialFunctions/Shapes/SDFs/MF_SDF_Box}.SDF)) {name=Edge Mask, nodeColor=<rgba 0 0 0 1>}
 ]
}
