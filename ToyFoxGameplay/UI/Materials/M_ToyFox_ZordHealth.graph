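# UI material for the ToyFox Zord health bar, drawn as a slanted, segmented red bar over a dark background.
# Skewed UVs feed SDF boxes and a sine stripe pattern for the bar shape, while HealthProgress and HealthDelta drive step masks for fill and change.
# Scalar switches blend in a time-sine red flash (BeepBoopOn), an animated cosine-gradient shimmer and an additive white highlight.
# Named reroutes share masks and glows across the graph.
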Material /ToyFoxGameplay/UI/Materials/M_ToyFox_ZordHealth
graph {
 properties {bAutomaticallySetUsageInEditor=false, blendMode=BLEND_AlphaComposite, materialDomain=MD_UI, usedShadingModels=Default Lit}
 nodes [
  n1 = NamedRerouteUsage()
  n2 = OneMinus(None=@n1)
  n3 = NamedRerouteUsage()
  Add(A=Multiply(A=@n1) {constB=0.7}, B=Multiply(A=Multiply(A=@n2, B=@n3)) {constB=0.8})
  Constant3Vector() {constant=<rgba 0.0781 0.0614 0.062 1>}
  n4 = NamedRerouteUsage()
  n5 = LinearInterpolate(A=Constant3Vector() {constant=<rgba 0.1927 0 0 0>}, B=Constant3Vector() {constant=<rgba 0.9349 0 0 0>}, Alpha=NamedRerouteUsage())
  n6 = ScalarParameter() {parameterName=Progress Flash On}
  n7 = Constant3Vector() {constant=<rgba 1 0.0286 0.0797 1>}
  n8 = Constant3Vector() {constant=<rgba 0.4531 0 0 0>}
  n9 = MaterialFunctionCall(Speed=Constant() {r=2}) {materialFunction=&DebugTimeSine}
  n10 = Multiply(A=NamedRerouteUsage(), B=NamedRerouteUsage())
  n11 = Multiply(
   A=LinearInterpolate(
    A=LinearInterpolate(A=@n5, Alpha=@n6)
    B=LinearInterpolate(A=@n7, B=@n8, Alpha=@n9.Result)
    Alpha=ScalarParameter() {desc=default 0, parameterName=BeepBoopOn}
   )
   B=@n10
  )
  n12 = Time()
  n13 = MaterialFunctionCall!(
   Gradient
   a
   b
   c
   d=MaterialFunctionCall(R=Multiply(A=@n12) {constB=5}, G=Constant() {r=0.7}, B=Constant() {r=0.1}) {materialFunction=&AppendMany}.RGB
  ) {materialFunction=/FortUILibrary/MaterialFunctions/Color/MF_UI_CosineGradient}
  n14 = Add(A=Multiply(A=Add(A=@n13.Result) {constB=0.2}, B=NamedRerouteUsage()), B=@n11)
  n15 = ScalarParameter() {parameterName=Noise Switch}
  n16 = LinearInterpolate(
   A=Multiply(A=Constant3Vector() {constant=<rgba 0.0677 0.0069 0.0111 1>}, B=Multiply(A=@n4) {constB=0.7})
   B=Saturate(None=LinearInterpolate(A=@n11, B=@n14, Alpha=@n15))
   Alpha=NamedRerouteUsage()
  )
  n17 = Multiply(A=Constant3Vector() {constant=<rgba 1 1 1 1>}, B=Multiply(A=NamedRerouteUsage()) {constB=0.2})
  n18 = Multiply(A=NamedRerouteUsage())
  n19 = NamedRerouteUsage()
  LinearInterpolate(
   A=LinearInterpolate(A=@n16, B=Add(A=@n17, B=@n16), Alpha=ScalarParameter() {parameterName=Gradient Switch})
   B=Multiply(A=Constant3Vector() {constant=<rgba 0.0208 0 0.0003 1>}, B=@n18)
   Alpha=Multiply(A=@n18, B=OneMinus(None=@n19))
  )
  MaterialFunctionCall(Speed=Constant() {r=0.1}) {materialFunction=&DebugTimeSine}
  n20 = NamedRerouteUsage()
  n21 = Constant2Vector() {r=-30}
  n22 = Step(
   Y=ComponentMask(None=MaterialFunctionCall!(UV=@n20, Skew Angle=@n21) {materialFunction=/FortUILibrary/MaterialFunctions/Basics/MF_UI_Skew}.Skewed UVs) {r=true}
   X=LinearInterpolate(Alpha=ScalarParameter() {defaultValue=0.5, parameterName=HealthProgress}) {constA=-3, constAlpha=1, constB=4}
  ) {constX=0.32}
  NamedRerouteDeclaration(None=Multiply(A=@n22)) {name=HealthProgressMask, nodeColor=<rgba 0 0 0 1>}
  NamedRerouteDeclaration(None=MaterialFunctionCall() {materialFunction=&MF_UI_UVScaling}.Scaled to Fit) {name=Skew Box UV, nodeColor=<rgba 0 0 0 1>}
  NamedRerouteDeclaration(None=@n10) {name=Combined Progress, nodeColor=<rgba 0 0 0 1>}
  n23 = ComponentMask(None=MaterialFunctionCall(UV=@n20, Skew Angle=@n21) {materialFunction=&MF_UI_Skew}.Skewed UVs) {r=true}
  n24 = LinearInterpolate(Alpha=ScalarParameter() {defaultValue=0.2, parameterName=HealthDelta}) {constA=0.08, constAlpha=1, constB=0.925}
  n25 = Multiply(A=Subtract(A=Step(Y=@n23, X=@n24) {constX=0.32}, B=@n22))
  NamedRerouteDeclaration(None=Max(A=NamedRerouteDeclaration(None=@n25) {name=HealthDeltaMask, nodeColor=<rgba 0 0 0 1>})) {name=CombinedHealthMask, nodeColor=<rgba 0 0 0 1>}
  n26 = MaterialFunctionCall(UV=NamedRerouteUsage(), Skew Angle=Constant2Vector() {r=-4}) {materialFunction=&MF_UI_Skew}
  n27 = Sine(None=Add(A=@n26.Skewed UVs, B=Constant2Vector() {r=0.035})) {period=0.042}
  n28 = Saturate(None=ComponentMask(None=Add(A=@n27) {constB=0.6}) {r=true})
  n29 = Constant() {r=0.2}
  n30 = Constant() {r=0.1}
  n31 = Constant2Vector() {g=0.6, r=0.85}
  n32 = Add(A=@n31, B=Constant2Vector() {g=-0.1})
  n33 = MaterialFunctionCall(
   DistanceField=Multiply(A=MaterialFunctionCall!(UVs=@n26.Skewed UVs, Size=@n32, CornerRoundness, KeepAspectRatio, PixelSize) {materialFunction=/FortUILibrary/MaterialFunctions/Shapes/SDFs/MF_SDF_Box}.SDF) {constB=-1}
   Threshold=Constant() {r=0.015}
  ) {materialFunction=&MF_UI_AntiAliasing}
  NamedRerouteDeclaration(None=Multiply(
    A=MaterialFunctionCall(DistanceField=@n28, Threshold=@n29, Smoothing=@n30) {materialFunction=&MF_UI_AntiAliasing}.Outer
    B=@n33.Outer
   )) {name=HealthBarMask, nodeColor=<rgba 0 0 0 1>}
  NamedRerouteDeclaration(None=MaterialFunctionCall(DistanceField=MaterialFunctionCall(UVs=@n26.Skewed UVs, Size=@n31) {materialFunction=&MF_SDF_Box}.SDF) {materialFunction=&MF_UI_AntiAliasing}.Inner) {name=Health Bg Mask, nodeColor=<rgba 0 0 0 1>}
  NamedRerouteDeclaration(None=MaterialFunctionCall() {materialFunction=&GetUserInterfaceUV}.Normalized UV) {name=Normalized Uv, nodeColor=<rgba 0 0 0 1>}
  n34 = NamedRerouteUsage()
  n35 = Multiply(A=Subtract(A=ComponentMask(None=@n34) {g=true}) {constB=0.5}) {constB=2}
  n36 = Multiply(A=OneMinus(None=Abs(None=@n35)), B=@n22)
  n37 = NamedRerouteUsage()
  n38 = MaterialFunctionCall(UV=@n34, Skew Angle=Constant() {r=-30}) {materialFunction=&MF_UI_Skew}
  n39 = Constant2Vector() {r=3}
  n40 = Multiply(A=ComponentMask(None=Add(A=@n38.Skewed UVs, B=@n39)) {r=true}) {constB=20}
  NamedRerouteDeclaration(None=Multiply(A=Multiply(A=@n36, B=@n37), B=Saturate(None=@n40))) {name=Progress Glow, nodeColor=<rgba 0 0 0 1>}
  n41 = TextureCoordinate()
  NamedRerouteDeclaration(None=ComponentMask(None=@n41) {r=true}) {name=Gradient X, nodeColor=<rgba 0 0 0 1>}
  NamedRerouteDeclaration(None=ComponentMask(None=@n41) {g=true}) {name=Gradient Y, nodeColor=<rgba 0 0 0 1>}
  n42 = NamedRerouteUsage()
  n43 = NamedRerouteUsage()
  NamedRerouteDeclaration(None=Multiply(A=@n33.Outer, B=Multiply(A=@n42, B=@n43))) {name=Progress X Glow, nodeColor=<rgba 0 0 0 1>}
  NamedRerouteUsage()
  Saturate()
 ]
}
