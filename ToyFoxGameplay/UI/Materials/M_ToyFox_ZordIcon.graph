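# UI material for the ToyFox Zord icon, drawing the Megazord profile portrait over a red, skewed backer with outline and highlight.
# Skewed SDF boxes form the backer, while two offset samples of the profile texture build portrait, glow mask and shadow.
# Panning noise warps icon UVs through Glitch In.
# Scalar parameters drive highlight pulsing, background recolor (Bg To Black), sine-timed portrait flash, animated cosine-gradient health-regain tint and profile opacity.
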Material /ToyFoxGameplay/UI/Materials/M_ToyFox_ZordIcon
graph {
 properties {bAutomaticallySetUsageInEditor=false, blendMode=BLEND_AlphaComposite, materialDomain=MD_UI, usedShadingModels=Default Lit}
 nodes [
  n1 = NamedRerouteUsage()
  n2 = Constant2Vector() {r=0.02}
  n3 = Constant2Vector() {g=0.8, r=0.55}
  n4 = StaticBool() {value=true}
  n5 = MaterialFunctionCall!(
   UVs=MaterialFunctionCall!(UV=Add(A=@n1, B=@n2), Skew Angle=Constant2Vector() {r=-30}) {materialFunction=/FortUILibrary/MaterialFunctions/Basics/MF_UI_Skew}.Skewed UVs
   Size=@n3
   CornerRoundness
   KeepAspectRatio=@n4
   PixelSize
  ) {materialFunction=/FortUILibrary/MaterialFunctions/Shapes/SDFs/MF_SDF_Box}
  n6 = Multiply(
   A=MaterialFunctionCall!(DistanceField=@n5.SDF, Threshold, Thickness=Constant() {r=0.02}, Alignment=Constant() {r=-1}, Smoothing) {materialFunction=/FortUILibrary/MaterialFunctions/Shapes/Modifiers/MF_UI_ShapeOutline}.Outline
   B=Power(Base=NamedRerouteUsage())
  )
  n7 = Multiply(
   A=MaterialFunctionCall(DistanceField=@n5.SDF) {materialFunction=&MF_UI_AntiAliasing}.Inner
   B=Power(Base=NamedRerouteUsage())
  )
  n8 = Saturate(None=Max(
    A=NamedRerouteDeclaration(None=@n6) {name=BackerOutline, nodeColor=<rgba 0 0 0 1>}
    B=NamedRerouteDeclaration(None=@n7) {name=Backer, nodeColor=<rgba 0 0 0 1>}
   ))
  n9 = MaterialFunctionCall(UV=Add(A=NamedRerouteUsage(), B=Constant2Vector() {r=-0.212}), Skew Angle=Constant2Vector()) {materialFunction=&MF_UI_Skew}
  n10 = Add(A=NamedRerouteUsage()) {constB=0.1}
  n11 = Multiply(
   A=MaterialFunctionCall(
    DistanceField=MaterialFunctionCall(UVs=@n9.Skewed UVs, Size=@n3, KeepAspectRatio=@n4) {materialFunction=&MF_SDF_Box}.SDF
    Smoothing=Constant() {r=0.03}
   ) {materialFunction=&MF_UI_AntiAliasing}.Inner
   B=Multiply(A=Power(Base=@n10)) {constB=2}
  )
  n12 = LinearInterpolate(Alpha=MaterialFunctionCall(Speed=Constant() {r=2}) {materialFunction=&DebugTimeSine}.Result) {constB=2}
  n13 = ScalarParameter() {parameterName=Highlight Pulse}
  Add(
   A=LinearInterpolate(A=Multiply(A=@n8) {constB=5}, Alpha=NamedRerouteUsage()) {constB=0.9}
   B=Multiply(
    A=NamedRerouteDeclaration(None=@n11) {name=Highlight, nodeColor=<rgba 0 0 0 1>}
    B=LinearInterpolate(B=@n12, Alpha=@n13) {constA=0.2, constB=0}
   ) {constB=0.2}
  )
  n14 = NamedRerouteUsage()
  NamedRerouteDeclaration(None=Add(A=ComponentMask(None=@n14) {g=true}) {constB=-0.18}) {name=Y Gradient, nodeColor=<rgba 0 0 0 1>}
  n15 = Add(A=NamedRerouteUsage(), B=Constant2Vector() {g=-0.02, r=-0.052})
  n16 = Constant() {r=0.9}
  n17 = TextureSample(UVs=MaterialFunctionCall(UVs=@n15, Texture Scale=@n16) {materialFunction=&ScaleUVsByCenter}.UVs) {texture=/ToyFoxGameplay/UI/Materials/Textures/T_UI_ToyFox_MegazordProfile}
  NamedRerouteDeclaration(None=MaterialFunctionCall(DistanceField=@n17.A, Threshold=Constant() {r=0.5}) {materialFunction=&MF_UI_AntiAliasing}.Outer) {name=Portrait Alpha, nodeColor=<rgba 0 0 0 1>}
  NamedRerouteDeclaration(None=@n17.G) {name=Portrait Green Channel, nodeColor=<rgba 0 0 0 1>}
  n18 = NamedRerouteUsage()
  n19 = Time()
  n20 = LinearInterpolate(Alpha=ScalarParameter() {parameterName=Glitch In}) {constA=0.2, constB=6}
  n21 = TextureSample(UVs=Panner(Coordinate=TextureCoordinate(), Time=Multiply(A=@n19, B=@n20), Speed=Constant2Vector() {g=0.25})) {samplerType=SAMPLERTYPE_LinearGrayscale, texture=/Game/UI/HUD/Textures/UIPatterns/T_UI_GlitchNoise}
  n22 = ScalarParameter() {parameterName=Glitch In}
  NamedRerouteDeclaration(None=LinearInterpolate(
    A=ComponentMask(None=@n18) {g=true, r=true}
    B=ComponentMask(None=@n21.RGB) {g=true, r=true}
    Alpha=LinearInterpolate(Alpha=@n22) {constB=0.1}
   )) {name=Icon UVs, nodeColor=<rgba 0 0 0 1>}
  n23 = SmoothStep(Value=@n17.R) {constMax=0.005}
  n24 = Add(A=@n15, B=Constant2Vector() {r=0.02})
  n25 = Add(A=@n16) {constB=0.02}
  n26 = SmoothStep(Value=TextureSample(UVs=MaterialFunctionCall(UVs=@n24, Texture Scale=@n25) {materialFunction=&ScaleUVsByCenter}.UVs) {texture=&T_UI_ToyFox_MegazordProfile}.R) {constMax=0.005}
  NamedRerouteDeclaration(None=Max(A=@n23, B=@n26)) {name=PortraitGlowMask, nodeColor=<rgba 0 0 0 1>}
  n27 = NamedRerouteDeclaration(None=Multiply(A=NamedRerouteUsage(), B=Subtract(A=@n26, B=@n23))) {name=PortraitShadow, nodeColor=<rgba 0 0 0 1>}
  n28 = NamedRerouteDeclaration(None=@n17.RGB) {name=PortraitRGB, nodeColor=<rgba 0 0.5312 1 1>}
  n29 = NamedRerouteDeclaration(None=@n23) {name=PortraitRGBMask, nodeColor=<rgba 0 0 0 1>}
  NamedRerouteDeclaration(None=Multiply(A=LinearInterpolate(A=@n27, B=@n28, Alpha=@n29), B=ScalarParameter() {defaultValue=1, parameterName=Profile Master Opacity})) {name=PortraitFinal, nodeColor=<rgba 0 0 0 1>}
  n30 = LinearInterpolate(A=Constant3Vector() {constant=<rgba 0.9647 0.0086 0.0399 1>}, Alpha=ScalarParameter() {parameterName=Bg To Black})
  n31 = MaterialFunctionCall(Speed=Constant() {r=2}) {materialFunction=&DebugTimeSine}
  n32 = Multiply(
   A=NamedRerouteUsage()
   B=LinearInterpolate(B=LinearInterpolate(Alpha=@n31.Result) {constB=2}, Alpha=ScalarParameter() {parameterName=Highlight Pulse}) {constA=0.2, constB=0}
  )
  n33 = ScalarParameter() {parameterName=Bg To Black}
  n34 = LinearInterpolate(A=Constant3Vector() {constant=<rgba 0.9647 0.0002 0.0516 1>}, Alpha=@n33)
  n35 = NamedRerouteUsage()
  n36 = LinearInterpolate(A=Constant3Vector() {constant=<rgba 1 0.0084 0.0628 1>}, Alpha=@n33)
  n37 = NamedRerouteUsage()
  n38 = NamedRerouteUsage()
  n39 = LinearInterpolate(
   A=@n38
   Alpha=MaterialFunctionCall(Speed=Constant() {r=10}) {materialFunction=&DebugTimeSine}.Result
  ) {constB=10}
  n40 = LinearInterpolate(
   A=LinearInterpolate(A=Constant3Vector() {constant=<rgba 0.2232 0 0.0659 1>}, B=Constant3Vector() {constant=<rgba 1 0 0.0177 1>}, Alpha=NamedRerouteUsage())
   B=Constant3Vector() {constant=<rgba 0.9647 0.9647 0.9647 1>}
   Alpha=ScalarParameter() {parameterName=Bg To Black}
  ) {constB=0}
  n41 = NamedRerouteUsage()
  n42 = Add(
   A=LinearInterpolate(A=NamedRerouteUsage(), B=Multiply(A=@n38, B=@n39) {constB=10}, Alpha=ScalarParameter() {parameterName=Portrait Flash}) {constA=1}
   B=Saturate(None=LinearInterpolate(B=@n40, Alpha=@n41))
  )
  n43 = Multiply(A=Time()) {constB=5}
  n44 = Constant() {r=0.3}
  n45 = Constant()
  n46 = Multiply(
   A=MaterialFunctionCall!(
    Gradient
    a
    b
    c
    d=MaterialFunctionCall(R=@n43, G=@n44, B=@n45) {materialFunction=&AppendMany}.RGB
   ) {materialFunction=/FortUILibrary/MaterialFunctions/Color/MF_UI_CosineGradient}.Result
   B=NamedRerouteUsage()
  ) {constA=1}
  n47 = ScalarParameter() {parameterName=Portrait Regain Health}
  n48 = Add(
   A=Add(A=LinearInterpolate(B=@n34, Alpha=@n35), B=LinearInterpolate(B=@n36, Alpha=@n37))
   B=Multiply(A=LinearInterpolate(A=@n42, B=@n46, Alpha=@n47), B=ScalarParameter() {defaultValue=1, parameterName=Profile Master Opacity})
  )
  NamedRerouteDeclaration(None=Add(A=Multiply(A=@n30, B=@n32), B=Saturate(None=@n48))) {name=ColorComposite, nodeColor=<rgba 0 0 0 1>}
  NamedRerouteDeclaration(None=MaterialFunctionCall() {materialFunction=&GetUserInterfaceUV}.Normalized UV) {name=Normalized UV, nodeColor=<rgba 0 0 0 1>}
  NamedRerouteUsage()
 ]
}
