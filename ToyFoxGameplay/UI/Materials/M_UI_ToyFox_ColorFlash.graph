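# UI material for ToyFox that draws a soft, skewed backer shape flashing through animated colors, composited with alpha.
# A skewed rounded-box SDF with soft smoothing is multiplied by a vertical UV falloff, then sharpened and boosted into the Backer mask.
# Time feeds a cosine gradient that tints the mask for color, while the untinted mask gives a matching alpha term.
# Both terms are dimmed and scaled by a Master Opacity parameter.
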Material /ToyFoxGameplay/UI/Materials/M_UI_ToyFox_ColorFlash
graph {
 properties {bAutomaticallySetUsageInEditor=false, blendMode=BLEND_AlphaComposite, materialDomain=MD_UI, usedShadingModels=Default Lit}
 nodes [
  n1 = ScalarParameter() {defaultValue=1, parameterName=Master Opacity}
  Multiply(A=Multiply(A=NamedRerouteUsage()) {constB=0.2}, B=@n1)
  n2 = MaterialFunctionCall(R=Multiply(A=Time()) {constB=5}, G=Constant() {r=0.3}, B=Constant()) {materialFunction=&AppendMany}
  n3 = NamedRerouteUsage()
  Multiply(
   A=Multiply(
    A=MaterialFunctionCall!(Gradient, a, b, c, d=@n2.RGB) {materialFunction=/FortUILibrary/MaterialFunctions/Color/MF_UI_CosineGradient}.Result
    B=Multiply(A=@n3)
   ) {constB=0.2}
   B=@n1
  )
  n4 = MaterialFunctionCall!(
   UV=Add(A=MaterialFunctionCall() {materialFunction=&GetUserInterfaceUV}.Normalized UV, B=Constant2Vector() {r=0.02})
   Skew Angle=Constant2Vector() {r=-30}
  ) {materialFunction=/FortUILibrary/MaterialFunctions/Basics/MF_UI_Skew}
  n5 = Constant2Vector() {g=0.8, r=0.55}
  n6 = Constant() {r=0.1}
  n7 = StaticBool() {value=true}
  n8 = TextureCoordinate()
  n9 = Multiply(
   A=MaterialFunctionCall(
    DistanceField=MaterialFunctionCall!(UVs=@n4.Skewed UVs, Size=@n5, CornerRoundness=@n6, KeepAspectRatio=@n7, PixelSize) {materialFunction=/FortUILibrary/MaterialFunctions/Shapes/SDFs/MF_SDF_Box}.SDF
    Smoothing=Constant() {r=0.2}
   ) {materialFunction=&MF_UI_AntiAliasing}.Inner
   B=Power(Base=ComponentMask(None=@n8) {g=true})
  )
  NamedRerouteDeclaration(None=Multiply(A=Power(Base=@n9) {constExponent=3}) {constB=4}) {name=Backer, nodeColor=<rgba 0 0 0 1>}
 ]
}
