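# HUD meter for the ToyFox giant's power level, drawn with a portrait and material-driven bar.
# It binds to B_ToyFoxGameStateComponent's power and low-state events, setting bar progress directly or toggling a low-energy flash, shake and blink.
# Power increases play a regen animation, time-scaled to hotfixable recovery duration, whose animated variable lerps toward a predicted end, then snaps to the latest value.
# Finisher-buff tag watchers call PlayAnimation without an assigned animation.
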WidgetBlueprint /ToyFoxGameplay/UI/Widgets/HUD/WBP_ToyFox_EnergyMeter
animations [
 Anim_Intro {
  binding_count=1
  length_seconds=1
  object=&self:Anim_Intro
  sequence {bindings [SafeZone_Mobile {
     id=#1
     possessed_class=SafeZone
     tracks [{
       type=MovieSceneFloatTrack
       display_name=Render Opacity
       sections [{type=MovieSceneFloatSection, end_frame=20, start_frame=0, channels [None {type=MovieSceneScriptingFloatChannel, keys [[0,0],[19.9997,1]]}]}]
      }]
    }]}
 }
 Anim_HealthRegen {
  binding_count=7
  length_seconds=0.8334
  object=&self:Anim_HealthRegen
  sequence {
   bindings [
    Image_BottomAccent {
     id=#2
     possessed_class=Image
     tracks [{
       type=MovieSceneWidgetMaterialTrack
       sections [{type=MovieSceneParameterSection, end_frame=25, start_frame=0, channels [Flash {type=MovieSceneScriptingFloatChannel, keys [[0,1],[21,0]]}]}]
      }]
    }
    Image_M_ProgressBar {
     id=#3
     possessed_class=Image
     tracks [{
       type=MovieSceneWidgetMaterialTrack
       sections [{
         type=MovieSceneParameterSection
         end_frame=25
         start_frame=0
         channels [Noise Switch {type=MovieSceneScriptingFloatChannel, keys [[4,1],[21,0]]}]
        }]
      }]
    }
    Image_TopAccent {
     id=#4
     possessed_class=Image
     tracks [{
       type=MovieSceneWidgetMaterialTrack
       sections [{type=MovieSceneParameterSection, end_frame=25, start_frame=0, channels [Flash {type=MovieSceneScriptingFloatChannel, keys [[0,1],[21,0]]}]}]
      }]
    }
    Image_GlowRainbow {
     id=#5
     possessed_class=Image
     tracks [
      {
       type=MovieSceneByteTrack
       display_name=Visibility
       sections [{type=MovieSceneByteSection, end_frame=0, start_frame=0, channels [None {type=MovieSceneScriptingByteChannel, const [0,1]}]}]
      }
      {
       type=MovieSceneFloatTrack
       display_name=Render Opacity
       sections [{type=MovieSceneFloatSection, end_frame=0, start_frame=0, channels [None {type=MovieSceneScriptingFloatChannel, const [1,1]}]}]
      }
      {
       type=MovieSceneWidgetMaterialTrack
       sections [{
         type=MovieSceneParameterSection
         end_frame=15
         start_frame=0
         channels [Master Opacity {type=MovieSceneScriptingFloatChannel, keys [[0,0],[2,1],[14,0]]}]
        }]
      }
     ]
    }
    WBP_ToyFox_EnergyMeter_C_81 {
     id=#6
     possessed_class=WBP_ToyFox_EnergyMeter_C
     tracks [{
       type=MovieSceneDoubleTrack
       display_name=Anim Var Health Regen
       sections [{
         type=MovieSceneDoubleSection
         end_frame=25
         start_frame=0
         channels [None {type=MovieSceneScriptingDoubleChannel, keys [[0,0],[4,0],[21,1]]}]
        }]
      }]
    }
    Image_Lines {
     id=#7
     possessed_class=Image
     tracks [
      {
       type=MovieSceneFloatTrack
       display_name=Render Opacity
       sections [{type=MovieSceneFloatSection, end_frame=0, start_frame=0, channels [None {type=MovieSceneScriptingFloatChannel, const [1,1]}]}]
      }
      {
       type=MovieSceneWidgetMaterialTrack
       sections [{
         type=MovieSceneParameterSection
         end_frame=15
         start_frame=0
         channels [Master Opacity {type=MovieSceneScriptingFloatChannel, keys [[0,0],[3,0.3],[12,0]]}]
        }]
      }
      {
       type=MovieScene2DTransformTrack
       sections [{
         type=MovieScene2DTransformSection
         end_frame=12
         start_frame=0
         channels [
          Translation.X {type=MovieSceneScriptingFloatChannel, const [-48,1]}
          Translation.Y {type=MovieSceneScriptingFloatChannel, const [334,1]}
          Angle {type=MovieSceneScriptingFloatChannel, const [-87,1]}
          Scale.X {type=MovieSceneScriptingFloatChannel, keys [[0,2.35],[12,2.5546]]}
          Scale.Y {type=MovieSceneScriptingFloatChannel, const [1.55,2]}
          Shear.X {type=MovieSceneScriptingFloatChannel, const [0,1]}
          Shear.Y {type=MovieSceneScriptingFloatChannel, const [0,1]}
         ]
        }]
      }
     ]
    }
    Image_Profile {
     id=#8
     possessed_class=Image
     tracks [{
       type=MovieSceneWidgetMaterialTrack
       sections [{
         type=MovieSceneParameterSection
         end_frame=15
         start_frame=0
         channels [Portrait Regain Health {type=MovieSceneScriptingFloatChannel, keys [[0,0],[1,1],[7,0]]}]
        }]
      }]
    }
   ]
   display_rate {numerator=30}
  }
 }
 Anim_LowHealth {
  binding_count=5
  length_seconds=4.2333
  object=&self:Anim_LowHealth
  sequence {
   bindings [
    Image_Profile {
     id=#9
     possessed_class=Image
     tracks [{
       type=MovieSceneWidgetMaterialTrack
       sections [{
         type=MovieSceneParameterSection
         end_frame=17
         start_frame=0
         channels [
          Portrait Flash {type=MovieSceneScriptingFloatChannel, keys [[0,1],[8,0]]}
          Glitch In {type=MovieSceneScriptingFloatChannel, keys [[0,1],[8,0]]}
          Bg To Black {type=MovieSceneScriptingFloatChannel, keys [[0,1],[8,0]]}
         ]
        }]
      }]
    }
    Image_BottomAccent {
     id=#10
     possessed_class=Image
     tracks [{
       type=MovieSceneWidgetMaterialTrack
       sections [{type=MovieSceneParameterSection, end_frame=15, start_frame=0, channels [Flash {type=MovieSceneScriptingFloatChannel, keys [[0,1],[15,0]]}]}]
      }]
    }
    Overlay_14 {
     id=#11
     possessed_class=Overlay
     tracks [{
       type=MovieScene2DTransformTrack
       sections [{
         type=MovieScene2DTransformSection
         end_frame=5
         start_frame=0
         channels [
          Translation.X {type=MovieSceneScriptingFloatChannel, keys [[0,0],[1,-10],[2,10],[4,-10],[5,0]]}
          Translation.Y {type=MovieSceneScriptingFloatChannel, const [0,1]}
          Angle {type=MovieSceneScriptingFloatChannel, const [-3,1]}
          Scale.X {type=MovieSceneScriptingFloatChannel, const [1,1]}
          Scale.Y {type=MovieSceneScriptingFloatChannel, const [1,1]}
          Shear.X {type=MovieSceneScriptingFloatChannel, const [0,1]}
          Shear.Y {type=MovieSceneScriptingFloatChannel, const [0,1]}
         ]
        }]
      }]
    }
    Image_M_ProgressBar {
     id=#12
     possessed_class=Image
     tracks [{
       type=MovieSceneWidgetMaterialTrack
       sections [{
         type=MovieSceneParameterSection
         end_frame=17
         start_frame=0
         channels [
          BeepBoopOn {type=MovieSceneScriptingFloatChannel, keys [[16.0557,0],[17,1]]}
          Progress Flash On {type=MovieSceneScriptingFloatChannel, keys [[0.9443,1],[15.111,0]]}
          Gradient Switch {type=MovieSceneScriptingFloatChannel, keys [[0.9443,1],[15,0]]}
         ]
        }]
      }]
    }
    Image_TopAccent {
     id=#13
     possessed_class=Image
     tracks [{
       type=MovieSceneWidgetMaterialTrack
       sections [{type=MovieSceneParameterSection, end_frame=15, start_frame=0, channels [Flash {type=MovieSceneScriptingFloatChannel, keys [[0,1],[15,0]]}]}]
      }]
    }
   ]
   tracks [{
     type=MovieSceneAudioTrack
     display_name=SFX
     sections [{
       type=MovieSceneAudioSection
       end_frame=84
       start_frame=0
       channels [
        Volume {type=MovieSceneScriptingFloatChannel, default=1}
        Pitch {type=MovieSceneScriptingFloatChannel, default=1}
        AttachActor {type=MovieSceneScriptingActorReferenceChannel, default="<Struct 'MovieSceneObjectBindingID' (0x4697abf8d20) {}>"}
       ]
      }]
    }]
  }
 }
]
graphs [
 EventGraph {nodes [
   n1 = HandleGiantPowerLevelChanged!() -> (OutputDelegate, then, PowerLevel:double=~) {is=title K2Node_CustomEvent}
   > n2 = Sequence()
   > n3 = Branch(Condition=IsAnimationPlaying(InAnimation=Get Anim_HealthRegen() -> (Anim_HealthRegen:Widget Animation Object Reference)))
   n4 = "Clamp (Float)"(Value=@n1.PowerLevel, Min=0, Max=1)
   n5 = Branch(execute=@n3.else, Condition="float < float"(A=Get PreviousPowerLevelRegen() -> (PreviousPowerLevelRegen:double), B=@n4))
   > n6 = Set StartRegenPowerLevel(StartRegenPowerLevel:double=@n4) -> (Output_Get:double)
   n7 = GetCachedScalableFloatValue!(Input:"Fort Cached Scalable Float Structure (by ref)"=Get Hotfix_PowerLevelRecoveryAmountPerInterval() -> (Hotfix_PowerLevelRecoveryAmountPerInterval:Fort Cached Scalable Float Structure)) -> (ReturnValue:float)
   n8 = float / float(
    A=GetCachedScalableFloatValue(Input=Get Hotfix_PowerLevelRecoveryDuration() -> (Hotfix_PowerLevelRecoveryDuration:Fort Cached Scalable Float Structure))
    B=GetCachedScalableFloatValue(Input=Get Hotfix_PowerLevelRecoveryInterval() -> (Hotfix_PowerLevelRecoveryInterval:Fort Cached Scalable Float Structure))
   )
   > Set EndRegenPowerLevel(EndRegenPowerLevel:double=float + float(A=@n6, B=float * float(A=@n8, B=@n7) {is=title K2Node_PromotableOperator})) -> (Output_Get:double)
   n9 = Get Anim_HealthRegen()
   n10 = Get Hotfix_PowerLevelRecoveryDuration()
   > n11 = Play Animation with Finished event!(
    execute:Exec
    Widget:User Widget Object Reference=Self-Reference()
    InAnimation:Widget Animation Object Reference=@n9
    StartAtTime:float=0
    NumLoopsToPlay:int=1
    PlayMode:EUMGSequencePlayMode Enum
    PlaybackSpeed:float=float / float(A=GetEndTime(self=@n9), B=GetCachedScalableFloatValue(Input=@n10))
   ) -> (then:Exec, Result:Widget Animation Handle Structure, Finished:Exec) {is=title K2Node_PlayAnimation2}
   n12 = GetDynamicMaterial(execute=@n11.Finished, self=Get Image_M_ProgressBar() -> (Image_M_ProgressBar:Image Object Reference))
   > SetScalarParameterValue(self=@n12, ParameterName=Get Param_HealthProgress() -> (Param_HealthProgress:Name=None), Value=Get PreviousPowerLevelRegen())
   n13 = GetDynamicMaterial(execute=@n5.else, self=Get Image_M_ProgressBar())
   > SetScalarParameterValue(self=@n13, ParameterName=Get Param_HealthProgress() -> (Param_HealthProgress=None), Value=@n4)
   Set PreviousPowerLevelRegen(execute=@n2.then_1, PreviousPowerLevelRegen:double=@n4) -> (Output_Get:double)
   n14 = SetBuffEnabled!() -> (OutputDelegate, then, bEnabled:bool=~) {is=title K2Node_CustomEvent}
   > n15 = Branch(Condition=@n14.bEnabled)
   > PlayAnimation(NumLoopsToPlay=0)
   PlayAnimation(execute=@n15.else, PlayMode=Reverse)
   n16 = OnPowerLevelLowStateChanged!() -> (OutputDelegate, then, bLow:bool=~) {is=title K2Node_CustomEvent}
   > n17 = Delta Gate Wildcard!(execute:Exec, Variable:"Boolean (by ref)"=Get IsShowingLowEnergy() -> (IsShowingLowEnergy:bool), New Value:bool=@n16.bLow) -> (Success:Exec, Updated Variable:"Boolean (by ref)", Previous Variable:bool=~) {is=title K2Node_MacroInstance}
   > n18 = Branch(Condition=@n17)
   > n19 = Play Animation with Finished event(
    Widget=Self-Reference()
    InAnimation=Get Anim_LowHealth() -> (Anim_LowHealth:Widget Animation Object Reference)
    StartAtTime=0
    NumLoopsToPlay=1
    PlaybackSpeed=1
   )
   n20 = Branch(execute=@n19.Finished, Condition=Get IsShowingLowEnergy())
   n21 = GetDynamicMaterial(execute=@n20.else, self=Get Image_M_ProgressBar())
   > SetScalarParameterValue(self=@n21, ParameterName=BeepBoopOn, Value=0)
   n22 = Branch(execute=@n18.else, Condition=IsAnimationPlaying(InAnimation=Get Anim_LowHealth()))
   n23 = Get Anim_LowHealth()
   > PlayAnimation(InAnimation=@n23, StartAtTime=GetEndTime(self=@n23))
   n24 = GetDynamicMaterial(execute=@n22.else, self=Get Image_M_ProgressBar())
   > SetScalarParameterValue(self=@n24, ParameterName=BeepBoopOn, Value=0)
   Event Construct()
   > Parent: Construct()
   > n25 = Sequence()
   n26 = GetGameStateAthena!() -> (ReturnValue:Fort Game State Athena Object Reference)
   Is Valid(exec=@n25.then_1, InputObject=@n26)
   > n27 = GetComponentByClass!(
    execute
    self:Actor Object Reference=@n26
    ComponentClass:Actor Component Class Reference=/ToyFoxGameplay/Gameplay/B_ToyFoxGameStateComponent.B_ToyFoxGameStateComponent_C
   ) -> (then, ReturnValue:B Toy Fox Game State Component Object Reference)
   > n28 = Set ToyFoxGameStateComponent(ToyFoxGameStateComponent:B Toy Fox Game State Component Object Reference=@n27) -> (Output_Get:B Toy Fox Game State Component Object Reference)
   > Is Valid(InputObject=@n28)
   > Bind Event to On Giant Power Level Changed(self:B Toy Fox Game State Component Object Reference=@n28, Delegate=Create Event()) {is=title K2Node_AddDelegate}
   > HandleGiantPowerLevelChanged!(execute, self, PowerLevel:double=GetPowerLevel!(self:B Toy Fox Game State Component Object Reference=@n28) -> (Return:double)) -> (then) {is=title K2Node_CallFunction}
   > Bind Event to On Giant Power Level Low State Changed(
    self:B Toy Fox Game State Component Object Reference=Get ToyFoxGameStateComponent() -> (ToyFoxGameStateComponent:B Toy Fox Game State Component Object Reference)
    Delegate=Create Event()
   ) {is=title K2Node_AddDelegate}
   n29 = GetOwningPlayer()
   Is Valid(exec=@n25.then_0, InputObject=@n29)
   > n30 = Cast To FortPlayerPawn(Object=GetViewTarget!(self:Controller Object Reference=@n29) -> (ReturnValue:Actor Object Reference))
   n31 = Get FinisherBuffGameplayTagQuery() -> (FinisherBuffGameplayTagQuery:Gameplay Tag Query Structure)
   > n32 = WaitGameplayTagQueryOnActor!(
    execute:Exec
    TargetActor:Actor Object Reference=@n30
    TagQuery:Gameplay Tag Query Structure=@n31
    TriggerCondition:EWaitGameplayTagQueryTriggerCondition Enum=WhenTrue
    bOnlyTriggerOnce:bool=false
   ) -> (then:Exec, AsyncTaskProxy:Ability Async Wait Gameplay Tag Query Object Reference, Triggered:Exec) {is=title K2Node_AsyncAction}
   > Add(
    TargetArray:Array of Ability Async Wait Gameplay Tag Query Object References=Get AsyncActions() -> (AsyncActions:Array of Ability Async Wait Gameplay Tag Query Object References)
    NewItem:"Ability Async Wait Gameplay Tag Query Object Reference (by ref)"=@n32
   ) {is=title K2Node_CallArrayFunction}
   > n33 = WaitGameplayTagQueryOnActor(TargetActor=@n30, TagQuery=@n31, TriggerCondition=WhenFalse, bOnlyTriggerOnce=false)
   > Add(TargetArray=Get AsyncActions(), NewItem=@n33)
   SetBuffEnabled!(execute=@n33.Triggered, self, bEnabled:bool=false) -> (then) {is=title K2Node_CallFunction}
   SetBuffEnabled(execute=@n32.Triggered, bEnabled=true)
   Event Destruct()
   > EndAction(self:Ability Async Object Reference=Get AsyncActions())
   > n34 = Get!(execute:Exec) -> (then:Exec, else:Exec, ToyFoxGameStateComponent:B Toy Fox Game State Component Object Reference)
   > Unbind Event from On Giant Power Level Changed(self:B Toy Fox Game State Component Object Reference=@n34, Delegate=Create Event()) {is=title K2Node_RemoveDelegate}
   > Unbind Event from On Giant Power Level Low State Changed(self:B Toy Fox Game State Component Object Reference=@n34, Delegate=Create Event()) {is=title K2Node_RemoveDelegate}
   > Set ToyFoxGameStateComponent()
   Get Anim_HealthRegen()
  ]}
 SetAnimVar_HealthRegen {nodes [
   n1 = SetAnimVar_HealthRegen!() -> (then, NewVar:double=~) {is=title K2Node_FunctionEntry}
   > n2 = GetDynamicMaterial(self=Get Image_M_ProgressBar())
   > SetScalarParameterValue(
    self=@n2
    ParameterName=Get Param_HealthProgress() -> (Param_HealthProgress=None)
    Value=Lerp(A=Get StartRegenPowerLevel() -> (StartRegenPowerLevel:double), B=Get EndRegenPowerLevel() -> (EndRegenPowerLevel:double), Alpha=@n1)
   )
  ]}
]
variables [
 AnimVar_HealthRegen {type=number}
 AsyncActions {type="array<AbilityAsync_WaitGameplayTagQuery>", default []}
 EndRegenPowerLevel {type=number, default=1}
 FinisherBuffGameplayTagQuery {
  type=GameplayTagQuery
  default {
   autoDescription=" ALL(  ANY( ToyFox.Ability.Finisher.Buff ) )"
   tokenStreamVersion=0
   userDescription=""
   queryTokenStream [0,1,5,1,1,1,0]
   tagDictionary [{tagName=ToyFox.Ability.Finisher.Buff}]
  }
 }
 Hotfix_PowerLevelRecoveryAmountPerInterval {
  type=FortCachedScalableFloat
  default {scalableFloat {
    registryType=None
    value=1
    curve {curveTable=/ToyFoxGameplay/DataTables/ToyFoxGameplayGameData, rowName=ToyFox.Ability.Finisher.Buff.PowerLevel.Increase.AmountPerInterval}
   }}
 }
 Hotfix_PowerLevelRecoveryDuration {
  type=FortCachedScalableFloat
  default {scalableFloat {
    registryType=None
    value=1
    curve {curveTable=&ToyFoxGameplayGameData, rowName=ToyFox.Ability.Finisher.Buff.PowerLevel.Increase.Duration}
   }}
 }
 Hotfix_PowerLevelRecoveryInterval {
  type=FortCachedScalableFloat
  default {scalableFloat {
    registryType=None
    value=1
    curve {curveTable=&ToyFoxGameplayGameData, rowName=ToyFox.Ability.Finisher.Buff.PowerLevel.Increase.Interval}
   }}
 }
 IsShowingLowEnergy {type=boolean}
 Param_HealthProgress {type=string, default=HealthProgress}
 PreviousPowerLevelRegen {type=number, default=1}
 StartRegenPowerLevel {type=number, default=1}
 ToyFoxGameStateComponent {type=B_ToyFoxGameStateComponent_C, default=null}
]
widgets {
 info {inheritedWidgetCount=1, parentClass=&FortHUDElementWidget, rootWidgetClass=&SafeZone, widgetCount=13}
 widgets [
  SafeZone_Mobile {widgetClassPath=&SafeZone, bIsVariable=true, widget=SafeZone_Mobile, properties {padLeft=false, padRight=false, padTop=false}}
  Overlay_14 {
   widgetClassPath=&Overlay
   parent=SafeZone_Mobile
   widget=Overlay_14
   properties {renderTransform {angle=-3, scale=<1 1>, shear=<0 0>, translation=<0 0>}}
   slot {class=SafeZoneSlot}
  }
  Image_TopAccent {
   widgetClassPath=&Image
   bIsVariable=true
   parent=Overlay_14
   widget=Image_TopAccent
   properties {
    brush {
     bIsDynamicallyLoaded=false
     drawAs=Image
     imageType=FullColor
     mirroring=NoMirror
     resourceName=None
     resourceObject=/ToyFoxGameplay/UI/Materials/M_ToyFoxHealthBottomBar
     tiling=NoTile
     imageSize=<400 90>
     margin=<ltrb 0 0 0 0>
     outlineSettings {
      bUseBrushTransparency=false
      roundingType=HalfHeightRadius
      width=0
      color {colorUseRule=UseColor_Specified, specifiedColor=<rgba 0 0 0 0>}
      cornerRadii=<0 0 0 0>
     }
     tintColor {colorUseRule=UseColor_Specified, specifiedColor=<rgba 1 1 1 1>}
     uVRegion {bIsValid=false, max=<0 0>, min=<0 0>}
    }
    colorAndOpacity=<rgba 1 1 1 0.4>
    renderTransform {angle=0, scale=<1 1>, shear=<0 0>, translation=<72 5>}
   }
   slot {class=OverlaySlot}
  }
  Image_BottomAccent {
   widgetClassPath=&Image
   bIsVariable=true
   parent=Overlay_14
   widget=Image_BottomAccent
   properties {
    brush {
     bIsDynamicallyLoaded=false
     drawAs=Image
     imageType=FullColor
     mirroring=NoMirror
     resourceName=None
     resourceObject=/ToyFoxGameplay/UI/Materials/M_ToyFox_HealthTopBar
     tiling=NoTile
     imageSize=<32 98>
     margin=<ltrb 0 0 0 0>
     outlineSettings {
      bUseBrushTransparency=false
      roundingType=HalfHeightRadius
      width=0
      color {colorUseRule=UseColor_Specified, specifiedColor=<rgba 0 0 0 0>}
      cornerRadii=<0 0 0 0>
     }
     tintColor {colorUseRule=UseColor_Specified, specifiedColor=<rgba 1 1 1 1>}
     uVRegion {bIsValid=false, max=<0 0>, min=<0 0>}
    }
    renderTransform {angle=0, scale=<1 1>, shear=<-6 4>, translation=<-90 45>}
   }
   slot {class=OverlaySlot, properties {horizontalAlignment=HAlign_Fill}}
  }
  HorizontalBox {
   widgetClassPath=&HorizontalBox
   parent=Overlay_14
   widget=HorizontalBox
   slot {class=OverlaySlot, properties {horizontalAlignment=HAlign_Fill, verticalAlignment=VAlign_Fill, padding=<ltrb 4 2 4 2>}}
  }
  Overlay_1 {
   widgetClassPath=&Overlay
   parent=HorizontalBox
   widget=Overlay_1
   slot {class=HorizontalBoxSlot, properties {horizontalAlignment=HAlign_Center, verticalAlignment=VAlign_Center}}
  }
  Image_Profile {
   widgetClassPath=&Image
   parent=Overlay_1
   widget=Image_Profile
   properties {
    visibility=SelfHitTestInvisible
    brush {
     bIsDynamicallyLoaded=false
     drawAs=Image
     imageType=FullColor
     mirroring=NoMirror
     resourceName=None
     resourceObject=/ToyFoxGameplay/UI/Materials/M_ToyFox_ZordIcon
     tiling=NoTile
     imageSize=<70 70>
     margin=<ltrb 0 0 0 0>
     outlineSettings {
      bUseBrushTransparency=false
      roundingType=HalfHeightRadius
      width=0
      color {colorUseRule=UseColor_Specified, specifiedColor=<rgba 0 0 0 0>}
      cornerRadii=<0 0 0 0>
     }
     tintColor {colorUseRule=UseColor_Specified, specifiedColor=<rgba 1 1 1 1>}
     uVRegion {bIsValid=false, max=<0 0>, min=<0 0>}
    }
    renderTransform {angle=0, scale=<2 2>, shear=<0 0>, translation=<0 0>}
   }
   slot {class=OverlaySlot}
  }
  Image_GlowRainbow {
   widgetClassPath=&Image
   bIsVariable=true
   parent=Overlay_1
   widget=Image_GlowRainbow
   properties {
    renderOpacity=0
    visibility=Collapsed
    brush {
     bIsDynamicallyLoaded=false
     drawAs=Image
     imageType=FullColor
     mirroring=NoMirror
     resourceName=None
     resourceObject=/ToyFoxGameplay/UI/Materials/M_UI_ToyFox_ColorFlash
     tiling=NoTile
     imageSize=<32 32>
     margin=<ltrb 0 0 0 0>
     outlineSettings {
      bUseBrushTransparency=false
      roundingType=HalfHeightRadius
      width=0
      color {colorUseRule=UseColor_Specified, specifiedColor=<rgba 0 0 0 0>}
      cornerRadii=<0 0 0 0>
     }
     tintColor {colorUseRule=UseColor_Specified, specifiedColor=<rgba 1 1 1 1>}
     uVRegion {bIsValid=false, max=<0 0>, min=<0 0>}
    }
    renderTransform {angle=0, scale=<2.5 2.5>, shear=<0 0>, translation=<-4 -10>}
   }
   slot {class=OverlaySlot, properties {horizontalAlignment=HAlign_Fill, verticalAlignment=VAlign_Fill}}
  }
  Image_Lines {
   widgetClassPath=&Image
   bIsVariable=true
   parent=Overlay_1
   widget=Image_Lines
   properties {
    renderOpacity=0
    visibility=Collapsed
    brush {
     bIsDynamicallyLoaded=false
     drawAs=Image
     imageType=FullColor
     mirroring=NoMirror
     resourceName=None
     resourceObject=/ToyFoxGameplay/UI/Materials/M_ToyFox_FX_Glow
     tiling=NoTile
     imageSize=<32 32>
     margin=<ltrb 0 0 0 0>
     outlineSettings {
      bUseBrushTransparency=false
      roundingType=HalfHeightRadius
      width=0
      color {colorUseRule=UseColor_Specified, specifiedColor=<rgba 0 0 0 0>}
      cornerRadii=<0 0 0 0>
     }
     tintColor {colorUseRule=UseColor_Specified, specifiedColor=<rgba 1 1 1 1>}
     uVRegion {bIsValid=false, max=<0 0>, min=<0 0>}
    }
    renderTransform {angle=-87, scale=<2.35 1.55>, shear=<0 0>, translation=<-48 334>}
    renderTransformPivot=<-1 -1>
   }
   slot {class=OverlaySlot, properties {horizontalAlignment=HAlign_Fill, verticalAlignment=VAlign_Fill}}
  }
  Overlay_0 {
   widgetClassPath=&Overlay
   parent=HorizontalBox
   widget=Overlay_0
   slot {class=HorizontalBoxSlot, properties {verticalAlignment=VAlign_Center}}
  }
  SizeBox_ProgressBar {
   widgetClassPath=&SizeBox
   parent=Overlay_0
   widget=SizeBox_ProgressBar
   properties {
    bOverride_HeightOverride=true
    bOverride_WidthOverride=true
    heightOverride=60
    widthOverride=500
    renderTransform {angle=0, scale=<1 1>, shear=<0 0>, translation=<0 21>}
   }
   slot {class=OverlaySlot}
  }
  Image_M_ProgressBar {
   widgetClassPath=&Image
   bIsVariable=true
   parent=SizeBox_ProgressBar
   widget=Image_M_ProgressBar
   properties {
    brush {
     bIsDynamicallyLoaded=false
     drawAs=Image
     imageType=FullColor
     mirroring=NoMirror
     resourceName=None
     resourceObject=/ToyFoxGameplay/UI/Materials/M_ToyFox_ZordHealth
     tiling=NoTile
     imageSize=<32 32>
     margin=<ltrb 0 0 0 0>
     outlineSettings {
      bUseBrushTransparency=false
      roundingType=HalfHeightRadius
      width=0
      color {colorUseRule=UseColor_Specified, specifiedColor=<rgba 0 0 0 0>}
      cornerRadii=<0 0 0 0>
     }
     tintColor {colorUseRule=UseColor_Specified, specifiedColor=<rgba 1 1 1 1>}
     uVRegion {bIsValid=false, max=<0 0>, min=<0 0>}
    }
    renderTransform {angle=-1, scale=<1 1>, shear=<0 0>, translation=<8 0>}
   }
   slot {class=SizeBoxSlot}
  }
  TutorialHighlight {widgetClassPath=&TutorialHighlight}
 ]
}
