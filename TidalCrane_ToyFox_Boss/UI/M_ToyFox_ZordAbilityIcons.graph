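# UI material that draws ToyFox boss Zord ability icons in a red palette with alpha-composite blending.
# An antialiased distance-field icon texture supplies shape and outline, with IconTimer stepping a vertical progress fill.
# Skewed SDF boxes build a gradient backer, outline and highlight, alongside an MSDF bottom border.
# Scalar parameters drive sine shake, panning glitch-noise UV distortion, dark fill, flash, highlight and opacity, composited through named reroutes.
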Material /TidalCrane_ToyFox_Boss/UI/M_ToyFox_ZordAbilityIcons
graph {
 properties {bAutomaticallySetUsageInEditor=false, blendMode=BLEND_AlphaComposite, materialDomain=MD_UI, usedShadingModels=Default Lit}
 nodes [
  Constant() {r=0.75}
  MaterialFunctionCall(Speed=Constant() {r=0.1}) {materialFunction=&DebugTimeSine}
  n1 = NamedRerouteUsage()
  n2 = NamedRerouteUsage()
  n3 = LinearInterpolate(A=Add(A=Multiply(A=@n1) {constB=0.2}, B=Multiply(A=@n2) {constB=0.3}), Alpha=ScalarParameter() {parameterName=Backing Off Switcher}) {constB=0}
  n4 = Multiply(A=NamedRerouteUsage()) {constB=0.7}
  n5 = NamedRerouteUsage()
  n6 = ScalarParameter() {defaultValue=1, parameterName=Master Opacity}
  Multiply(A=Add(A=LinearInterpolate(A=@n3, Alpha=@n4), B=Multiply(A=@n5) {constB=0}), B=@n6)
  Multiply(A=NamedRerouteUsage(), B=@n6)
  n7 = TextureSampleParameter2D(UVs=NamedRerouteUsage()) {parameterName=Texture Input, texture=&/T_UI_ToyFox_Icon_Melee}
  n8 = Constant() {r=0.5}
  n9 = Constant() {r=0.1}
  n10 = MaterialFunctionCall(DistanceField=@n7.B, Threshold=Constant() {r=0.4}, Smoothing=Constant()) {materialFunction=&MF_UI_AntiAliasing}
  n11 = NamedRerouteDeclaration(None=Add(
    A=MaterialFunctionCall(DistanceField=@n7.R, Threshold=@n8, Smoothing=@n9) {materialFunction=&MF_UI_AntiAliasing}.Outer
    B=Multiply(A=@n10.Outer) {constB=0.2}
   )) {name=Icon, nodeColor=<rgba 0 0 0 1>}
  n12 = OneMinus(None=ComponentMask(None=NamedRerouteUsage()) {g=true})
  n13 = ScalarParameter() {defaultValue=1, parameterName=IconTimer}
  n14 = LinearInterpolate(Alpha=@n13) {constA=0.15, constB=0.85}
  n15 = NamedRerouteDeclaration(None=Multiply(A=NamedRerouteUsage(), B=Step(Y=@n12, X=@n14))) {name=IconFill, nodeColor=<rgba 0 0 0 1>}
  n16 = MaterialFunctionCall(DistanceField=@n7.G, Threshold=Constant() {r=0.1}, Smoothing=Constant()) {materialFunction=&MF_UI_AntiAliasing}
  NamedRerouteDeclaration(None=LinearInterpolate(A=@n11, B=Max(A=@n11, B=@n15), Alpha=NamedRerouteDeclaration(None=@n16.Outer) {name=IconOutline, nodeColor=<rgba 0 0 0 1>}) {constAlpha=0.7}) {name=IconComposite, nodeColor=<rgba 0 0 0 1>}
  n17 = MaterialFunctionCall(UVs=Add(A=NamedRerouteUsage(), B=Constant2Vector() {g=-0.37, r=0.14}), Texture Scale=Constant2Vector() {g=0.73, r=0.73}) {materialFunction=&ScaleUVsByCenter}
  n18 = MaterialFunctionCall!(MSDF RGB=TextureSample(UVs=MaterialFunctionCall(UV=@n17.UVs) {materialFunction=&MF_UI_UVScaling}.Scaled to Fit) {samplerType=SAMPLERTYPE_LinearColor, texture=&/T_UI_ToyFox_Ability_Border}.RGB) {materialFunction=/FortUILibrary/MaterialFunctions/Shapes/Modifiers/MF_UI_MSDFtoSDF}
  n19 = Constant() {r=0.5}
  NamedRerouteDeclaration(None=Multiply(A=MaterialFunctionCall(DistanceField=@n18.SDF, Threshold=@n19) {materialFunction=&MF_UI_AntiAliasing}.Outer) {constB=0.2}) {name=AbilityBottomBorder, nodeColor=<rgba 0 0 0 1>}
  n20 = Add(A=NamedRerouteUsage(), B=Constant2Vector() {r=-0.08})
  n21 = Constant2Vector() {g=-6, r=28.5}
  n22 = Constant2Vector() {g=0.825, r=0.525}
  n23 = StaticBool() {value=true}
  n24 = MaterialFunctionCall!(
   UVs=MaterialFunctionCall!(UV=@n20, Skew Angle=@n21) {materialFunction=/FortUILibrary/MaterialFunctions/Basics/MF_UI_Skew}.Skewed UVs
   Size=@n22
   CornerRoundness
   KeepAspectRatio=@n23
   PixelSize
  ) {materialFunction=/FortUILibrary/MaterialFunctions/Shapes/SDFs/MF_SDF_Box}
  n25 = Constant() {r=0.02}
  n26 = Constant() {r=-1}
  n27 = NamedRerouteUsage()
  n28 = Power(Base=ComponentMask(None=MaterialFunctionCall(UV=@n27, Skew Angle=@n21) {materialFunction=&MF_UI_Skew}.Skewed UVs) {g=true}) {constExponent=3}
  n29 = Add(A=NamedRerouteDeclaration(None=Saturate(None=@n28)) {name=Gradient, nodeColor=<rgba 1 0 0.9375 1>}) {constB=-0.1}
  n30 = Multiply(
   A=Multiply(
    A=MaterialFunctionCall!(DistanceField=@n24.SDF, Threshold, Thickness=@n25, Alignment=@n26, Smoothing) {materialFunction=/FortUILibrary/MaterialFunctions/Shapes/Modifiers/MF_UI_ShapeOutline}.Outline
    B=@n29
   )
   B=LinearInterpolate(Alpha=ScalarParameter() {parameterName=HighlightSwitch}) {constB=2}
  ) {constB=2}
  n31 = MaterialFunctionCall(DistanceField=@n24.SDF) {materialFunction=&MF_UI_AntiAliasing}
  n32 = Max(
   A=NamedRerouteDeclaration(None=Multiply(A=@n30) {constB=0.1}) {name=BackerOutline, nodeColor=<rgba 0 0 0 1>}
   B=NamedRerouteDeclaration(None=Multiply(A=@n31.Inner, B=@n29)) {name=Backer, nodeColor=<rgba 0 0 0 1>}
  )
  NamedRerouteDeclaration(None=Multiply(A=Saturate(None=@n32)) {constB=0.8}) {name=Backer+OutlineComposite, nodeColor=<rgba 0 0 0 1>}
  NamedRerouteDeclaration(None=ComponentMask(None=TextureCoordinate()) {g=true}) {name=V Gradient, nodeColor=<rgba 0 0 0 1>}
  NamedRerouteDeclaration(None=Step(Y=@n12, X=LinearInterpolate(Alpha=@n13) {constA=0.15, constB=0.8})) {name=ProgressGradientHighlight, nodeColor=<rgba 0 0 0 1>}
  n33 = MaterialFunctionCall(UV=Add(A=@n20, B=Constant2Vector() {g=0.03, r=0.22}), Skew Angle=Add(A=@n21, B=Constant2Vector() {r=-30}) {constB=10}) {materialFunction=&MF_UI_Skew}
  n34 = Add(A=@n22, B=Constant2Vector() {r=-0.05})
  n35 = Add(A=@n29) {constB=0.1}
  n36 = Multiply(
   A=MaterialFunctionCall(
    DistanceField=MaterialFunctionCall(UVs=@n33.Skewed UVs, Size=@n34, KeepAspectRatio=@n23) {materialFunction=&MF_SDF_Box}.SDF
    Smoothing=Constant() {r=0.05}
   ) {materialFunction=&MF_UI_AntiAliasing}.Inner
   B=Multiply(A=Power(Base=@n35) {constExponent=0.7}) {constB=0.625}
  )
  n37 = LinearInterpolate(Alpha=ScalarParameter() {parameterName=HighlightSwitch}) {constB=0.75}
  NamedRerouteDeclaration(None=Multiply(A=Multiply(A=@n36, B=@n37) {constB=0}) {constB=0.2}) {name=HighlightBacker, nodeColor=<rgba 0 0 0 1>}
  n38 = NamedRerouteUsage()
  n39 = Constant2Vector() {g=-0.025, r=0.05}
  n40 = LinearInterpolate(Alpha=ScalarParameter() {defaultValue=1, parameterName=Shake}) {constA=7, constB=0}
  n41 = Add(
   A=MaterialFunctionCall(UVs=Add(A=@n38, B=@n39), Texture Scale=ScalarParameter() {defaultValue=0.8, parameterName=Icon Size}) {materialFunction=&ScaleUVsByCenter}.UVs
   B=LinearInterpolate(
    A=Constant2Vector() {r=-0.036}
    B=Constant2Vector() {g=-0.048}
    Alpha=MaterialFunctionCall(Speed=@n40) {materialFunction=&DebugTimeSine}.Result
   )
  )
  n42 = Time()
  n43 = LinearInterpolate(Alpha=ScalarParameter() {parameterName=Speed Up}) {constA=0.2, constB=3}
  n44 = TextureSample(UVs=Panner(Coordinate=TextureCoordinate(), Time=Multiply(A=@n42, B=@n43), Speed=Constant2Vector() {g=0.25})) {samplerType=SAMPLERTYPE_LinearGrayscale, texture=/Game/UI/HUD/Textures/UIPatterns/T_UI_GlitchNoise}
  n45 = ScalarParameter() {parameterName=Glitch In}
  NamedRerouteDeclaration(None=LinearInterpolate(
    A=ComponentMask(None=@n41) {g=true, r=true}
    B=ComponentMask(None=@n44.RGB) {g=true, r=true}
    Alpha=LinearInterpolate(Alpha=@n45) {constB=0.1}
   )) {name=Icon UVs, nodeColor=<rgba 0 0 0 1>}
  NamedRerouteDeclaration(None=MaterialFunctionCall() {materialFunction=&GetUserInterfaceUV}.Normalized UV) {name=Normalized UV, nodeColor=<rgba 0 0 0 1>}
  n46 = Constant3Vector() {constant=<rgba 0.9647 0.0978 0.1628 1>}
  n47 = NamedRerouteUsage()
  n48 = LinearInterpolate(
   A=Constant3Vector() {constant=<rgba 2 0 0.0126 1>}
   B=LinearInterpolate(A=Constant3Vector() {constant=<rgba 0.7161 0.0057 0 1>}, B=Constant3Vector() {constant=<rgba 0.1276 0 0.0211 1>}, Alpha=NamedRerouteUsage())
   Alpha=Multiply(A=NamedRerouteUsage()) {constB=0.9}
  )
  n49 = Multiply(A=@n48) {constB=0.2}
  n50 = ScalarParameter() {parameterName=Dark Fill Switch}
  n51 = Constant3Vector() {constant=<rgba 0.0208 0 0.0003 1>}
  n52 = NamedRerouteUsage()
  n53 = Multiply(A=@n52, B=OneMinus(None=NamedRerouteUsage()))
  n54 = Constant3Vector() {constant=<rgba 0.0625 0 0.0013 1>}
  n55 = NamedRerouteUsage()
  n56 = Add(
   A=LinearInterpolate(B=LinearInterpolate(A=@n48, B=@n49, Alpha=@n50), Alpha=NamedRerouteUsage())
   B=Add(A=Multiply(A=@n51, B=@n53), B=Multiply(A=@n54, B=@n55))
  )
  n57 = Multiply(A=NamedRerouteUsage())
  n58 = ScalarParameter() {parameterName=Icon Flash Switch}
  n59 = OneMinus(None=@n52)
  n60 = Constant3Vector() {constant=<rgba 0.3542 0.0098 0.0349 1>}
  n61 = Add(
   A=Multiply(A=@n60, B=NamedRerouteUsage())
   B=LinearInterpolate(B=Constant3Vector() {constant=<rgba 0.9647 0.0068 0.032 1>}, Alpha=NamedRerouteUsage())
  )
  n62 = LinearInterpolate(A=Add(A=LinearInterpolate(A=@n56, B=@n57, Alpha=@n58), B=Multiply(A=@n59, B=@n61)), B=@n60, Alpha=NamedRerouteUsage())
  NamedRerouteDeclaration(None=Add(A=Multiply(A=@n46, B=@n47), B=Saturate(None=@n62))) {name=ColorComposite, nodeColor=<rgba 0 0 0 1>}
 ]
}
