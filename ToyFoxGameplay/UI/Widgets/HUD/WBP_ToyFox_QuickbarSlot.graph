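# HUD quickbar slot widget for ToyFox, showing an ability's material icon, keybind and cooldown state.
# Once the game state component's event state matches, it watches the controlled pawn's cooldown tag to play cooldown or ready animations, and activation tags to play used feedback.
# A self-rescheduling half-second timer sets the icon material's IconTimer from the active cooldown effect with the longest remaining duration.
# Destruct ends async waits and unbinds the state delegate.
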WidgetBlueprint /ToyFoxGameplay/UI/Widgets/HUD/WBP_ToyFox_QuickbarSlot
animations [
 Anim_Cooldown {
  binding_count=1
  length_seconds=0.5
  object=&self:Anim_Cooldown
  sequence {bindings [Image_Icon {
     id=#1
     possessed_class=Image
     tracks [
      {
       type=MovieSceneFloatTrack
       display_name=Render Opacity
       sections [{type=MovieSceneFloatSection, end_frame=10, start_frame=0, channels [None {type=MovieSceneScriptingFloatChannel, const [1,1]}]}]
      }
      {
       type=MovieScene2DTransformTrack
       sections [{
         type=MovieScene2DTransformSection
         end_frame=10
         start_frame=0
         channels [
          Scale.X {type=MovieSceneScriptingFloatChannel, const [1,1]}
          Scale.Y {type=MovieSceneScriptingFloatChannel, const [1,1]}
         ]
        }]
      }
      {
       type=MovieSceneWidgetMaterialTrack
       sections [{
         type=MovieSceneParameterSection
         channels [
          Icon Size {type=MovieSceneScriptingFloatChannel, const [0.8,1]}
          IconTimer {type=MovieSceneScriptingFloatChannel, const [0,2]}
          HighlightSwitch {type=MovieSceneScriptingFloatChannel, keys [[0,1],[7,0]]}
         ]
        }]
      }
     ]
    }]}
 }
 Anim_ReadyForUse {
  binding_count=2
  length_seconds=0.2833
  object=&self:Anim_ReadyForUse
  sequence {
   bindings [
    Image_Glow {
     id=#2
     possessed_class=Image
     tracks [
      {
       type=MovieSceneFloatTrack
       display_name=Render Opacity
       sections [{type=MovieSceneFloatSection, channels [None {type=MovieSceneScriptingFloatChannel, keys [[0,1],[17,0]]}]}]
      }
      {type=MovieSceneWidgetMaterialTrack, sections []}
      {
       type=MovieScene2DTransformTrack
       sections [{
         type=MovieScene2DTransformSection
         end_frame=6
         start_frame=0
         channels [
          Translation.X {type=MovieSceneScriptingFloatChannel, keys [[0,-26],[6,-27]]}
          Translation.Y {type=MovieSceneScriptingFloatChannel, keys [[0,93.579],[6,-30]]}
          Angle {type=MovieSceneScriptingFloatChannel, const [6,1]}
          Scale.X {type=MovieSceneScriptingFloatChannel, const [0.55,1]}
          Scale.Y {type=MovieSceneScriptingFloatChannel, const [1,1]}
          Shear.X {type=MovieSceneScriptingFloatChannel, const [8,1]}
          Shear.Y {type=MovieSceneScriptingFloatChannel, const [0,1]}
         ]
        }]
      }
     ]
    }
    Image_Icon {
     id=#3
     possessed_class=Image
     tracks [{
       type=MovieSceneWidgetMaterialTrack
       sections [{
         type=MovieSceneParameterSection
         channels [
          Icon Size {type=MovieSceneScriptingFloatChannel, keys [[0,0.8],[6.667,1],[15.467,0.8]]}
          IconTimer {type=MovieSceneScriptingFloatChannel, const [1,1]}
          HighlightSwitch {type=MovieSceneScriptingFloatChannel, default=0, keys [[14,0],[17,1]]}
          Outline Color Change {type=MovieSceneScriptingFloatChannel, const [0,2]}
          Fill Noise {type=MovieSceneScriptingFloatChannel, const [0,2]}
          Icon Flash Switch {type=MovieSceneScriptingFloatChannel, keys [[0,1],[1,0],[2,1],[3,0],[4,1],[5,0],[6,1],[7,0],[8,1],[9,0]]}
          Glitch In {type=MovieSceneScriptingFloatChannel, keys [[0,1],[6,0]]}
          Shake {type=MovieSceneScriptingFloatChannel, keys [[0,0],[6,1]]}
          Speed Up {type=MovieSceneScriptingFloatChannel, keys [[0,1],[6,0]]}
         ]
        }]
      }]
    }
   ]
   display_rate {numerator=60}
  }
 }
 Anim_MeleeUsedFeedback {
  binding_count=1
  length_seconds=0.2667
  object=&self:Anim_MeleeUsedFeedback
  sequence {
   bindings [Image_Icon {
     id=#4
     possessed_class=Image
     tracks [
      {
       type=MovieScene2DTransformTrack
       sections [{
         type=MovieScene2DTransformSection
         channels [
          Translation.X {type=MovieSceneScriptingFloatChannel, const [0,1]}
          Translation.Y {type=MovieSceneScriptingFloatChannel, const [0,1]}
          Angle {type=MovieSceneScriptingFloatChannel, const [0,1]}
          Scale.X {type=MovieSceneScriptingFloatChannel, const [1,1]}
          Scale.Y {type=MovieSceneScriptingFloatChannel, const [1,1]}
          Shear.X {type=MovieSceneScriptingFloatChannel, const [0,1]}
          Shear.Y {type=MovieSceneScriptingFloatChannel, const [0,1]}
         ]
        }]
      }
      {
       type=MovieSceneWidgetMaterialTrack
       sections [{
         type=MovieSceneParameterSection
         channels [
          Icon Size {type=MovieSceneScriptingFloatChannel, const [0.8,3]}
          HighlightSwitch {type=MovieSceneScriptingFloatChannel, const [1,1]}
         ]
        }]
      }
     ]
    }]
   display_rate {numerator=60}
  }
 }
]
graphs [
 EventGraph {nodes [
   n1 = HandleCooldownTagEnabled!() -> (OutputDelegate, then, bEnabled:bool=~) {is=title K2Node_CustomEvent}
   > n2 = Branch(Condition=@n1.bEnabled)
   > PlayAnimationForward(InAnimation=Get Anim_Cooldown() -> (Anim_Cooldown:Widget Animation Object Reference))
   PlayAnimationForward(execute=@n2.else, InAnimation=Get Anim_ReadyForUse() -> (Anim_ReadyForUse:Widget Animation Object Reference))
   n3 = StartUpdatingCooldown() {is=title K2Node_CustomEvent}
   > n4 = Sequence()
   > n5 = Get!(execute:Exec) -> (then:Exec, else:Exec, ViewTargetPlayerPawn:Fort Player Pawn Object Reference)
   > n6 = GetActiveEffectsWithAllTags!(
    execute
    self:Ability System Component Object Reference=GetFortAbilitySystemComponent!(Actor:Actor Object Reference=@n5) -> (ReturnValue:Fort Ability System Component Object Reference)
    Tags:Gameplay Tag Container Structure=MakeGameplayTagContainerFromTag!(SingleTag:Gameplay Tag Structure=Get CooldownGameplayTag() -> (CooldownGameplayTag:Gameplay Tag Structure)) -> (ReturnValue:Gameplay Tag Container Structure)
   ) -> (then, ReturnValue:Array of Active Gameplay Effect Handle Structures)
   > n7 = GetCooldownDurationInfo!(execute, self, Handles:Array of Active Gameplay Effect Handle Structures=@n6) -> (then, bValid:bool, DurationRemaining:double, TotalDuration:double)
   > n8 = Branch(Condition=@n7)
   > n9 = GetDynamicMaterial(self=Get Image_Icon() -> (Image_Icon:Image Object Reference))
   > SetScalarParameterValue(self=@n9, ParameterName=IconTimer, Value=float - float(A=1, B=SafeDivide(A=@n7.DurationRemaining, B=@n7.TotalDuration)))
   n10 = GetDynamicMaterial(execute=@n8.else, self=Get Image_Icon())
   > SetScalarParameterValue(self=@n10, ParameterName=IconTimer, Value=1)
   n11 = Set Timer by Event(execute=@n4.then_1, Delegate=@n3, Time=0.5, bLooping=false)
   > Set UpdateCooldownSlider(UpdateCooldownSlider:Timer Handle Structure=@n11) -> (Output_Get:Timer Handle Structure)
   n12 = OnEventStateChanged!() -> (OutputDelegate, then, EventState:ToyFoxEventStateEnum Enum=NewEnumerator0) {is=title K2Node_CustomEvent}
   > Branch(Condition="Equal (Enum)"(A:ToyFoxEventStateEnum Enum=@n12.EventState, B:ToyFoxEventStateEnum Enum=NewEnumerator3))
   > InitializeForPlayer()
   InitializeForPlayer() {is=title K2Node_CustomEvent}
   > n13 = Cast To FortPlayerPawn(Object=Get Controlled Pawn!(self:Controller Object Reference=GetOwningPlayer()) -> (ReturnValue:Pawn Object Reference) {is=title K2Node_CallFunction})
   > n14 = Set ViewTargetPlayerPawn(ViewTargetPlayerPawn:Fort Player Pawn Object Reference=@n13) -> (Output_Get:Fort Player Pawn Object Reference)
   > n15 = Sequence()
   n16 = Get CooldownGameplayTag()
   > Branch(Condition=IsGameplayTagValid!(GameplayTag:Gameplay Tag Structure=@n16) -> (ReturnValue:bool))
   > n17 = WaitGameplayTagCountChangedOnActor!(execute:Exec, TargetActor:Actor Object Reference=@n14, Tag:Gameplay Tag Structure=@n16) -> (then:Exec, AsyncTaskProxy:Ability Async Wait Gameplay Tag Count Changed Object Reference, TagCountChanged:Exec, TagCount:int) {is=title K2Node_AsyncAction}
   > n18 = Get!(execute) -> (then, else, AsyncAction_Cooldown:Ability Async Wait Gameplay Tag Count Changed Object Reference)
   > n19 = EndAction(self:Ability Async Object Reference=@n18)
   Set AsyncAction_Cooldown(execute=[@n19.then, @n18.else], AsyncAction_Cooldown:Ability Async Wait Gameplay Tag Count Changed Object Reference=@n17) -> (Output_Get:Ability Async Wait Gameplay Tag Count Changed Object Reference)
   > n20 = HasMatchingGameplayTag!(execute:Exec, self:Object Reference=@n14, TagToCheck:Gameplay Tag Structure=Get CooldownGameplayTag()) -> (then:Exec, ReturnValue:bool) {is=title K2Node_Message}
   > HandleCooldownTagEnabled!(execute, self, bEnabled:bool=@n20) -> (then) {is=title K2Node_CallFunction}
   > StartUpdatingCooldown() {is=title K2Node_CallFunction}
   HandleCooldownTagEnabled(execute=@n17.TagCountChanged, bEnabled="integer > integer"(A=@n17.TagCount))
   n21 = For Each Loop(
    Exec=@n15.then_1
    Array:Array of Gameplay Tag Structures=Get AbilityActivationTags() -> (AbilityActivationTags:Array of Gameplay Tag Structures)
   ) -> (Array Element:Gameplay Tag Structure)
   > Branch(Condition=IsGameplayTagValid(GameplayTag=@n21))
   > n22 = WaitGameplayTagAddToActor!(
    execute:Exec
    TargetActor:Actor Object Reference=Get ViewTargetPlayerPawn() -> (ViewTargetPlayerPawn:Fort Player Pawn Object Reference)
    Tag:Gameplay Tag Structure=@n21
    OnlyTriggerOnce:bool=false
   ) -> (then:Exec, AsyncTaskProxy:Ability Async Wait Gameplay Tag Added Object Reference, Added:Exec) {is=title K2Node_AsyncAction}
   > Add(
    TargetArray:Array of Ability Async Wait Gameplay Tag Added Object References=Get AsyncAction_AbilityList() -> (AsyncAction_AbilityList:Array of Ability Async Wait Gameplay Tag Added Object References)
    NewItem:"Ability Async Wait Gameplay Tag Added Object Reference (by ref)"=@n22
   ) {is=title K2Node_CallArrayFunction}
   PlayAnimation(execute=@n22.Added, InAnimation=Get Anim_MeleeUsedFeedback() -> (Anim_MeleeUsedFeedback:Widget Animation Object Reference))
   Event Destruct()
   > n23 = Sequence!(execute) -> (then_0, then_1, then_2:Exec)
   > n24 = Get()
   > EndAction(self=@n24)
   n25 = Get!(execute=@n23.then_1) -> (then, else, ToyFoxGameStateComponent:B Toy Fox Game State Component Object Reference)
   > Unbind Event from On Event State Changed(self:B Toy Fox Game State Component Object Reference=@n25, Delegate=Create Event()) {is=title K2Node_RemoveDelegate}
   EndAction(execute=@n23.then_2, self=Get AsyncAction_AbilityList())
   Event Construct()
   > InitAbilityIcon()
   > InitKeybindActions()
   > n26 = Set ToyFoxGameStateComponent(ToyFoxGameStateComponent:B Toy Fox Game State Component Object Reference=GetComponentByClass!(
     self:Actor Object Reference=GetGameStateAthena!() -> (ReturnValue:Fort Game State Athena Object Reference)
     ComponentClass:Actor Component Class Reference=/ToyFoxGameplay/Gameplay/B_ToyFoxGameStateComponent.B_ToyFoxGameStateComponent_C
    ) -> (ReturnValue:B Toy Fox Game State Component Object Reference)) -> (Output_Get:B Toy Fox Game State Component Object Reference)
   > n27 = Branch(Condition="Equal (Enum)"(
     A=GetEventState!(self:B Toy Fox Game State Component Object Reference=@n26) -> (State:ToyFoxEventStateEnum Enum=NewEnumerator0)
     B=NewEnumerator3
    ))
   > InitializeForPlayer() {is=title K2Node_CallFunction}
   Bind Event to On Event State Changed(
    execute=@n27.else
    self:B Toy Fox Game State Component Object Reference=Get ToyFoxGameStateComponent() -> (ToyFoxGameStateComponent:B Toy Fox Game State Component Object Reference)
    Delegate=Create Event()
   ) {is=title K2Node_AddDelegate}
   n28 = Event PreConstruct()
   > Parent: PreConstruct(IsDesignTime=@n28.IsDesignTime)
   > Branch(Condition=@n28.IsDesignTime)
   > InitAbilityIcon()
   > InitKeybindActions()
  ]}
 GetCooldownDurationInfo {nodes [
   n1 = GetCooldownDurationInfo!() -> (then, Handles:Array of Active Gameplay Effect Handle Structures) {is=title K2Node_FunctionEntry}
   > Set LargestCooldownIndex(LargestCooldownIndex:int=-1) -> (Output_Get:int=0)
   > n2 = For Each Loop(Array:Array of Active Gameplay Effect Handle Structures=@n1) -> (Array Element:Active Gameplay Effect Handle Structure)
   n3 = Get LargestCooldownIndex() -> (LargestCooldownIndex:int=0)
   n4 = Branch(
    execute=@n2.Completed
    Condition=Is Valid Index!(TargetArray:Array of Active Gameplay Effect Handle Structures=@n1, IndexToTest:int=@n3) -> (ReturnValue:bool)
   )
   n5 = "Get (a ref)"(Array:Array of Active Gameplay Effect Handle Structures=@n1, Dimension 1=@n3) -> (Output:"Active Gameplay Effect Handle Structure (by ref)")
   > n6 = GetActiveGameplayEffectTotalDuration!(execute, ActiveHandle:Active Gameplay Effect Handle Structure=@n5) -> (then, ReturnValue:float)
   > n7 = GetActiveGameplayEffectRemainingDuration!(execute, ActiveHandle:Active Gameplay Effect Handle Structure=@n5) -> (then, ReturnValue:float)
   > Return Node!(execute, bValid:bool=true, DurationRemaining:double=@n7, TotalDuration:double=@n6)
   Return Node(execute=@n4.else, bValid=false, DurationRemaining=0, TotalDuration=0)
   n8 = GetActiveGameplayEffectRemainingDuration(execute=@n2.LoopBody, ActiveHandle=@n2)
   n9 = Get LargestCooldownDuration() -> (LargestCooldownDuration:double)
   > Branch(Condition=AND Boolean(A="float > float"(A=@n8), B="float > float"(A=@n8, B=@n9)))
   > Set LargestCooldownIndex(LargestCooldownIndex=@n2.Array Index) -> (Output_Get=0)
   > Set LargestCooldownDuration(LargestCooldownDuration:double=@n8) -> (Output_Get:double)
  ]}
 InitAbilityIcon {nodes [
   InitAbilityIcon() {is=title K2Node_FunctionEntry}
   > SetBrushFromMaterial(self=Get Image_Icon(), Material=Get AbilityIcon() -> (AbilityIcon:Material Interface Object Reference))
  ]}
 InitKeybindActions {nodes [
   InitKeybindActions() {is=title K2Node_FunctionEntry}
   > SetBoundAction!(
    execute
    self:Fort Keybind Widget Object Reference=Get Keybind() -> (Keybind:Keybind Widget Object Reference)
    NewBoundAction:Name=Get InputBoundAction() -> (InputBoundAction:Name=None)
   ) -> (then)
  ]}
]
variables [
 AbilityActivationTags {type="array<GameplayTag>", default []}
 AbilityIcon {type=MaterialInterface, default=/TidalCrane_ToyFox_Boss/UI/M_ToyFox_ZordAbilityIcons}
 AsyncAction_AbilityList {type="array<AbilityAsync_WaitGameplayTagAdded>", default []}
 AsyncAction_Cooldown {type=AbilityAsync_WaitGameplayTagCountChanged, default=null}
 bShowCooldownSlider {type=boolean}
 CooldownGameplayTag {type=GameplayTag, default {tagName=None}}
 InputBoundAction {type=string, default=None}
 ToyFoxGameStateComponent {type=B_ToyFoxGameStateComponent_C, default=null}
 UpdateCooldownSlider {type=TimerHandle}
 ViewTargetPlayerPawn {type=FortPlayerPawn}
]
widgets {
 info {inheritedWidgetCount=1, parentClass=&FortHUDElementWidget, rootWidgetClass=&SizeBox, widgetCount=8}
 widgets [
  SizeBox_14 {
   widgetClassPath=&SizeBox
   widget=SizeBox_14
   properties {bOverride_HeightOverride=true, bOverride_WidthOverride=true, heightOverride=200, widthOverride=200}
  }
  Overlay_20 {widgetClassPath=&Overlay, parent=SizeBox_14, widget=Overlay_20, slot {class=SizeBoxSlot}}
  Overlay_42 {widgetClassPath=&Overlay, parent=Overlay_20, widget=Overlay_42, slot {class=OverlaySlot}}
  Image_Icon {
   widgetClassPath=&Image
   bIsVariable=true
   parent=Overlay_42
   widget=Image_Icon
   properties {
    visibility=SelfHitTestInvisible
    brush {
     bIsDynamicallyLoaded=false
     drawAs=Image
     imageType=FullColor
     mirroring=NoMirror
     resourceName=None
     resourceObject=&M_ToyFox_ZordAbilityIcons
     tiling=NoTile
     imageSize=<200 200>
     margin=<ltrb 0 0 0 0>
     outlineSettings {
      bUseBrushTransparency=false
      roundingType=HalfHeightRadius
      width=0
      color {colorUseRule=UseColor_Specified, specifiedColor=<rgba 0 0 0 0>}
      cornerRadii=<0 0 0 0>
     }
     tintColor {colorUseRule=UseColor_Specified, specifiedColor=<rgba 1 1 1 1>}
     uVRegion {bIsValid=false, max=<0 0>, min=<0 0>}
    }
   }
   slot {class=OverlaySlot, properties {horizontalAlignment=HAlign_Fill, verticalAlignment=VAlign_Fill}}
  }
  Image_Glow {
   widgetClassPath=&Image
   bIsVariable=true
   parent=Overlay_42
   widget=Image_Glow
   properties {
    renderOpacity=0
    brush {
     bIsDynamicallyLoaded=false
     drawAs=Image
     imageType=FullColor
     mirroring=NoMirror
     resourceName=None
     resourceObject=/ToyFoxGameplay/UI/Widgets/HUD/Assets/BossBar/M_UI_HardEdgedGlow
     tiling=NoTile
     imageSize=<100 100>
     margin=<ltrb 0 0 0 0>
     outlineSettings {
      bUseBrushTransparency=false
      roundingType=HalfHeightRadius
      width=0
      color {colorUseRule=UseColor_Specified, specifiedColor=<rgba 0 0 0 0>}
      cornerRadii=<0 0 0 0>
     }
     tintColor {colorUseRule=UseColor_Specified, specifiedColor=<rgba 1 1 1 1>}
     uVRegion {bIsValid=false, max=<0 0>, min=<0 0>}
    }
    renderTransform {angle=6, scale=<0.55 1>, shear=<8 0>, translation=<-26 41>}
   }
   slot {class=OverlaySlot, properties {horizontalAlignment=HAlign_Fill, verticalAlignment=VAlign_Fill}}
  }
  SizeBox {
   widgetClassPath=&SizeBox
   parent=Overlay_20
   widget=SizeBox
   properties {renderTransform {angle=6, scale=<1 1>, shear=<-23 0>, translation=<93 -22>}}
   slot {class=OverlaySlot, properties {verticalAlignment=VAlign_Bottom}}
  }
  Keybind {
   widgetClassPath=/FortUILibrary/Input/KeybindWidget
   bIsVariable=true
   parent=SizeBox
   widget=Keybind
   properties {bUseKeyboardTextureForKeys=true, foregroundColor {colorUseRule=UseColor_Specified, specifiedColor=<rgba 1 1 1 1>}}
   slot {class=SizeBoxSlot, properties {horizontalAlignment=HAlign_Left, verticalAlignment=VAlign_Center}}
  }
  TutorialHighlight {widgetClassPath=&TutorialHighlight}
 ]
}
