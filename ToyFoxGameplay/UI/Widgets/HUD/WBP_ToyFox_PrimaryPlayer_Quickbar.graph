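# Primary player HUD quickbar for ToyFox, showing Scan, Cannon, Stomp and Melee ability slots inside a tag-visibility root.
# Each QuickbarSlot instance is configured with an input action, with cooldown tags and sliders on the first three and activation tags on Melee.
# Slots step diagonally over a mirrored border, and an Intro animation scales and fades them in staggered from Melee to Scan.
# The event graph holds only empty event stubs.
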WidgetBlueprint /ToyFoxGameplay/UI/Widgets/HUD/WBP_ToyFox_PrimaryPlayer_Quickbar
animations [Intro {
  binding_count=5
  length_seconds=0.4333
  object=&self:Intro
  sequence {
   bindings [
    QuickbarSlot_Melee {
     id=#1
     possessed_class=WBP_ToyFox_QuickbarSlot_C
     tracks [
      {
       type=MovieSceneFloatTrack
       display_name=Render Opacity
       sections [{type=MovieSceneFloatSection, channels [None {type=MovieSceneScriptingFloatChannel, keys [[0,0],[11,1]]}]}]
      }
      {
       type=MovieScene2DTransformTrack
       sections [{
         type=MovieScene2DTransformSection
         channels [
          Translation.X {type=MovieSceneScriptingFloatChannel, const [0,1]}
          Translation.Y {type=MovieSceneScriptingFloatChannel, const [27,1]}
          Angle {type=MovieSceneScriptingFloatChannel, const [0,1]}
          Scale.X {type=MovieSceneScriptingFloatChannel, keys [[0,0],[11,1]]}
          Scale.Y {type=MovieSceneScriptingFloatChannel, keys [[0,0],[11,1]]}
          Shear.X {type=MovieSceneScriptingFloatChannel, const [0,1]}
          Shear.Y {type=MovieSceneScriptingFloatChannel, const [0,1]}
         ]
        }]
      }
     ]
    }
    Image_Border {
     id=#2
     possessed_class=Image
     tracks [{
       type=MovieSceneFloatTrack
       display_name=Render Opacity
       sections [{type=MovieSceneFloatSection, channels [None {type=MovieSceneScriptingFloatChannel, keys [[4,0],[25,1]]}]}]
      }]
    }
    QuickbarSlot_Cannon {
     id=#3
     possessed_class=WBP_ToyFox_QuickbarSlot_C
     tracks [
      {
       type=MovieSceneFloatTrack
       display_name=Render Opacity
       sections [{type=MovieSceneFloatSection, channels [None {type=MovieSceneScriptingFloatChannel, keys [[11,0],[22,1]]}]}]
      }
      {
       type=MovieScene2DTransformTrack
       sections [{
         type=MovieScene2DTransformSection
         channels [
          Translation.X {type=MovieSceneScriptingFloatChannel, const [0,1]}
          Translation.Y {type=MovieSceneScriptingFloatChannel, const [-14,1]}
          Angle {type=MovieSceneScriptingFloatChannel, const [0,1]}
          Scale.X {type=MovieSceneScriptingFloatChannel, keys [[11,0],[22,1]]}
          Scale.Y {type=MovieSceneScriptingFloatChannel, keys [[11,0],[22,1]]}
          Shear.X {type=MovieSceneScriptingFloatChannel, const [0,1]}
          Shear.Y {type=MovieSceneScriptingFloatChannel, const [0,1]}
         ]
        }]
      }
     ]
    }
    QuickbarSlot_Boot {
     id=#4
     possessed_class=WBP_ToyFox_QuickbarSlot_C
     tracks [
      {
       type=MovieSceneFloatTrack
       display_name=Render Opacity
       sections [{type=MovieSceneFloatSection, channels [None {type=MovieSceneScriptingFloatChannel, keys [[3,0],[14,1]]}]}]
      }
      {
       type=MovieScene2DTransformTrack
       sections [{
         type=MovieScene2DTransformSection
         channels [
          Translation.X {type=MovieSceneScriptingFloatChannel, const [0,1]}
          Translation.Y {type=MovieSceneScriptingFloatChannel, const [5,1]}
          Angle {type=MovieSceneScriptingFloatChannel, const [0,1]}
          Scale.X {type=MovieSceneScriptingFloatChannel, keys [[3,0],[14,1]]}
          Scale.Y {type=MovieSceneScriptingFloatChannel, keys [[3,0],[14,1]]}
          Shear.X {type=MovieSceneScriptingFloatChannel, const [0,1]}
          Shear.Y {type=MovieSceneScriptingFloatChannel, const [0,1]}
         ]
        }]
      }
     ]
    }
    QuickbarSlot_Scan {
     id=#5
     possessed_class=WBP_ToyFox_QuickbarSlot_C
     tracks [
      {
       type=MovieSceneFloatTrack
       display_name=Render Opacity
       sections [{type=MovieSceneFloatSection, channels [None {type=MovieSceneScriptingFloatChannel, keys [[16,0],[26,1]]}]}]
      }
      {
       type=MovieScene2DTransformTrack
       sections [{
         type=MovieScene2DTransformSection
         channels [
          Translation.X {type=MovieSceneScriptingFloatChannel, const [0,1]}
          Translation.Y {type=MovieSceneScriptingFloatChannel, const [-32,1]}
          Angle {type=MovieSceneScriptingFloatChannel, const [0,1]}
          Scale.X {type=MovieSceneScriptingFloatChannel, keys [[16,0],[26,1]]}
          Scale.Y {type=MovieSceneScriptingFloatChannel, keys [[16,0],[26,1]]}
          Shear.X {type=MovieSceneScriptingFloatChannel, const [0,1]}
          Shear.Y {type=MovieSceneScriptingFloatChannel, const [0,1]}
         ]
        }]
      }
     ]
    }
   ]
   display_rate {numerator=60}
  }
 }]
graphs [EventGraph {nodes [
   Event PreConstruct()
   Event Construct()
   Event Tick()
  ]}]
variables []
widgets {
 info {
  inheritedWidgetCount=1
  parentClass=&FortHUDElementWidget
  rootWidgetClass=/Script/FortniteUI.CommonTagVisibilityWidget
  widgetCount=10
 }
 widgets [
  CommonTagVisibilityWidget {widgetClassPath=&CommonTagVisibilityWidget, bIsVariable=true, widget=CommonTagVisibilityWidget}
  SizeBox_0 {
   widgetClassPath=&SizeBox
   parent=CommonTagVisibilityWidget
   widget=SizeBox_0
   properties {bOverride_WidthOverride=true, widthOverride=580}
   slot {class=PanelSlot}
  }
  Overlay_16 {widgetClassPath=&Overlay, parent=SizeBox_0, widget=Overlay_16, slot {class=SizeBoxSlot}}
  Image_Border {
   widgetClassPath=&Image
   bIsVariable=true
   parent=Overlay_16
   widget=Image_Border
   properties {
    brush {
     bIsDynamicallyLoaded=false
     drawAs=Image
     imageType=FullColor
     mirroring=NoMirror
     resourceName=None
     resourceObject=/ToyFoxGameplay/UI/Materials/M_ToyFox_HealthTopBar
     tiling=NoTile
     imageSize=<128 128>
     margin=<ltrb 0 0 0 0>
     outlineSettings {
      bUseBrushTransparency=false
      roundingType=HalfHeightRadius
      width=0
      color {colorUseRule=UseColor_Specified, specifiedColor=<rgba 0 0 0 0>}
      cornerRadii=<0 0 0 0>
     }
     tintColor {colorUseRule=UseColor_Specified, specifiedColor=<rgba 1 1 1 1>}
     uVRegion {bIsValid=false, max=<0 0>, min=<0 0>}
    }
    renderTransform {angle=0, scale=<-1 1>, shear=<0 0>, translation=<26 90>}
   }
   slot {class=OverlaySlot, properties {horizontalAlignment=HAlign_Fill, verticalAlignment=VAlign_Fill}}
  }
  GridPanel {
   widgetClassPath=&HorizontalBox
   parent=Overlay_16
   widget=GridPanel
   properties {bOverride_Cursor=true, renderTransform {angle=0, scale=<0.8 0.8>, shear=<0 0>, translation=<-87 8>}}
   slot {class=OverlaySlot, properties {horizontalAlignment=HAlign_Center, verticalAlignment=VAlign_Center}}
  }
  QuickbarSlot_Scan {
   widgetClassPath=&/WBP_ToyFox_QuickbarSlot
   bIsVariable=true
   parent=GridPanel
   widget=QuickbarSlot_Scan
   properties {
    abilityIcon=/ToyFoxGameplay/UI/Materials/MI_ToyFox_ZordAbilityIcons_Scan
    bShowCooldownSlider=true
    inputBoundAction=Reload
    cooldownGameplayTag {tagName=ToyFox.Ability.Cooldown.Scan}
    renderTransform {angle=0, scale=<1 1>, shear=<0 0>, translation=<0 -30>}
   }
   slot {class=HorizontalBoxSlot, properties {padding=<ltrb 0 0 -35 0>}}
  }
  QuickbarSlot_Cannon {
   widgetClassPath=&WBP_ToyFox_QuickbarSlot
   bIsVariable=true
   parent=GridPanel
   widget=QuickbarSlot_Cannon
   properties {
    abilityIcon=/ToyFoxGameplay/UI/Materials/MI_ToyFox_ZordAbilityIcons_Cannon
    bShowCooldownSlider=true
    inputBoundAction=Target
    cooldownGameplayTag {tagName=ToyFox.Ability.Cooldown.Cannon}
    renderTransform {angle=0, scale=<1 1>, shear=<0 0>, translation=<0 -11>}
   }
   slot {class=HorizontalBoxSlot, properties {padding=<ltrb 0 0 -35 0>}}
  }
  QuickbarSlot_Boot {
   widgetClassPath=&WBP_ToyFox_QuickbarSlot
   bIsVariable=true
   parent=GridPanel
   widget=QuickbarSlot_Boot
   properties {
    abilityIcon=/ToyFoxGameplay/UI/Materials/MI_ToyFox_ZordAbilityIcons_Stomp
    bShowCooldownSlider=true
    inputBoundAction=Jump
    cooldownGameplayTag {tagName=ToyFox.Ability.Cooldown.Stomp}
    renderTransform {angle=0, scale=<1 1>, shear=<0 0>, translation=<0 6>}
   }
   slot {class=HorizontalBoxSlot, properties {padding=<ltrb 0 0 -35 0>}}
  }
  QuickbarSlot_Melee {
   widgetClassPath=&WBP_ToyFox_QuickbarSlot
   bIsVariable=true
   parent=GridPanel
   widget=QuickbarSlot_Melee
   properties {
    inputBoundAction=Fire
    abilityActivationTags [{tagName=ToyFox.Ability.Melee.Active}, {tagName=Gameplay.Melee.Combo2}]
    renderTransform {angle=0, scale=<1 1>, shear=<0 0>, translation=<0 25>}
   }
   slot {class=HorizontalBoxSlot}
  }
  TutorialHighlight {widgetClassPath=&TutorialHighlight}
 ]
}
