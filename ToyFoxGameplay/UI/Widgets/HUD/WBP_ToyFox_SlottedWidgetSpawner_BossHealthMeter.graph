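# Widget Blueprint slotted-widget spawner for the ToyFox HUD, named for the boss health meter.
# It derives from FortSlottedWidgetSpawner and defines no widget tree or variables of its own.
# On Construct it calls SpawnSlottedWidgets on itself, leaving what gets spawned and where to the parent class.
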WidgetBlueprint /ToyFoxGameplay/UI/Widgets/HUD/WBP_ToyFox_SlottedWidgetSpawner_BossHealthMeter
graphs [EventGraph {nodes [
   Event Construct()
   > SpawnSlottedWidgets(self:Fort Slotted Widget Spawner Object Reference)
  ]}]
variables []
widgets {info {inheritedWidgetCount=0, parentClass=/Script/FortniteUI.FortSlottedWidgetSpawner, rootWidgetClass=null, widgetCount=0}, widgets []}
