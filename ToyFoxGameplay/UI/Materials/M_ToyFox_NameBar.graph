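# UI material for the ToyFox name bar that draws a translucent red-orange shape from a border texture and can flash toward white.
# Scaled-to-fit UVs, stretched and shifted vertically, sample an MSDF texture that is converted to SDF and antialiased into a mask.
# The mask drives dimmed color and low opacity under alpha-composite blending.
# Flash blends the color to white, and Master Opacity scales both outputs.
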Material /ToyFoxGameplay/UI/Materials/M_ToyFox_NameBar
graph {
 properties {bAutomaticallySetUsageInEditor=false, blendMode=BLEND_AlphaComposite, materialDomain=MD_UI, usedShadingModels=Default Lit}
 nodes [
  n1 = LinearInterpolate(A=Constant3Vector() {constant=<rgba 0.4062 0.0039 0 1>}, Alpha=ScalarParameter() {parameterName=Flash})
  n2 = MaterialFunctionCall(UV=Add(A=TextureCoordinate() {vTiling=3}, B=Constant2Vector() {g=-1})) {materialFunction=&MF_UI_UVScaling}
  n3 = MaterialFunctionCall(
   DistanceField=MaterialFunctionCall!(MSDF RGB=TextureSample(UVs=@n2.Scaled to Fit) {samplerType=SAMPLERTYPE_LinearColor, texture=/ToyFoxGameplay/UI/Materials/Textures/T_UI_ToyFox_Border2}.RGB) {materialFunction=/FortUILibrary/MaterialFunctions/Shapes/Modifiers/MF_UI_MSDFtoSDF}.SDF
   Threshold=Constant() {r=0.43}
   Smoothing=Constant()
  ) {materialFunction=&MF_UI_AntiAliasing}
  n4 = ScalarParameter() {defaultValue=1, parameterName=Master Opacity}
  Multiply(A=Multiply(A=Multiply(A=@n1, B=@n3.Outer), B=Constant() {r=0.4}), B=@n4)
  Multiply(A=Multiply(A=@n3.Outer) {constB=0.2}, B=@n4)
 ]
}
