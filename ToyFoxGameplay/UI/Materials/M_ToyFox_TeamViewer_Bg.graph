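# Translucent UI material for the ToyFox team viewer background, drawing a tinted panel that fades toward the right.
# An anti-aliased SDF box, fed by fill-scaled UVs through a skew function, forms the panel shape.
# That shape is multiplied by a horizontal gradient with adjustable falloff.
# The resulting mask is scaled by a Blend parameter and, separately, by Opacity and the dark red default Color.
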Material /ToyFoxGameplay/UI/Materials/M_ToyFox_TeamViewer_Bg
graph {
 properties {bAutomaticallySetUsageInEditor=false, blendMode=BLEND_Translucent, materialDomain=MD_UI, usedShadingModels=Default Lit}
 nodes [
  n1 = MaterialFunctionCall!(
   UV=Add(A=MaterialFunctionCall() {materialFunction=&MF_UI_UVScaling}.Scaled To Fill, B=Constant2Vector())
   Skew Angle=Constant2Vector()
  ) {materialFunction=/FortUILibrary/MaterialFunctions/Basics/MF_UI_Skew}
  n2 = Constant2Vector() {g=1, r=0.95}
  n3 = Multiply(
   A=MaterialFunctionCall(DistanceField=MaterialFunctionCall!(UVs=@n1.Skewed UVs, Size=@n2, CornerRoundness, KeepAspectRatio, PixelSize) {materialFunction=/FortUILibrary/MaterialFunctions/Shapes/SDFs/MF_SDF_Box}.SDF) {materialFunction=&MF_UI_AntiAliasing}.Inner
   B=NamedRerouteUsage()
  )
  Multiply(A=@n3, B=ScalarParameter() {defaultValue=0.5, parameterName=Blend}) {constB=0.5}
  Multiply(
   A=VectorParameter() {parameterName=Color, defaultValue=<rgba 0.0339 0 0 1>}.RGB
   B=Multiply(A=ScalarParameter() {defaultValue=1, parameterName=Opacity}, B=@n3)
  )
  n4 = OneMinus(None=ComponentMask(None=TextureCoordinate()) {r=true})
  NamedRerouteDeclaration(None=Power(Base=Multiply(A=@n4) {constB=2}, Exp=ScalarParameter() {defaultValue=1, parameterName=Fall Off}) {constExponent=1.2}) {name=Right Gradient Mask, nodeColor=<rgba 0 0 0 1>}
 ]
}
