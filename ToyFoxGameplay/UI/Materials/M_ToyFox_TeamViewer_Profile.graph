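# UI material for the ToyFox team viewer profile, drawing a red hexagonal badge with alpha-composite blending.
# Rotated SDF polygons and an outline form antialiased masks, shared through named reroutes to build color and opacity.
# Scalar parameters flash the red fill and outline toward neutral tones, switch to a gradient-only look, and scale opacity.
# A "Who This?" parameter overlays a masked Megazord profile texture, using its thresholded alpha for opacity.
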Material /ToyFoxGameplay/UI/Materials/M_ToyFox_TeamViewer_Profile
graph {
 properties {bAutomaticallySetUsageInEditor=false, blendMode=BLEND_AlphaComposite, materialDomain=MD_UI, usedShadingModels=Default Lit}
 nodes [
  n1 = NamedRerouteUsage()
  n2 = LinearInterpolate(
   A=Multiply(A=Constant3Vector() {constant=<rgba 0.005 0 0 1>}, B=@n1)
   B=Multiply(A=@n1) {constB=0.5}
   Alpha=ScalarParameter() {parameterName=Flash}
  )
  n3 = Constant3Vector() {constant=<rgba 0.6867 0.0369 0.0452 1>}
  n4 = NamedRerouteUsage()
  n5 = Constant3Vector() {constant=<rgba 0.1683 0.0252 0.0262 1>}
  n6 = MaterialFunctionCall(UVs=NamedRerouteUsage(), "Rotation Angle (0-1)"=Constant() {r=0.1}) {materialFunction=&CustomRotator}
  n7 = LinearInterpolate(
   A=LinearInterpolate(A=Multiply(A=@n3, B=@n4), B=Multiply(A=@n5, B=@n4), Alpha=ComponentMask(None=@n6.Rotated Values) {g=true})
   B=Multiply(A=@n4) {constB=0.05}
   Alpha=ScalarParameter() {parameterName=Flash}
  ) {constB=0}
  n8 = NamedRerouteUsage()
  n9 = Constant3Vector() {constant=<rgba 0.6867 0.0369 0.0452 1>}
  n10 = NamedRerouteUsage()
  n11 = ScalarParameter() {defaultValue=0.6, parameterName=Gradient Master Opacity}
  n12 = LinearInterpolate(
   A=Multiply(A=LinearInterpolate(A=@n2, B=@n7, Alpha=@n8), B=ScalarParameter() {defaultValue=1, parameterName=Master Opacity})
   B=Multiply(A=Multiply(A=@n9, B=@n10), B=LinearInterpolate(Alpha=@n11)) {constB=0.6}
   Alpha=ScalarParameter() {parameterName=Gradient Switcher}
  )
  LinearInterpolate(A=@n12, B=LinearInterpolate(A=@n12, B=NamedRerouteUsage(), Alpha=NamedRerouteUsage()), Alpha=ScalarParameter() {parameterName=Who This?})
  n13 = Multiply(A=NamedRerouteUsage()) {constB=0.6}
  n14 = Multiply(A=NamedRerouteUsage()) {constB=0.2}
  n15 = Multiply(A=NamedRerouteUsage(), B=Power(Base=NamedRerouteUsage()) {constExponent=3})
  n16 = LinearInterpolate(
   A=Multiply(A=Add(A=@n13, B=@n14), B=ScalarParameter() {defaultValue=1, parameterName=Master Opacity})
   B=Multiply(A=@n15, B=ScalarParameter() {defaultValue=0.6, parameterName=Gradient Master Opacity}) {constB=0}
   Alpha=ScalarParameter() {parameterName=Gradient Switcher}
  )
  LinearInterpolate(A=@n16, B=Add(A=@n16, B=NamedRerouteUsage()), Alpha=ScalarParameter() {parameterName=Who This?})
  NamedRerouteDeclaration(None=ComponentMask(None=TextureCoordinate()) {g=true}) {name=G Mask, nodeColor=<rgba 0 0 0 1>}
  n17 = MaterialFunctionCall(UVs=NamedRerouteUsage(), "Rotation Angle (0-1)"=Constant() {r=0.25}) {materialFunction=&CustomRotator}
  n18 = Constant() {r=6}
  n19 = MaterialFunctionCall!(UVs=@n17.Rotated Values, Radius=Constant() {r=0.9}, Sides=@n18, InnerAngle) {materialFunction=/FortUILibrary/MaterialFunctions/Shapes/SDFs/MF_SDF_RegularPolygon}
  NamedRerouteDeclaration(None=MaterialFunctionCall(DistanceField=@n19.SDF) {materialFunction=&MF_UI_AntiAliasing}.Inner) {name=Polygon, nodeColor=<rgba 0 0 0 1>}
  NamedRerouteDeclaration(None=MaterialFunctionCall!(DistanceField=@n19.SDF, Threshold, Thickness=Constant() {r=0.035}, Alignment=Constant() {r=-3}, Smoothing) {materialFunction=/FortUILibrary/MaterialFunctions/Shapes/Modifiers/MF_UI_ShapeOutline}.Outline) {name=Outline, nodeColor=<rgba 0 0 0 1>}
  NamedRerouteDeclaration(None=MaterialFunctionCall() {materialFunction=&GetUserInterfaceUV}.Normalized UV) {name=Normalized UV, nodeColor=<rgba 0 0 0 1>}
  n20 = Constant() {r=0.82}
  NamedRerouteDeclaration(None=MaterialFunctionCall(DistanceField=MaterialFunctionCall(UVs=@n17.Rotated Values, Radius=@n20, Sides=@n18) {materialFunction=&MF_SDF_RegularPolygon}.SDF) {materialFunction=&MF_UI_AntiAliasing}.Inner) {name=Polygon Small, nodeColor=<rgba 0 0 0 1>}
  NamedRerouteDeclaration(None=@n15) {name=Polygon Gradient, nodeColor=<rgba 0 0 0 1>}
  n21 = TextureSample(UVs=MaterialFunctionCall(Texture Scale=Constant() {r=1.2}) {materialFunction=&ScaleUVsByCenter}.UVs) {texture=/ToyFoxGameplay/UI/Materials/Textures/T_UI_ToyFox_MegazordProfile}
  n22 = Multiply(A=Multiply(A=@n21.RGB, B=NamedRerouteUsage()), B=OneMinus(None=NamedRerouteUsage()))
  NamedRerouteDeclaration(None=@n22) {name=Who This?, nodeColor=<rgba 0 0.2266 1 1>}
  n23 = Constant() {r=0.7}
  n24 = Multiply(
   A=Multiply(
    A=MaterialFunctionCall(DistanceField=@n21.A, Threshold=@n23) {materialFunction=&MF_UI_AntiAliasing}.Outer
    B=NamedRerouteUsage()
   )
   B=OneMinus(None=NamedRerouteUsage())
  )
  NamedRerouteDeclaration(None=@n24) {name=Who This Alpha, nodeColor=<rgba 0.5469 1 0 1>}
  Saturate(None=@n22)
  Saturate(None=@n24)
 ]
}
