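# M_UI_ToyFox is an alpha-composite UI material that draws hexagon shapes with a switchable tint and a master opacity.
# Rotated six-sided SDF polygons on scalable, fitted UVs form three named masks: a soft glow, an outline and a filled hex.
# A center-peaked UV gradient shades red and grayscale color options, which ColorSwitcher blends before mask and Master Opacity multiply in.
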Material /ToyFoxGameplay/UI/Materials/M_UI_ToyFox
graph {
 properties {bAutomaticallySetUsageInEditor=false, blendMode=BLEND_AlphaComposite, materialDomain=MD_UI, usedShadingModels=Default Lit}
 nodes [
  n1 = NamedRerouteUsage()
  Add(A=@n1, B=NamedRerouteUsage())
  Constant3Vector() {constant=<rgba 1 0 0.0804 1>}
  Constant3Vector() {constant=<rgba 1 1 1 1>}
  Constant3Vector() {constant=<rgba 0 0.8514 1 1>}
  Constant3Vector() {constant=<rgba 0 1 0.107 1>}
  MaterialFunctionCall() {materialFunction=&3ColorBlend}
  MaterialFunctionCall() {materialFunction=&DebugTimeSine}
  MaterialFunctionCall() {materialFunction=&DebugTimeSine}
  n2 = MaterialFunctionCall() {materialFunction=&LinearGradient}
  n3 = Subtract(A=@n2.VGradient) {constB=0.5}
  n4 = OneMinus(None=Abs(None=Multiply(A=@n3) {constB=2}))
  n5 = Multiply(A=Subtract(A=@n2.UGradient) {constB=0.5}) {constB=2}
  n6 = Power(Base=OneMinus(None=Abs(None=@n5)))
  n7 = Multiply(A=Saturate(None=Power(Base=@n4)), B=Saturate(None=Multiply(A=@n6) {constB=30}))
  n8 = LinearInterpolate(A=Constant3Vector() {constant=<rgba 1 0 0.0804 1>}, B=Constant3Vector() {constant=<rgba 1 0 0 1>}, Alpha=@n7)
  n9 = LinearInterpolate(A=Constant(), B=Constant() {r=0.5}, Alpha=@n7)
  n10 = ScalarParameter() {parameterName=ColorSwitcher}
  n11 = ScalarParameter() {defaultValue=1, parameterName=Master Opacity}
  Multiply(A=Multiply(A=LinearInterpolate(A=@n8, B=@n9, Alpha=@n10), B=@n1), B=@n11)
  Multiply(A=@n1, B=@n11)
  n12 = MaterialFunctionCall(UVs=NamedRerouteUsage(), Texture Scale=LinearInterpolate(Alpha=ScalarParameter() {defaultValue=0.5, parameterName=Scale}) {constA=0.5}) {materialFunction=&ScaleUVsByCenter}
  n13 = AppendVector(A=LinearInterpolate(Alpha=Constant()) {constB=0.7}, B=Constant())
  n14 = MaterialFunctionCall!(
   UVs=MaterialFunctionCall(UVs=Add(A=@n12.UVs, B=@n13), "Rotation Angle (0-1)"=Constant() {r=0.25}) {materialFunction=&CustomRotator}.Rotated Values
   Radius=Constant() {r=0.6}
   Sides=Constant() {r=6}
   InnerAngle
  ) {materialFunction=/FortUILibrary/MaterialFunctions/Shapes/SDFs/MF_SDF_RegularPolygon}
  n15 = Constant() {r=0.2}
  n16 = Constant() {r=0.2}
  NamedRerouteDeclaration(None=Power(Base=MaterialFunctionCall(DistanceField=@n14.SDF, Threshold=@n15, Smoothing=@n16) {materialFunction=&MF_UI_AntiAliasing}.Inner) {constExponent=4}) {name=Hex Glow, nodeColor=<rgba 0 0 0 1>}
  n17 = ScalarParameter() {defaultValue=1, parameterName=X}
  n18 = Add(
   A=MaterialFunctionCall(UVs=NamedRerouteUsage(), Texture Scale=ScalarParameter() {defaultValue=0.8, parameterName=Outline Scale}) {materialFunction=&ScaleUVsByCenter}.UVs
   B=AppendVector(A=LinearInterpolate(Alpha=@n17) {constB=0}, B=Constant())
  )
  n19 = Constant() {r=0.25}
  n20 = MaterialFunctionCall!(
   DistanceField=MaterialFunctionCall(
    UVs=MaterialFunctionCall(UVs=@n18, "Rotation Angle (0-1)"=@n19) {materialFunction=&CustomRotator}.Rotated Values
    Radius=Constant() {r=0.5}
    Sides=Constant() {r=6}
   ) {materialFunction=&MF_SDF_RegularPolygon}.SDF
   Threshold
   Thickness
   Alignment
   Smoothing
  ) {materialFunction=/FortUILibrary/MaterialFunctions/Shapes/Modifiers/MF_UI_ShapeOutline}
  n21 = Constant() {r=0.1}
  n22 = Constant()
  NamedRerouteDeclaration(None=Multiply(A=MaterialFunctionCall(DistanceField=@n20.Outline, Threshold=@n21, Smoothing=@n22) {materialFunction=&MF_UI_AntiAliasing}.Outer) {constB=0.4}) {name=Outline, nodeColor=<rgba 0 0 0 1>}
  NamedRerouteDeclaration(None=MaterialFunctionCall() {materialFunction=&MF_UI_UVScaling}.Scaled to Fit) {name=Scale to Fit, nodeColor=<rgba 0 0 0 1>}
  n23 = ScalarParameter() {defaultValue=1, parameterName=X}
  n24 = Add(
   A=MaterialFunctionCall(UVs=NamedRerouteUsage(), Texture Scale=ScalarParameter() {defaultValue=0.8, parameterName=Hex Scale}) {materialFunction=&ScaleUVsByCenter}.UVs
   B=AppendVector(A=LinearInterpolate(Alpha=@n23) {constB=0}, B=Constant())
  )
  n25 = Constant() {r=0.25}
  n26 = MaterialFunctionCall(
   DistanceField=MaterialFunctionCall(
    UVs=MaterialFunctionCall(UVs=@n24, "Rotation Angle (0-1)"=@n25) {materialFunction=&CustomRotator}.Rotated Values
    Radius=Constant() {r=0.5}
    Sides=Constant() {r=6}
   ) {materialFunction=&MF_SDF_RegularPolygon}.SDF
   Threshold=Constant() {r=0.1}
   Smoothing=Constant()
  ) {materialFunction=&MF_UI_AntiAliasing}
  NamedRerouteDeclaration(None=Multiply(A=Power(Base=@n26.Inner) {constExponent=4}) {constB=0.3}) {name=Hex, nodeColor=<rgba 0 0 0 1>}
 ]
}
