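# UI material for a small ToyFox ability icon, filed under team viewer icons, using alpha-composite blending.
# A swappable texture parameter, defaulting to the melee icon, is sampled with center-scaled UVs and split into channels through named reroutes.
# Distance-field anti-aliasing and an SDF shape outline produce icon and outline masks.
# Weighted sums combine layers, including a dark-to-bright red color tint.
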Material /ToyFoxGameplay/UI/Materials/TeamViewerIcon/M_UI_ToyFox_AbilitySmall
graph {
 properties {bAutomaticallySetUsageInEditor=false, blendMode=BLEND_AlphaComposite, materialDomain=MD_UI, usedShadingModels=Default Lit}
 nodes [
  n1 = NamedRerouteUsage()
  n2 = Constant3Vector() {constant=<rgba 1 0.0045 0.004 1>}
  n3 = Multiply(A=NamedRerouteUsage())
  n4 = Constant3Vector() {constant=<rgba 0.2526 0 0 0>}
  n5 = Multiply(A=NamedRerouteUsage())
  Add(
   A=Multiply(A=Constant3Vector() {constant=<rgba 0.013 0.0006 0 1>}, B=Multiply(A=@n1))
   B=Add(A=Multiply(A=@n2, B=@n3), B=Multiply(A=@n4, B=@n5))
  )
  n6 = NamedRerouteUsage()
  n7 = NamedRerouteUsage()
  Add(A=Add(A=Multiply(A=@n6) {constB=0.8}, B=Multiply(A=@n7) {constB=0.5}), B=Multiply(A=NamedRerouteUsage()) {constB=0.5})
  Constant3Vector() {constant=<rgba 1 0.2347 0.2344 1>}
  Constant3Vector() {constant=<rgba 0.2526 0.0724 0.0724 0>}
  n8 = NamedRerouteUsage()
  n9 = Constant() {r=0.5}
  n10 = Constant()
  NamedRerouteDeclaration(None=Multiply(A=MaterialFunctionCall(DistanceField=@n8, Threshold=@n9, Smoothing=@n10) {materialFunction=&MF_UI_AntiAliasing}.Outer)) {name=Icon 02, nodeColor=<rgba 0 0 0 1>}
  NamedRerouteDeclaration(None=MaterialFunctionCall(DistanceField=NamedRerouteUsage(), Threshold=Constant() {r=0.5}, Smoothing=Constant() {r=0.1}) {materialFunction=&MF_UI_AntiAliasing}.Outer) {name=Icon, nodeColor=<rgba 0 0 0 1>}
  n11 = TextureSampleParameter2D(UVs=MaterialFunctionCall(Texture Scale=Constant() {r=0.9}) {materialFunction=&ScaleUVsByCenter}.UVs) {parameterName=Texture Input, texture=/TidalCrane_ToyFox_Boss/UI/T_UI_ToyFox_Icon_Melee}
  NamedRerouteDeclaration(None=@n11.B) {name=Blue Channel, nodeColor=<rgba 0 0 0 1>}
  n12 = NamedRerouteUsage()
  n13 = NamedRerouteUsage()
  NamedRerouteDeclaration(None=Add(A=Add(A=@n12, B=@n13), B=NamedRerouteUsage())) {name=Icon Composite, nodeColor=<rgba 0 0 0 1>}
  NamedRerouteDeclaration(None=@n11.A) {name=Alpha Channel, nodeColor=<rgba 0 0 0 1>}
  NamedRerouteDeclaration(None=@n11.R) {name=Red Channel, nodeColor=<rgba 0 0 0 1>}
  NamedRerouteDeclaration(None=@n11.G) {name=Green Channel, nodeColor=<rgba 0 0 0 1>}
  n14 = NamedRerouteUsage()
  n15 = Constant() {r=0.3}
  NamedRerouteDeclaration(None=MaterialFunctionCall(DistanceField=MaterialFunctionCall!(DistanceField=@n14, Threshold, Thickness=@n15, Alignment, Smoothing) {materialFunction=/FortUILibrary/MaterialFunctions/Shapes/Modifiers/MF_UI_ShapeOutline}.SDFOutline) {materialFunction=&MF_UI_AntiAliasing}.Outer) {name=IconOutline, nodeColor=<rgba 0 0 0 1>}
 ]
}
