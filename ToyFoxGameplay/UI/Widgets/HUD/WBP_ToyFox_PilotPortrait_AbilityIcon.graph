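# HUD widget for the ToyFox pilot portrait that shows an ability icon and animates it green while that ability is in use.
# Setting a pilot index binds to the game state component's pilot-actors update and looks up that pilot's actor.
# Two async gameplay tag query waits on the actor play the tint animation forward or in reverse.
# Each update adds new waits without ending earlier ones; they end only on Destruct.
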WidgetBlueprint /ToyFoxGameplay/UI/Widgets/HUD/WBP_ToyFox_PilotPortrait_AbilityIcon
animations [AnimInUse {
  binding_count=1
  length_seconds=0.5
  object=&self:AnimInUse
  sequence {bindings [Image {
     id=#1
     possessed_class=Image
     tracks [{
       type=MovieSceneColorTrack
       display_name=Color and Opacity
       sections [{
         type=MovieSceneColorSection
         end_frame=10
         start_frame=0
         channels [
          Color.R {type=MovieSceneScriptingFloatChannel, keys [[0,1],[10,0],[40,0]]}
          Color.G {type=MovieSceneScriptingFloatChannel, const [1,3]}
          Color.B {type=MovieSceneScriptingFloatChannel, keys [[0,1],[10,0],[40,0]]}
          Color.A {type=MovieSceneScriptingFloatChannel, const [1,3]}
         ]
        }]
      }]
    }]}
 }]
graphs [
 EventGraph {nodes [
   HandlePilotActorsUpdated() {is=title K2Node_CustomEvent}
   > n1 = Get!(execute:Exec) -> (then:Exec, else:Exec, GameStateComponent:B Toy Fox Game State Component Object Reference)
   n2 = GetPilotActors!(self:B Toy Fox Game State Component Object Reference=@n1) -> (PilotActors:Array of Actor Object References)
   n3 = Get PilotIndex() -> (PilotIndex:int=0)
   > Branch(Condition=Is Valid Index!(TargetArray:Array of Actor Object References=@n2, IndexToTest:int=@n3) -> (ReturnValue:bool))
   n4 = "Get (a copy)"(Array:Array of Actor Object References=@n2, Dimension 1=@n3) -> (Output:Actor Object Reference)
   > Is Valid(InputObject=@n4)
   n5 = Get InUseGameplayTagQuery() -> (InUseGameplayTagQuery:Gameplay Tag Query Structure)
   > n6 = WaitGameplayTagQueryOnActor!(
    execute:Exec
    TargetActor:Actor Object Reference=@n4
    TagQuery:Gameplay Tag Query Structure=@n5
    TriggerCondition:EWaitGameplayTagQueryTriggerCondition Enum=WhenTrue
    bOnlyTriggerOnce:bool=false
   ) -> (then:Exec, AsyncTaskProxy:Ability Async Wait Gameplay Tag Query Object Reference, Triggered:Exec) {is=title K2Node_AsyncAction}
   > Add(
    TargetArray:Array of Ability Async Wait Gameplay Tag Query Object References=Get AsyncActions() -> (AsyncActions:Array of Ability Async Wait Gameplay Tag Query Object References)
    NewItem:"Ability Async Wait Gameplay Tag Query Object Reference (by ref)"=@n6
   ) {is=title K2Node_CallArrayFunction}
   > n7 = WaitGameplayTagQueryOnActor(TargetActor=@n4, TagQuery=@n5, TriggerCondition=WhenFalse, bOnlyTriggerOnce=false)
   > Add(TargetArray=Get AsyncActions(), NewItem=@n7)
   SetAbilityInUse!(execute=@n7.Triggered, self, bEnable:bool=false) -> (then)
   SetAbilityInUse(execute=@n6.Triggered, bEnable=true)
   HandlePilotIndexSet() {is=title K2Node_CustomEvent}
   > n8 = GetGameStateAthena!(execute) -> (then, ReturnValue:Fort Game State Athena Object Reference)
   > Is Valid(InputObject=@n8)
   > n9 = GetComponentByClass!(
    execute
    self:Actor Object Reference=@n8
    ComponentClass:Actor Component Class Reference=/ToyFoxGameplay/Gameplay/B_ToyFoxGameStateComponent.B_ToyFoxGameStateComponent_C
   ) -> (then, ReturnValue:B Toy Fox Game State Component Object Reference)
   > Is Valid(InputObject=@n9)
   > Set GameStateComponent(GameStateComponent:B Toy Fox Game State Component Object Reference=@n9) -> (Output_Get:B Toy Fox Game State Component Object Reference)
   > Bind Event to On Pilot Actors Updated(self:B Toy Fox Game State Component Object Reference=@n9, Delegate=Create Event()) {is=title K2Node_AddDelegate}
   > HandlePilotActorsUpdated() {is=title K2Node_CallFunction}
   n10 = SetAbilityInUse!() -> (OutputDelegate, then, bEnable:bool=~) {is=title K2Node_CustomEvent}
   > n11 = Branch(Condition=@n10.bEnable)
   n12 = Get AnimInUse() -> (AnimInUse:Widget Animation Object Reference)
   > PlayAnimationForward(InAnimation=@n12)
   PlayAnimationReverse(execute=@n11.else, InAnimation=@n12)
   Event Construct()
   > n13 = Parent: Construct()
   SetBrushFromMaterial(
    execute=[@n13.then, @n15.then]
    self=Get Image() -> (Image:Image Object Reference)
    Material=Get AbilityIcon() -> (AbilityIcon:Material Interface Object Reference)
   )
   n14 = Event PreConstruct()
   > n15 = Parent: PreConstruct(IsDesignTime=@n14.IsDesignTime)
   Event Destruct()
   > Parent: Destruct()
   > EndAction(self:Ability Async Object Reference=Get AsyncActions())
  ]}
 SetAbilityIcon {nodes [
   n1 = SetAbilityIcon!() -> (then, Material:Material Interface Object Reference) {is=title K2Node_FunctionEntry}
   > SetBrushFromMaterial(self=Get Image(), Material=@n1)
  ]}
 SetPilotIndex {nodes [
   n1 = SetPilotIndex!() -> (then, Index:int) {is=title K2Node_FunctionEntry}
   > Set PilotIndex(PilotIndex:int=@n1) -> (Output_Get:int=0)
   > HandlePilotIndexSet() {is=title K2Node_CallFunction}
  ]}
]
variables [
 AbilityIcon {type=MaterialInterface, default=/ToyFoxGameplay/UI/Materials/TeamViewerIcon/M_UI_ToyFox_AbilitySmall}
 AsyncActions {type="array<AbilityAsync_WaitGameplayTagQuery>", default []}
 GameStateComponent {type=B_ToyFoxGameStateComponent_C, default=null}
 InUseGameplayTagQuery {type=GameplayTagQuery, default {autoDescription="", tokenStreamVersion=0, userDescription="", queryTokenStream [], tagDictionary []}}
 PilotIndex {type=integer}
]
widgets {
 info {inheritedWidgetCount=1, parentClass=&FortHUDElementWidget, rootWidgetClass=&Overlay, widgetCount=3}
 widgets [
  Overlay_22 {widgetClassPath=&Overlay, widget=Overlay_22}
  Image {
   widgetClassPath=&Image
   bIsVariable=true
   parent=Overlay_22
   widget=Image
   properties {
    visibility=HitTestInvisible
    brush {
     bIsDynamicallyLoaded=false
     drawAs=Image
     imageType=FullColor
     mirroring=NoMirror
     resourceName=None
     resourceObject=&M_UI_ToyFox_AbilitySmall
     tiling=NoTile
     imageSize=<35 35>
     margin=<ltrb 0 0 0 0>
     outlineSettings {
      bUseBrushTransparency=false
      roundingType=HalfHeightRadius
      width=0
      color {colorUseRule=UseColor_Specified, specifiedColor=<rgba 0 0 0 0>}
      cornerRadii=<0 0 0 0>
     }
     tintColor {colorUseRule=UseColor_Specified, specifiedColor=<rgba 1 1 1 1>}
     uVRegion {bIsValid=false, max=<0 0>, min=<0 0>}
    }
   }
   slot {class=OverlaySlot, properties {horizontalAlignment=HAlign_Fill, verticalAlignment=VAlign_Fill}}
  }
  TutorialHighlight {widgetClassPath=&TutorialHighlight}
 ]
}
