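# ToyFox HUD widget showing one pilot's portrait, name and ability icons, selected by PilotIndex.
# It binds to the game state component's pilot-actors update, then async-loads the portrait texture into a dynamic material and sets the name.
# Invalid pilot data starts a short randomized refresh timer; name changes after the first play the swap animation.
# Ability icons receive PilotIndex; bPrimaryPilot collapses the rocket icons, otherwise melee, cannon, stomp and scan collapse.
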WidgetBlueprint /ToyFoxGameplay/UI/Widgets/HUD/WBP_ToyFox_PilotPortrait
animations [Anim_PlayerSwap {
  binding_count=9
  length_seconds=0.6
  object=&self:Anim_PlayerSwap
  sequence {
   bindings [
    Image_Portrait {
     id=#1
     possessed_class=Image
     tracks [
      {
       type=MovieSceneFloatTrack
       display_name=Render Opacity
       sections [{type=MovieSceneFloatSection, channels [None {type=MovieSceneScriptingFloatChannel, keys [[2,0],[8,1]]}]}]
      }
      {
       type=MovieSceneWidgetMaterialTrack
       sections [{type=MovieSceneParameterSection, channels [Flash {type=MovieSceneScriptingFloatChannel, keys [[5,1],[18,0]]}]}]
      }
     ]
    }
    Image_Portrait_Backer {
     id=#2
     possessed_class=Image
     tracks [{
       type=MovieSceneWidgetMaterialTrack
       sections [{
         type=MovieSceneParameterSection
         channels [
          Master Opacity {type=MovieSceneScriptingFloatChannel, keys [[3,0],[11.4,1]]}
          Flash {type=MovieSceneScriptingFloatChannel, keys [[8.6,1],[17,0]]}
         ]
        }]
      }]
    }
    Text_PlayerName {
     id=#3
     possessed_class=CommonTextBlock
     tracks [
      {
       type=MovieScene2DTransformTrack
       sections [{
         type=MovieScene2DTransformSection
         channels [
          Translation.X {type=MovieSceneScriptingFloatChannel, keys [[9,-59.4856],[20.538,8]]}
          Translation.Y {type=MovieSceneScriptingFloatChannel, const [-7,1]}
          Angle {type=MovieSceneScriptingFloatChannel, const [0,1]}
          Scale.X {type=MovieSceneScriptingFloatChannel, const [1,1]}
          Scale.Y {type=MovieSceneScriptingFloatChannel, const [1,1]}
          Shear.X {type=MovieSceneScriptingFloatChannel, const [0,1]}
          Shear.Y {type=MovieSceneScriptingFloatChannel, const [0,1]}
         ]
        }]
      }
      {
       type=MovieSceneFloatTrack
       display_name=Render Opacity
       sections [{type=MovieSceneFloatSection, channels [None {type=MovieSceneScriptingFloatChannel, keys [[9,0],[29,1]]}]}]
      }
     ]
    }
    Image_Portrait_YGradient {
     id=#4
     possessed_class=Image
     tracks [{
       type=MovieSceneWidgetMaterialTrack
       sections [{
         type=MovieSceneParameterSection
         channels [Gradient Master Opacity {type=MovieSceneScriptingFloatChannel, keys [[4,0],[12.665,1],[16.998,0.6]]}]
        }]
      }]
    }
    Image_HexGlow {
     id=#5
     possessed_class=Image
     tracks [
      {
       type=MovieSceneFloatTrack
       display_name=Render Opacity
       sections [{type=MovieSceneFloatSection, channels [None {type=MovieSceneScriptingFloatChannel, const [0.0196,1]}]}]
      }
      {
       type=MovieScene2DTransformTrack
       sections [{
         type=MovieScene2DTransformSection
         channels [
          Translation.X {type=MovieSceneScriptingFloatChannel, keys [[0,-170],[16,-103.1724]]}
          Translation.Y {type=MovieSceneScriptingFloatChannel, const [0,1]}
          Angle {type=MovieSceneScriptingFloatChannel, const [0,1]}
          Scale.X {type=MovieSceneScriptingFloatChannel, const [5,1]}
          Scale.Y {type=MovieSceneScriptingFloatChannel, const [5,1]}
          Shear.X {type=MovieSceneScriptingFloatChannel, const [0,1]}
          Shear.Y {type=MovieSceneScriptingFloatChannel, const [0,1]}
         ]
        }]
      }
      {
       type=MovieSceneWidgetMaterialTrack
       sections [{type=MovieSceneParameterSection, channels [Master Opacity {type=MovieSceneScriptingFloatChannel, keys [[0,0],[16,1],[36,0]]}]}]
      }
     ]
    }
    Image_Underline {
     id=#6
     possessed_class=Image
     tracks [{
       type=MovieSceneWidgetMaterialTrack
       sections [{
         type=MovieSceneParameterSection
         end_frame=36
         start_frame=0
         channels [
          Flash {type=MovieSceneScriptingFloatChannel, keys [[11,1],[24,0]]}
          Master Opacity {type=MovieSceneScriptingFloatChannel, keys [[12,0],[24,1]]}
         ]
        }]
      }]
    }
    HorizontalBox_38 {
     id=#7
     possessed_class=HorizontalBox
     tracks [{
       type=MovieScene2DTransformTrack
       sections [{
         type=MovieScene2DTransformSection
         channels [
          Translation.X {type=MovieSceneScriptingFloatChannel, keys [[0,-100],[10,0]]}
          Translation.Y {type=MovieSceneScriptingFloatChannel, const [0,1]}
          Angle {type=MovieSceneScriptingFloatChannel, const [0,1]}
          Scale.X {type=MovieSceneScriptingFloatChannel, const [1,1]}
          Scale.Y {type=MovieSceneScriptingFloatChannel, const [1,1]}
          Shear.X {type=MovieSceneScriptingFloatChannel, const [0,1]}
          Shear.Y {type=MovieSceneScriptingFloatChannel, const [0,1]}
         ]
        }]
      }]
    }
    Overlay_2 {
     id=#8
     possessed_class=Overlay
     tracks [{
       type=MovieSceneFloatTrack
       display_name=Render Opacity
       sections [{type=MovieSceneFloatSection, channels [None {type=MovieSceneScriptingFloatChannel, keys [[0,0],[22,1]]}]}]
      }]
    }
    Image_Lightning {
     id=#9
     possessed_class=Image
     tracks [
      {
       type=MovieSceneFloatTrack
       display_name=Render Opacity
       sections [{type=MovieSceneFloatSection, channels [None {type=MovieSceneScriptingFloatChannel, keys [[0,0.2],[22,0]]}]}]
      }
      {
       type=MovieScene2DTransformTrack
       sections [{
         type=MovieScene2DTransformSection
         channels [
          Translation.X {type=MovieSceneScriptingFloatChannel, keys [[0,-113],[22,80]]}
          Translation.Y {type=MovieSceneScriptingFloatChannel, const [0,1]}
          Angle {type=MovieSceneScriptingFloatChannel, const [90,1]}
          Scale.X {type=MovieSceneScriptingFloatChannel, const [0.3,1]}
          Scale.Y {type=MovieSceneScriptingFloatChannel, const [5,1]}
          Shear.X {type=MovieSceneScriptingFloatChannel, const [0,1]}
          Shear.Y {type=MovieSceneScriptingFloatChannel, const [0,1]}
         ]
        }]
      }
     ]
    }
   ]
   display_rate {numerator=60}
  }
 }]
graphs [
 EventGraph {nodes [
   HandlePilotActorsUpdated() {is=title K2Node_CustomEvent}
   > Clear and Invalidate Timer by Handle(Handle=Get RefreshPilotPortraitTimer() -> (RefreshPilotPortraitTimer:Timer Handle Structure))
   > n1 = Get!(execute:Exec) -> (then:Exec, else:Exec, GameStateComponent:B Toy Fox Game State Component Object Reference)
   n2 = GetPilotActors!(self:B Toy Fox Game State Component Object Reference=@n1) -> (PilotActors:Array of Actor Object References)
   n3 = Get PilotIndex() -> (PilotIndex:int=0)
   > n4 = Branch(Condition=Is Valid Index!(TargetArray:Array of Actor Object References=@n2, IndexToTest:int=@n3) -> (ReturnValue:bool))
   n5 = "Get (a copy)"(Array:Array of Actor Object References=@n2, Dimension 1=@n3) -> (Output:Actor Object Reference)
   > n6 = Is Valid(InputObject=@n5)
   > n7 = GetPilotActorPortraitInfo!(execute, self:B Toy Fox Game State Component Object Reference=@n1, Actor:Actor Object Reference=@n5) -> (then, bValid:bool, SmallPreviewImage:Texture 2D Soft Object Reference, LargePreviewImage:Texture 2D Soft Object Reference, Name:Text)
   > n8 = Branch(Condition=@n7)
   > Set PendingNameChange(PendingNameChange:Text=@n7.Name) -> (Output_Get:Text)
   > n9 = Async Load Asset(Asset=@n7.SmallPreviewImage)
   n10 = Cast To Texture2D(execute=@n9.Completed, Object=@n9)
   > n11 = GetDynamicMaterial(self=Get Image_Portrait() -> (Image_Portrait:Image Object Reference))
   > n12 = SetTextureParameterValue(self=@n11, ParameterName=SmallPreview, Value=@n10)
   n13 = Get Text_PlayerName() -> (Text_PlayerName:Common Text Object Reference)
   n14 = Branch(
    execute=[@n12.then, @n10.CastFailed]
    Condition="Equal Exactly (Text)"!(A:"Text (by ref)"=Get Text(self:Text Object Reference=@n13), B:"Text (by ref)"=Get PendingNameChange() -> (PendingNameChange:Text)) -> (ReturnValue:bool)
   )
   "SetText (Text)"(execute=@n14.else, self=Get Text_PlayerName(), InText=Get PendingNameChange())
   > n15 = Branch(Condition=Get IsInitialNameSet() -> (IsInitialNameSet:bool))
   > PlayAnimation(InAnimation=Get Anim_PlayerSwap() -> (Anim_PlayerSwap:Widget Animation Object Reference))
   Set IsInitialNameSet(execute=@n15.else, IsInitialNameSet:bool=true) -> (Output_Get:bool)
   n16 = Set Timer by Event(
    execute=[@n8.else, @n6.Is Not Valid, @n4.else]
    Delegate=Create Event()
    Time=RandomFloatInRange(Min=0.01, Max=0.1)
    bLooping=false
    bMaxOncePerFrame=true
   )
   > Set RefreshPilotPortraitTimer(RefreshPilotPortraitTimer:Timer Handle Structure=@n16) -> (Output_Get:Timer Handle Structure)
   n17 = Event PreConstruct()
   > Parent: PreConstruct(IsDesignTime=@n17.IsDesignTime)
   > UpdateAbilityWidgets()
   > SetVisibility(
    self=[
     Get AbilityIcon_RocketBasic() -> (AbilityIcon_RocketBasic:WBP Toy Fox Pilot Portrait Ability Icon Object Reference)
     Get AbilityIcon_Stomp() -> (AbilityIcon_Stomp:WBP Toy Fox Pilot Portrait Ability Icon Object Reference)
     Get AbilityIcon_Scan() -> (AbilityIcon_Scan:WBP Toy Fox Pilot Portrait Ability Icon Object Reference)
     Get AbilityIcon_Melee() -> (AbilityIcon_Melee:WBP Toy Fox Pilot Portrait Ability Icon Object Reference)
     Get AbilityIcon_Cannon() -> (AbilityIcon_Cannon:WBP Toy Fox Pilot Portrait Ability Icon Object Reference)
     Get AbilityIcon_RocketSeeking() -> (AbilityIcon_RocketSeeking:WBP Toy Fox Pilot Portrait Ability Icon Object Reference)
    ]
    InVisibility=SelfHitTestInvisible
   )
   Event Construct()
   > Parent: Construct()
   > n18 = Sequence()
   > n19 = GetGameStateAthena!(execute) -> (then, ReturnValue:Fort Game State Athena Object Reference)
   > Is Valid(InputObject=@n19)
   > n20 = GetComponentByClass!(
    execute
    self:Actor Object Reference=@n19
    ComponentClass:Actor Component Class Reference=/ToyFoxGameplay/Gameplay/B_ToyFoxGameStateComponent.B_ToyFoxGameStateComponent_C
   ) -> (then, ReturnValue:B Toy Fox Game State Component Object Reference)
   > Is Valid(InputObject=@n20)
   > Set GameStateComponent(GameStateComponent:B Toy Fox Game State Component Object Reference=@n20) -> (Output_Get:B Toy Fox Game State Component Object Reference)
   > Bind Event to On Pilot Actors Updated(self:B Toy Fox Game State Component Object Reference=@n20, Delegate=Create Event()) {is=title K2Node_AddDelegate}
   > HandlePilotActorsUpdated() {is=title K2Node_CallFunction}
   UpdateAbilityWidgets(execute=@n18.then_1)
  ]}
 UpdateAbilityWidgets {nodes [
   UpdateAbilityWidgets() {is=title K2Node_FunctionEntry}
   > SetPilotIndex!(
    execute
    self:WBP Toy Fox Pilot Portrait Ability Icon Object Reference=[
     Get AbilityIcon_RocketBasic()
     Get AbilityIcon_Stomp()
     Get AbilityIcon_Scan()
     Get AbilityIcon_Melee()
     Get AbilityIcon_Cannon()
     Get AbilityIcon_RocketSeeking()
    ]
    Index:int=Get PilotIndex() -> (PilotIndex=0)
   ) -> (then)
   > n1 = Branch(Condition=Get bPrimaryPilot() -> (bPrimaryPilot:bool))
   > SetVisibility(self=[Get AbilityIcon_RocketSeeking(), Get AbilityIcon_RocketBasic()], InVisibility=Collapsed)
   SetVisibility(
    execute=@n1.else
    self=[Get AbilityIcon_Stomp(), Get AbilityIcon_Scan(), Get AbilityIcon_Melee(), Get AbilityIcon_Cannon()]
    InVisibility=Collapsed
   )
  ]}
]
variables [
 bPrimaryPilot {type=boolean}
 GameStateComponent {type=B_ToyFoxGameStateComponent_C, default=null}
 IsInitialNameSet {type=boolean}
 PendingNameChange {type=string, default=""}
 PilotIndex {type=integer, default=-1}
 RefreshPilotPortraitTimer {type=TimerHandle}
]
widgets {
 info {inheritedWidgetCount=1, parentClass=&FortHUDElementWidget, rootWidgetClass=&Overlay, widgetCount=26}
 widgets [
  Overlay_18 {widgetClassPath=&Overlay, widget=Overlay_18}
  HorizontalBox_38 {
   widgetClassPath=&HorizontalBox
   parent=Overlay_18
   widget=HorizontalBox_38
   slot {class=OverlaySlot, properties {horizontalAlignment=HAlign_Fill, verticalAlignment=VAlign_Fill}}
  }
  Overlay_31 {widgetClassPath=&Overlay, parent=HorizontalBox_38, widget=Overlay_31, slot {class=HorizontalBoxSlot}}
  Image_Portrait_Backer {
   widgetClassPath=&Image
   bIsVariable=true
   parent=Overlay_31
   widget=Image_Portrait_Backer
   properties {brush {
     bIsDynamicallyLoaded=false
     drawAs=Image
     imageType=FullColor
     mirroring=NoMirror
     resourceName=None
     resourceObject=/ToyFoxGameplay/UI/Materials/M_ToyFox_TeamViewer_Profile
     tiling=NoTile
     imageSize=<100 100>
     margin=<ltrb 0 0 0 0>
     outlineSettings {
      bUseBrushTransparency=false
      roundingType=HalfHeightRadius
      width=0
      color {colorUseRule=UseColor_Specified, specifiedColor=<rgba 0 0 0 0>}
      cornerRadii=<0 0 0 0>
     }
     tintColor {colorUseRule=UseColor_Specified, specifiedColor=<rgba 1 1 1 1>}
     uVRegion {bIsValid=false, max=<0 0>, min=<0 0>}
    }}
   slot {class=OverlaySlot, properties {horizontalAlignment=HAlign_Center, verticalAlignment=VAlign_Center, padding=<ltrb 0 -15 0 -15>}}
  }
  Image_Portrait {
   widgetClassPath=&Image
   bIsVariable=true
   parent=Overlay_31
   widget=Image_Portrait
   properties {brush {
     bIsDynamicallyLoaded=false
     drawAs=Image
     imageType=FullColor
     mirroring=NoMirror
     resourceName=None
     resourceObject=/ToyFoxGameplay/UI/Materials/M_UI_ToyFox_UserPortrait
     tiling=NoTile
     imageSize=<78 78>
     margin=<ltrb 0 0 0 0>
     outlineSettings {
      bUseBrushTransparency=false
      roundingType=HalfHeightRadius
      width=0
      color {colorUseRule=UseColor_Specified, specifiedColor=<rgba 0 0 0 0>}
      cornerRadii=<0 0 0 0>
     }
     tintColor {colorUseRule=UseColor_Specified, specifiedColor=<rgba 1 1 1 1>}
     uVRegion {bIsValid=false, max=<0 0>, min=<0 0>}
    }}
   slot {class=OverlaySlot, properties {horizontalAlignment=HAlign_Center, verticalAlignment=VAlign_Center, padding=<ltrb 0 -15 0 -15>}}
  }
  Image_Portrait_YGradient {
   widgetClassPath=&Image
   bIsVariable=true
   parent=Overlay_31
   widget=Image_Portrait_YGradient
   properties {brush {
     bIsDynamicallyLoaded=false
     drawAs=Image
     imageType=FullColor
     mirroring=NoMirror
     resourceName=None
     resourceObject=/ToyFoxGameplay/UI/Materials/MI_ToyFox_TeamViewer_Profile
     tiling=NoTile
     imageSize=<100 100>
     margin=<ltrb 0 0 0 0>
     outlineSettings {
      bUseBrushTransparency=false
      roundingType=HalfHeightRadius
      width=0
      color {colorUseRule=UseColor_Specified, specifiedColor=<rgba 0 0 0 0>}
      cornerRadii=<0 0 0 0>
     }
     tintColor {colorUseRule=UseColor_Specified, specifiedColor=<rgba 1 1 1 1>}
     uVRegion {bIsValid=false, max=<0 0>, min=<0 0>}
    }}
   slot {class=OverlaySlot, properties {horizontalAlignment=HAlign_Center, verticalAlignment=VAlign_Center, padding=<ltrb 0 -15 0 -15>}}
  }
  VerticalBox_140 {
   widgetClassPath=&VerticalBox
   parent=HorizontalBox_38
   widget=VerticalBox_140
   slot {class=HorizontalBoxSlot, properties {horizontalAlignment=HAlign_Left, padding=<ltrb 0 10 0 0>}}
  }
  Text_PlayerName {
   widgetClassPath=&CommonTextBlock
   bIsVariable=true
   parent=VerticalBox_140
   widget=Text_PlayerName
   properties {
    text=PlayerName
    colorAndOpacity {colorUseRule=UseColor_Specified, specifiedColor=<rgba 1 0.9557 0.9557 1>}
    font {
     bForceMonospaced=false
     bMaterialIsStencil=false
     fontMaterial=null
     fontObject=&BurbankSmall
     letterSpacing=0
     monospacedWidth=1
     size=11.25
     skewAmount=0
     typefaceFontName=Black
     outlineSettings {
      bApplyOutlineToDropShadows=false
      bMiteredCorners=false
      bSeparateFillAlpha=false
      outlineMaterial=&M-FontShadow
      outlineSize=1
      outlineColor=<rgba 0 0 0 0.5>
     }
    }
    renderTransform {angle=0, scale=<1 1>, shear=<0 0>, translation=<8 -7>}
    shadowOffset=<0 0>
   }
   slot {class=VerticalBoxSlot, properties {verticalAlignment=VAlign_Top, padding=<ltrb -2 0 0 -18>}}
  }
  Overlay_2 {
   widgetClassPath=&Overlay
   parent=VerticalBox_140
   widget=Overlay_2
   properties {bOverride_Cursor=true}
   slot {class=VerticalBoxSlot, properties {padding=<ltrb 0 -4 0 0>}}
  }
  VerticalBox_1 {widgetClassPath=&VerticalBox, parent=Overlay_2, widget=VerticalBox_1, slot {class=OverlaySlot}}
  SizeBox_0 {
   widgetClassPath=&SizeBox
   parent=VerticalBox_1
   widget=SizeBox_0
   properties {bOverride_MaxDesiredWidth=true, bOverride_WidthOverride=true, maxDesiredWidth=100, widthOverride=130}
   slot {class=VerticalBoxSlot, properties {horizontalAlignment=HAlign_Left, padding=<ltrb 0 0 0 -10>}}
  }
  Image_Underline {
   widgetClassPath=&Image
   bIsVariable=true
   parent=SizeBox_0
   widget=Image_Underline
   properties {
    brush {
     bIsDynamicallyLoaded=false
     drawAs=Image
     imageType=FullColor
     mirroring=NoMirror
     resourceName=None
     resourceObject=/ToyFoxGameplay/UI/Materials/M_ToyFox_NameBar
     tiling=NoTile
     imageSize=<32 32>
     margin=<ltrb 0 0 0 0>
     outlineSettings {
      bUseBrushTransparency=false
      roundingType=HalfHeightRadius
      width=0
      color {colorUseRule=UseColor_Specified, specifiedColor=<rgba 0 0 0 0>}
      cornerRadii=<0 0 0 0>
     }
     tintColor {colorUseRule=UseColor_Specified, specifiedColor=<rgba 1 1 1 1>}
     uVRegion {bIsValid=false, max=<0 0>, min=<0 0>}
    }
    renderTransform {angle=0, scale=<3 3>, shear=<0 0>, translation=<14 -2>}
   }
   slot {class=SizeBoxSlot, properties {horizontalAlignment=HAlign_Left, padding=<ltrb 10 0 0 0>}}
  }
  HorizontalBox_468 {
   widgetClassPath=&HorizontalBox
   parent=VerticalBox_1
   widget=HorizontalBox_468
   properties {renderTransform {angle=0, scale=<1 1>, shear=<0 0>, translation=<5 0>}}
   slot {class=VerticalBoxSlot, properties {horizontalAlignment=HAlign_Left, padding=<ltrb -20 4 0 0>}}
  }
  Overlay_0 {
   widgetClassPath=&Overlay
   parent=HorizontalBox_468
   widget=Overlay_0
   slot {class=HorizontalBoxSlot, properties {horizontalAlignment=HAlign_Left, padding=<ltrb 0 -2 0 0>, size {sizeRule=Fill, value=1}}}
  }
  Image_Bg {
   widgetClassPath=&Image
   bIsVariable=true
   parent=Overlay_0
   widget=Image_Bg
   properties {
    renderOpacity=0.4
    brush {
     bIsDynamicallyLoaded=false
     drawAs=Image
     imageType=FullColor
     mirroring=NoMirror
     resourceName=None
     resourceObject=/ToyFoxGameplay/UI/Materials/M_ToyFox_TeamViewer_Bg
     tiling=NoTile
     imageSize=<100 10>
     margin=<ltrb 0 0 0 0>
     outlineSettings {
      bUseBrushTransparency=false
      roundingType=HalfHeightRadius
      width=0
      color {colorUseRule=UseColor_Specified, specifiedColor=<rgba 0 0 0 0>}
      cornerRadii=<0 0 0 0>
     }
     tintColor {colorUseRule=UseColor_Specified, specifiedColor=<rgba 1 1 1 1>}
     uVRegion {bIsValid=false, max=<0 0>, min=<0 0>}
    }
    renderTransform {angle=0, scale=<1.3 1>, shear=<-30 0>, translation=<19 0>}
   }
   slot {class=OverlaySlot, properties {verticalAlignment=VAlign_Fill}}
  }
  Spacer_87 {widgetClassPath=&Spacer, parent=Overlay_0, widget=Spacer_87, properties {size=<80 0>}, slot {class=OverlaySlot}}
  HorizontalBox_0 {
   widgetClassPath=&HorizontalBox
   parent=Overlay_0
   widget=HorizontalBox_0
   properties {bOverride_Cursor=true}
   slot {class=OverlaySlot, properties {verticalAlignment=VAlign_Fill, padding=<ltrb 10 0 6 0>}}
  }
  AbilityIcon_Melee {
   widgetClassPath=&/WBP_ToyFox_PilotPortrait_AbilityIcon
   bIsVariable=true
   parent=HorizontalBox_0
   widget=AbilityIcon_Melee
   properties {inUseGameplayTagQuery {
     autoDescription=" ALL(  ANY( Gameplay.Melee.Combo1, Gameplay.Melee.Combo2 ) )"
     tokenStreamVersion=0
     userDescription=""
     queryTokenStream [0,1,5,1,1,2,0,1]
     tagDictionary [{tagName=Gameplay.Melee.Combo1}, {tagName=Gameplay.Melee.Combo2}]
    }}
   slot {class=HorizontalBoxSlot, properties {horizontalAlignment=HAlign_Left}}
  }
  AbilityIcon_Cannon {
   widgetClassPath=&WBP_ToyFox_PilotPortrait_AbilityIcon
   bIsVariable=true
   parent=HorizontalBox_0
   widget=AbilityIcon_Cannon
   properties {
    abilityIcon=/ToyFoxGameplay/UI/Materials/TeamViewerIcon/MI_UI_ToyFox_AbilitySmall_Cannon
    inUseGameplayTagQuery {
     autoDescription=" ALL(  ANY( ToyFox.Ability.Cannon.Active ) )"
     tokenStreamVersion=0
     userDescription=""
     queryTokenStream [0,1,5,1,1,1,0]
     tagDictionary [{tagName=ToyFox.Ability.Cannon.Active}]
    }
   }
   slot {class=HorizontalBoxSlot, properties {horizontalAlignment=HAlign_Left}}
  }
  AbilityIcon_Stomp {
   widgetClassPath=&WBP_ToyFox_PilotPortrait_AbilityIcon
   bIsVariable=true
   parent=HorizontalBox_0
   widget=AbilityIcon_Stomp
   properties {
    abilityIcon=/ToyFoxGameplay/UI/Materials/TeamViewerIcon/MI_UI_ToyFox_AbilitySmall_Stomp
    inUseGameplayTagQuery {
     autoDescription=" ALL(  ANY( ToyFox.Ability.Stomp.Active ) )"
     tokenStreamVersion=0
     userDescription=""
     queryTokenStream [0,1,5,1,1,1,0]
     tagDictionary [{tagName=ToyFox.Ability.Stomp.Active}]
    }
   }
   slot {class=HorizontalBoxSlot, properties {horizontalAlignment=HAlign_Left}}
  }
  AbilityIcon_Scan {
   widgetClassPath=&WBP_ToyFox_PilotPortrait_AbilityIcon
   bIsVariable=true
   parent=HorizontalBox_0
   widget=AbilityIcon_Scan
   properties {
    abilityIcon=/ToyFoxGameplay/UI/Materials/TeamViewerIcon/MI_UI_ToyFox_AbilitySmall_Scan
    inUseGameplayTagQuery {
     autoDescription=" ALL(  ANY( ToyFox.Ability.Scan.RecentlyUsed ) )"
     tokenStreamVersion=0
     userDescription=""
     queryTokenStream [0,1,5,1,1,1,0]
     tagDictionary [{tagName=ToyFox.Ability.Scan.RecentlyUsed}]
    }
   }
   slot {class=HorizontalBoxSlot, properties {horizontalAlignment=HAlign_Left}}
  }
  AbilityIcon_RocketSeeking {
   widgetClassPath=&WBP_ToyFox_PilotPortrait_AbilityIcon
   bIsVariable=true
   parent=HorizontalBox_0
   widget=AbilityIcon_RocketSeeking
   properties {
    abilityIcon=/ToyFoxGameplay/UI/Materials/TeamViewerIcon/MI_UI_ToyFox_AbilitySmall_SeekingRocket
    inUseGameplayTagQuery {
     autoDescription=" ALL(  ANY( ToyFox.Ability.Rocket.Seeking.Active ) )"
     tokenStreamVersion=0
     userDescription=""
     queryTokenStream [0,1,5,1,1,1,0]
     tagDictionary [{tagName=ToyFox.Ability.Rocket.Seeking.Active}]
    }
   }
   slot {class=HorizontalBoxSlot, properties {horizontalAlignment=HAlign_Left}}
  }
  AbilityIcon_RocketBasic {
   widgetClassPath=&WBP_ToyFox_PilotPortrait_AbilityIcon
   bIsVariable=true
   parent=HorizontalBox_0
   widget=AbilityIcon_RocketBasic
   properties {
    abilityIcon=/ToyFoxGameplay/UI/Materials/TeamViewerIcon/MI_UI_ToyFox_AbilitySmall_BasicRocket
    inUseGameplayTagQuery {
     autoDescription=" ALL(  ANY( ToyFox.Ability.Rocket.Basic.RecentlyUsed ) )"
     tokenStreamVersion=0
     userDescription=""
     queryTokenStream [0,1,5,1,1,1,0]
     tagDictionary [{tagName=ToyFox.Ability.Rocket.Basic.RecentlyUsed}]
    }
   }
   slot {class=HorizontalBoxSlot, properties {horizontalAlignment=HAlign_Left}}
  }
  Image_HexGlow {
   widgetClassPath=&Image
   bIsVariable=true
   parent=Overlay_18
   widget=Image_HexGlow
   properties {
    renderOpacity=0
    brush {
     bIsDynamicallyLoaded=false
     drawAs=Image
     imageType=FullColor
     mirroring=NoMirror
     resourceName=None
     resourceObject=/ToyFoxGameplay/UI/Materials/M_UI_ToyFox
     tiling=NoTile
     imageSize=<35 35>
     margin=<ltrb 0 0 0 0>
     outlineSettings {
      bUseBrushTransparency=false
      roundingType=HalfHeightRadius
      width=0
      color {colorUseRule=UseColor_Specified, specifiedColor=<rgba 0 0 0 0>}
      cornerRadii=<0 0 0 0>
     }
     tintColor {colorUseRule=UseColor_Specified, specifiedColor=<rgba 1 1 1 1>}
     uVRegion {bIsValid=false, max=<0 0>, min=<0 0>}
    }
    renderTransform {angle=0, scale=<5 5>, shear=<0 0>, translation=<-126 0>}
   }
   slot {class=OverlaySlot, properties {horizontalAlignment=HAlign_Center, verticalAlignment=VAlign_Center}}
  }
  Image_Lightning {
   widgetClassPath=&Image
   bIsVariable=true
   parent=Overlay_18
   widget=Image_Lightning
   properties {
    renderOpacity=0
    brush {
     bIsDynamicallyLoaded=false
     drawAs=Image
     imageType=FullColor
     mirroring=NoMirror
     resourceName=None
     resourceObject=/ToyFoxGameplay/UI/Widgets/HUD/Assets/BossBar/M_UI_HardEdgedGlow
     tiling=NoTile
     imageSize=<32 32>
     margin=<ltrb 0 0 0 0>
     outlineSettings {
      bUseBrushTransparency=false
      roundingType=HalfHeightRadius
      width=0
      color {colorUseRule=UseColor_Specified, specifiedColor=<rgba 0 0 0 0>}
      cornerRadii=<0 0 0 0>
     }
     tintColor {colorUseRule=UseColor_Specified, specifiedColor=<rgba 1 1 1 1>}
     uVRegion {bIsValid=false, max=<0 0>, min=<0 0>}
    }
    renderTransform {angle=90, scale=<0.3 5>, shear=<0 0>, translation=<-113 0>}
   }
   slot {class=OverlaySlot, properties {horizontalAlignment=HAlign_Fill, verticalAlignment=VAlign_Fill}}
  }
  TutorialHighlight {widgetClassPath=&TutorialHighlight}
 ]
}
