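# HUD element widget for ToyFox that shows a column of five pilot portraits, indexed 0 to 4, with the first marked as primary.
# On construct it broadcasts an active TeammateStruct through the global event router with the owning player as context, and broadcasts it inactive on destruct.
# When the compact UI is not in use, it offsets the root overlay's render translation.
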WidgetBlueprint /ToyFoxGameplay/UI/Widgets/HUD/WBP_ToyFox_TeammateWidget
animations [Intro {binding_count=0, length_seconds=5, object=&self:Intro, sequence {bindings []}}]
graphs [EventGraph {nodes [
   Event Construct()
   > n1 = Set!(execute, ActiveStruct_IsActive_1_#1:bool=true) -> (then, Output_Get:Teammate Struct Structure)
   > Broadcast Event!(
    execute
    self:Gameplay Event Router Component Object Reference=GetGlobalEventRouter()
    EventData:Teammate Struct Structure=@n1
    EventContext:Object Reference=GetOwningPlayer()
   ) -> (then, ReturnValue:bool)
   > n2 = Branch(Condition=IsUsingCompactUI())
   SetRenderTranslation(execute=@n2.else, self=Get Overlay_0() -> (Overlay_0:Overlay Object Reference), Translation=(X=24,Y=16))
   Event Destruct()
   > n3 = Set(ActiveStruct_IsActive_1_#1=False)
   > Broadcast Event(self=GetGlobalEventRouter(), EventData=@n3, EventContext=GetOwningPlayer())
  ]}]
variables [ActiveStruct {type=TeammateStruct, default {isActive=false}}]
widgets {
 info {inheritedWidgetCount=1, parentClass=&FortHUDElementWidget, rootWidgetClass=&Overlay, widgetCount=8}
 widgets [
  Overlay_0 {widgetClassPath=&Overlay, bIsVariable=true, widget=Overlay_0, properties {bOverride_Cursor=true}}
  GridPanel_39 {widgetClassPath=&GridPanel, parent=Overlay_0, widget=GridPanel_39, slot {class=OverlaySlot}}
  PilotPortrait_1 {
   widgetClassPath=&/WBP_ToyFox_PilotPortrait
   bIsVariable=true
   parent=GridPanel_39
   widget=PilotPortrait_1
   properties {bPrimaryPilot=true, pilotIndex=0}
   slot {class=GridSlot, properties {row=1, padding=<ltrb 0 0 0 10>}}
  }
  PilotPortrait_2 {
   widgetClassPath=&WBP_ToyFox_PilotPortrait
   bIsVariable=true
   parent=GridPanel_39
   widget=PilotPortrait_2
   properties {pilotIndex=1}
   slot {class=GridSlot, properties {row=2, padding=<ltrb 0 0 0 10>}}
  }
  PilotPortrait_3 {
   widgetClassPath=&WBP_ToyFox_PilotPortrait
   bIsVariable=true
   parent=GridPanel_39
   widget=PilotPortrait_3
   properties {pilotIndex=2}
   slot {class=GridSlot, properties {row=3, padding=<ltrb 0 0 0 10>}}
  }
  PilotPortrait_4 {
   widgetClassPath=&WBP_ToyFox_PilotPortrait
   bIsVariable=true
   parent=GridPanel_39
   widget=PilotPortrait_4
   properties {pilotIndex=3}
   slot {class=GridSlot, properties {row=4, padding=<ltrb 0 0 0 10>}}
  }
  PilotPortrait_5 {
   widgetClassPath=&WBP_ToyFox_PilotPortrait
   bIsVariable=true
   parent=GridPanel_39
   widget=PilotPortrait_5
   properties {pilotIndex=4}
   slot {class=GridSlot, properties {row=5, padding=<ltrb 0 0 0 10>}}
  }
  TutorialHighlight {widgetClassPath=&TutorialHighlight}
 ]
}
